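; Blueprint function library of shared "_Base" helpers for Fortnite text chat widgets, with viewmodels, widgets and delegates passed in by callers.
; Channel functions rebuild tabs with the add command last and toggle input for read-only channels.
; Message functions send text, insert date and "New" dividers, and manage list scrolling and selection.
; Context menu functions bind mention, profile, mute and block actions, and build whisper friend lists.
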(fn AddAdditionalWhisperContextInfo_Base
  (params (Outer :type Object%20Reference) (Object :type Object%20Reference) (Event :type Delegate%20%28by%20ref%29))
  (bind _returnvalue (Game|ConstructObjectfromClass "/Script/CommonUI.CommonTextBlock" Outer true 0 0 "Orient_Horizontal"))
  ;[1 1 1] Adding style sheet for text
  (StyleSheets|UpdateWidgetStyleTags _returnvalue (LocalVariables|GetInAddedTags) (LocalVariables|GetInRemovedTags))
  (Appearance|SetColorandOpacity _returnvalue (LocalVariables|GetTextColor) :!class /Script/UMG.TextBlock)
  (bind _aswbp_uikit_context_menu (Utilities|Casting|CastToWBP_UIKit_ContextMenu Object))
  (bind _returnvalue_2 (Widget|AddChildtoGrid (Variables|/FortUILibrary/Systems/ContextMenu/WBP_UIKit_ContextMenu.WBP_UIKit_ContextMenu_C|GetRoot _aswbp_uikit_context_menu) _returnvalue 0 0))
  (Layout|GridSlot|SetHorizontalAlignment _returnvalue_2 "HAlign_Left")
  (Layout|GridSlot|SetVerticalAlignment _returnvalue_2 "VAlign_Center")
  (Layout|BorderSlot|SetPadding _returnvalue_2 (struct :Left 32.0 :Top 18.0 :Right 32.0 :Bottom 18.0) :!class /Script/UMG.GridSlot)
  (Variables|/Script/UMG.TextBlock|SetFont (LocalVariables|GetWhisper_ContextMenu_AdditionalContent_FontStyle) _returnvalue)
  (Variables|/Script/UMG.TextBlock|SetText "NSLOCTEXT(\"[DA2FC953810905D310C177B6FAA7FF88]\", \"9A74E0D9426BAEBC13B695B565725807\", \"SEND A DM\")" _returnvalue)
  ;[1 1 1] Binding for gen context menu items so we can modify any data without creating a new widget
  (Events|BindEventtoOnEntryGenerated (Variables|/FortUILibrary/Systems/ContextMenu/WBP_UIKit_ContextMenu.WBP_UIKit_ContextMenu_C|GetContent _aswbp_uikit_context_menu) Event)
  (return _aswbp_uikit_context_menu))

(fn BlockClickedConfirm_Base
  (params (ChatVM :type MVVM%20Base%20Viewmodel%20Object%20Reference) (MessageID :type Name) (DialogRef :type Object%20Reference) (SelfTarget :type Common%20Activatable%20Widget%20Object%20Reference))
  (MessageActions|PerformMessageActionBase (Variables|GetChatVM) (Variables|GetMessageID) "Block")
  (bind _aswbp_dialog_vm_text_chat_block (Utilities|Casting|CastToWBP_DialogVM_TextChat_Block (Variables|GetDialogRef)))
  (ActivatableWidget|DeactivateWidget _aswbp_dialog_vm_text_chat_block)
  (ActivatableWidget|RequestRefreshFocus (Variables|GetSelfTarget)))

(fn BlockPlayer_Base
  (params (Username :type String) (BlockPlayerConfirmedEvent :type Delegate%20%28by%20ref%29) (SelfTarget :type User%20Widget%20Object%20Reference))
  (bind _returnvalue (UserInterface|CreateWidget 0 "/FortniteCoreUI/UI/Social/Chat/TextChat/Dialogs/WBP_DialogVM_TextChat_Block.WBP_DialogVM_TextChat_Block_C" 0 0))
  (Class|WBPDialogVMTextChatBlock|SetDescriptionText _returnvalue (Utilities|Text|ToText(String) (Variables|GetUsername)))
  (Default|BindEventtoOnClickedButtonConfirm _returnvalue (Variables|GetBlockPlayerConfirmedEvent))
  (FortUIManagement|TryPushWidgetasModalPanel (Variables|GetSelfTarget) _returnvalue))

(fn FilloutChannelData_Base
  (params (TabGroup :type WBP%20UIKit%20Tab%20Group%20Base%20Object%20Reference) (Channels :type Array%20of%20Fort%20Text%20Chat%20Channel%20VM%20Object%20References) (TabAddCmd :type Name) (SelfTarget :type Common%20Activatable%20Widget%20Object%20Reference))
  (bind _currentlyselectedtabid (LocalVariables|GetCurrentlySelectedTabID))
  (bind _channels (Variables|GetChannels))
  (LocalVariables|SetCurrentlySelectedTabID (TabList|GetSelectedTabId TabGroup))
  (LocalVariables|SetAddCmdVM 0)
  (TabList|RemoveAllTabs (Variables|GetTabGroup))
  (if (Utilities|Array|IsNotEmpty _channels)
    ;[1 0.9987 0] Add a New Tab "chats". Using index 0 as a placeholder.
    (UIKitTabGroup|AddTab (Variables|GetTabGroup) (Utilities|Array|Get(acopy) _channels) "Chats" 0)
    (for (_array_element _array_index) (Variables|GetChannels)
      (if (!= (Variables|GetTabAddCmd) (Variables|/Script/FortniteUI.FortTextChatChannelVM|GetChannelId _array_element))
        ;[1 1 1] Adding tabs
        (UIKitTabGroup|AddTab (Variables|GetTabGroup) _array_element (Variables|/Script/FortniteUI.FortTextChatChannelVM|GetChannelId _array_element) (+ _array_index 1))
        (else
          (LocalVariables|SetAddCmdVM _array_element) ; Cache "Add Cmd" so we can add it to the end of the tab list
        )))
    ;[1 0.9987 0] Moved "Add Command". To last index to match UX.
    (bind _addcmdvm (LocalVariables|GetAddCmdVM)
      (:then
        (UIKitTabGroup|AddTab (Variables|GetTabGroup) _addcmdvm (Variables|/Script/FortniteUI.FortTextChatChannelVM|GetChannelId _addcmdvm) (TabList|GetTabCount (Variables|GetTabGroup))))
      (:else
        _))
    (if (and (!= _currentlyselectedtabid "") (not (== _currentlyselectedtabid (Variables|GetTabAddCmd))))
      (TabList|SelectTabbyID (Variables|GetTabGroup) _currentlyselectedtabid true)
      (ActivatableWidget|RequestRefreshFocus (Variables|GetSelfTarget)))))

(fn FillOutMessageData_Base
  (params (MessageList :type Fort%20Polymorphic%20List%20View%20Object%20Reference) (Messages :type Array%20of%20Object%20References) (Switcher :type Common%20Visibility%20Switcher%20Object%20Reference) (ClippingBox :type Widget%20Object%20Reference) (BlankMessage :type Widget%20Object%20Reference))
  (bind _messages (Variables|GetMessages))
  (bind _returnvalue (Utilities|Array|IsEmpty _messages))
  (bind _listview (Variables|/Script/FortniteUI.FortPolymorphicListView|GetListView MessageList))
  (LocalVariables|SetbScrollToBottom (ListView|IsItemVisible _listview (ListView|GetItemAt _listview (- (ListView|GetNumItems _listview) 1))))
  (ListView|ClearListItems _listview)
  (CommonVisibilitySwitcher|SetActiveWidget (Variables|GetSwitcher) (select _returnvalue (Variables|GetBlankMessage) (Variables|GetClippingBox))) ; temp
  (if (not _returnvalue)
    (ListView|SetListItems _listview (Variables|GetMessages))
    (if (LocalVariables|GetbScrollToBottom)
      (ListViewBase|ScrolltoBottom _listview))))

(fn GetInputFieldTextHint_Base
  (params (IsRecordingCurrentChannel :type Boolean) (IsCurrentChannelReadOnly :type Boolean) (CurrentChannelName :type String) (CurrentChannelType :type EFortTextChatChannelType%20Enum :default "All"))
  (if (Variables|GetIsCurrentChannelReadOnly)
    ;[1 1 1] Frontend Only , for Game Tab
    (if (and (not (GlobalUIContext|IsinZone (LocalPlayerSubsystems|GetFortGlobalUIContext 0))) (Utilities|Enum|Equal(Enum) (Variables|GetCurrentChannelType) "Game"))
      (return "NSLOCTEXT(\"[DA2FC953810905D310C177B6FAA7FF88]\", \"F694BA5F4ECB95702CD88BBE90CD137C\", \"You are not in a game\")")
      (else
        ;[1 1 1] If channel is disabled, get the name of it
        (return (Utilities|Text|FormatText "NSLOCTEXT(\"[DA2FC953810905D310C177B6FAA7FF88]\", \"640DEEAA4FB2FF4CC02AFDB9F4BF8E0F\", \"{ChannelName} not available\")" (Variables|GetCurrentChannelName)))))
    (else
      ;[1 1 1] when channel is enabled, show reporting status
      (return (select (Variables|GetIsRecordingCurrentChannel) "NSLOCTEXT(\"[DA2FC953810905D310C177B6FAA7FF88]\", \"ABA24D43443247728D9E8C80A5D5976F\", \"Text reporting on\")" "NSLOCTEXT(\"[DA2FC953810905D310C177B6FAA7FF88]\", \"2C2F889142822BCCE1D6B6AE3162ACE6\", \"Text reporting off\")")))))

(fn HandleSendMessage_Base
  (params (MessageText :type Text%20%28by%20ref%29) (ChatMainVM :type Fort%20Text%20Chat%20Main%20VM%20Object%20Reference) (TextInput :type WBP%20UIKit%20Input%20Field%20Base%20Object%20Reference) (MessageList :type Fort%20Polymorphic%20List%20View%20Object%20Reference) (SelfTarget :type Common%20Activatable%20Widget%20Object%20Reference))
  (bind _messagetext (Variables|GetMessageText))
  (bind _chatmainvm (Variables|GetChatMainVM))
  (if (not (Utilities|Text|TextIsEmpty _messagetext))
    (Variables|/Script/FortniteUI.FortTextChatSendingMessageVM|SetMessageToSend (Utilities|Text|ToText(String) (Utilities|String|ToString(Text) _messagetext)) (Variables|/Script/FortniteUI.FortTextChatMainVM|GetMessageToSend _chatmainvm))
    (Class|FortTextChatMainVM|Send _chatmainvm)
    (UIKitInputField|SetText (Variables|GetTextInput) "") ; Clearing input field text after text committing
    ;[0 0 0] This event triggers when the sender is actually you. So skip all unread messages and scroll to the end of the list
    ; TODO: Call function to set unread messages to 0
    (ListViewBase|ScrolltoBottom (Variables|/Script/FortniteUI.FortPolymorphicListView|GetListView (Variables|GetMessageList)))
    (ActivatableWidget|RequestRefreshFocus (Variables|GetSelfTarget))))

(fn InitializeContextMenu_Base
  (params (ContextAchor :type WBP%20UIKit%20Context%20Menu%20Anchor%20Button%20Base%20Object%20Reference) (Outer :type Object%20Reference))
  (bind _returnvalue (Game|ConstructObjectfromClass "/Script/FortniteUI.UIKitContextMenuActionContext" Outer))
  (Initialization|SetContext (Variables|GetContextAchor) _returnvalue)
  (return _returnvalue))

(fn InitilizeBindEvents_Base
  (params (WhisperContextMenu :type WBP%20UIKit%20Context%20Menu%20Anchor%20Button%20Base%20Object%20Reference) (MessageList :type Fort%20Polymorphic%20List%20View%20Object%20Reference) (IsSideBarChat :type Boolean) (WhisperInitFinished :type Delegate%20%28by%20ref%29) (MessageEnteryGenerated :type Delegate%20%28by%20ref%29))
  (Default|BindEventtoInitFinished WhisperContextMenu (Variables|GetWhisperInitFinished))
  (Events|BindEventtoOnEntryGenerated (Variables|/Script/FortniteUI.FortPolymorphicListView|GetListView (Variables|GetMessageList)) (Variables|GetMessageEnteryGenerated)))

(fn InsertDividersIntoMessageList_Base
  (params (MessageVMs :type Array%20of%20Text%20Chat%20Message%20VM%20Object%20References) (OwningWidget :type Object%20Reference))
  (bind _output_get (LocalVariables|SetInMessageVMs MessageVMs))
  (for (_array_element _array_index) _output_get
    (LocalVariables|SetArrayIndex _array_index)
    (LocalVariables|SetArrayItem _array_element)
    (Utilities|FlowControl|Sequence
      (:then_0
        ;[1 1 1] Add Timestamp divider BEFORE the first message of a new day.  And always make a timestamp divider for the first message in the list.
        ; (Code sets IsFirstMessageOfANewDay by comparing a previous message, but for the first item in the list, there is no previous message.
        (if (or (== _array_index 0) (Variables|/Script/FortniteUI.FortTextChatMessageVM|GetbIsFirstMessageOfANewDay (LocalVariables|GetArrayItem)))
          (bind _returnvalue (Game|ConstructObjectfromClass "/FortniteCoreUI/UI/Social/Chat/TextChat/Data/BP_VM_TextChat_DateMarker_Entry.BP_VM_TextChat_DateMarker_Entry_C" (Variables|GetOwningWidget)))
          (Variables|/FortniteCoreUI/UI/Social/Chat/TextChat/Data/BP_VM_TextChat_DateMarker_Entry.BP_VM_TextChat_DateMarker_Entry_C|SetDateReceived (Variables|/Script/FortniteUI.FortTextChatMessageVM|GetDateReceived (LocalVariables|GetArrayItem)) _returnvalue)
          (Utilities|Array|Add (LocalVariables|GetMessageListWithDividers) _returnvalue)))
      (:then_1
        ;[1 1 1] Add the message
        (Utilities|Array|Add (LocalVariables|GetMessageListWithDividers) (LocalVariables|GetArrayItem)))
      (:then_2
        ;[1 1 1] Add a "New" divider AFTER the last read message.
        (if (and (Variables|/Script/FortniteUI.FortTextChatMessageVM|GetbIsLastReadMessage (LocalVariables|GetArrayItem)) (!= (LocalVariables|GetArrayIndex) (Utilities|Array|LastIndex (LocalVariables|GetInMessageVMs))))
          (bind _returnvalue_2 (Game|ConstructObjectfromClass "/FortniteCoreUI/UI/Social/Chat/TextChat/Data/BP_VM_TextChat_NewMarker_Entry.BP_VM_TextChat_NewMarker_Entry_C" (Variables|GetOwningWidget)))
          (Utilities|Array|Add (LocalVariables|GetMessageListWithDividers) _returnvalue_2)))))
  ;[1 1 1] Return the List with any necessary dividers.
  (return (LocalVariables|GetMessageListWithDividers)))

(fn ListViewEntryGenerated_Base
  (params (Widget :type User%20Widget%20Object%20Reference) (ListView :type Fort%20Polymorphic%20List%20View%20Object%20Reference))
  (bind _listview (Variables|GetListView))
  (bind _listview_2 (Variables|/Script/FortniteUI.FortPolymorphicListView|GetListView _listview))
  (bind _aswbp_text_chat_entry_message_base (Utilities|Casting|CastToWBP_TextChat_Entry_MessageBase Widget))
  (if (== (Variables|/FortniteCoreUI/UI/Social/Chat/TextChat/WBP_TextChat_Entry_MessageBase.WBP_TextChat_Entry_MessageBase_C|GetIndexOfItem _aswbp_text_chat_entry_message_base) (- (ListView|GetNumItems (Variables|/Script/FortniteUI.FortPolymorphicListView|GetListView (Variables|GetListView))) 1))
    (if (not (Utilities|Enum|NotEqual(Enum) (CommonInputSubsystem|GetCurrentInputType (LocalPlayerSubsystems|GetCommonInputSubsystem 0)) "Gamepad"))
      (ListView|SetSelectedIndex _listview_2 (Variables|/FortniteCoreUI/UI/Social/Chat/TextChat/WBP_TextChat_Entry_MessageBase.WBP_TextChat_Entry_MessageBase_C|GetIndexOfItem _aswbp_text_chat_entry_message_base)))
    (ListViewBase|ScrolltoBottom _listview_2)))

(fn Mention_Base
  (params (ChatMainVM :type Fort%20Text%20Chat%20Main%20VM%20Object%20Reference) (ChatInput :type WBP%20UIKit%20Input%20Field%20Base%20Object%20Reference) (MessageID :type Name))
  (bind _messageid (Variables|GetMessageID))
  (bind _returnvalue (Utilities|String|ToString(Name) _messageid))
  (bind _returnvalue_2 (Class|FortTextChatMainVM|GetMessageVM ChatMainVM _returnvalue))
  (Utilities|IsValid _returnvalue_2
    (:Is%20Valid
      (UIKitInputField|SetText (Variables|GetChatInput) (Utilities|Text|FormatText "NSLOCTEXT(\"[DA2FC953810905D310C177B6FAA7FF88]\", \"B852685A4CCB2705B22BA08C9F02BD31\", \"@{DisplayName}\")" (Variables|/Script/FortniteUI.FortTextChatMessageVM|GetDisplayName _returnvalue_2))))))

(fn MessageClickedEvent_Base
  (params (ContextMenu :type UIKit%20Context%20Menu%20Anchor%20Object%20Reference) (ButtonWidget :type Object%20Reference))
  (MenuAnchor|Open (Variables|/Script/FortniteUI.UIKitContextMenuAnchor|GetAnchor (Variables|GetContextMenu)) true)
  (bind _aswbp_text_chat_entry_user_message (Utilities|Casting|CastToWBP_TextChat_Entry_UserMessage (Variables|GetButtonWidget)))
  (Class|WBPTextChatEntryUserMessage|SetBackgroundVisibility _aswbp_text_chat_entry_user_message (MenuAnchor|IsOpen (Variables|/Script/FortniteUI.UIKitContextMenuAnchor|GetAnchor (Variables|GetContextMenu))))
  (return _aswbp_text_chat_entry_user_message))

(fn MessageContextClosed_Base
  (params (TextChatEntry :type WBP%20Text%20Chat%20Entry%20User%20Message%20Object%20Reference) (ContextMenu :type UIKit%20Context%20Menu%20Anchor%20Object%20Reference))
  (Class|WBPTextChatEntryUserMessage|SetBackgroundVisibility TextChatEntry (MenuAnchor|IsOpen (Variables|/Script/FortniteUI.UIKitContextMenuAnchor|GetAnchor ContextMenu)))
  (bind _output_get (Variables|/FortniteCoreUI/UI/Social/Chat/TextChat/WBP_TextChat_Entry_UserMessage.WBP_TextChat_Entry_UserMessage_C|SetbIsContextIconVisible false TextChatEntry))
  (Class|WBPTextChatEntryUserMessage|SetContextMenuIconVisibility TextChatEntry _output_get))

(fn MutePlayer_Base
  (params (ChatVM :type MVVM%20Base%20Viewmodel%20Object%20Reference) (MessageID :type Name))
  (MessageActions|PerformMessageActionBase ChatVM MessageID "Mute"))

(fn OnInputMethodChanged_Base
  (params (ChatInput :type WBP%20UIKit%20Input%20Field%20Base%20Object%20Reference))
  (if (Widget|IsHovered (Variables|GetChatInput))
    (ChatInput|UpdateInputPaddingBase (Variables|GetChatInput) true)))

(fn OnWhisperUserClicked_Base
  (params (TabGroup :type Common%20Tab%20List%20Widget%20Base%20Object%20Reference) (ChatMainVM :type Fort%20Text%20Chat%20Main%20VM%20Object%20Reference) (WhisperChannelID :type Name))
  (bind _whisperchannelid (Variables|GetWhisperChannelID))
  (Class|FortTextChatMainVM|CreateWhisperChannelifNeeded (Variables|GetChatMainVM) _whisperchannelid)
  (TabList|SelectTabbyID (Variables|GetTabGroup) _whisperchannelid true)
  (Class|FortTextChatMainVM|SetDirectMessageContext (Variables|GetChatMainVM) "TextChat"))

(fn PerformMessageAction_Base
  (params (ChatVM :type MVVM%20Base%20Viewmodel%20Object%20Reference) (MessageID :type Name) (MenuAction :type EFortTextChatMessageContextMenuAction%20Enum :default "Mention"))
  (bind _asfort_text_chat_main_vm (Utilities|Casting|CastToFortTextChatMainVM (Variables|GetChatVM)))
  (Class|FortTextChatMainVM|PerformMessageContextMenuAction _asfort_text_chat_main_vm (Utilities|String|ToString(Name) (Variables|GetMessageID)) (Variables|GetMenuAction)))

(fn Report_Base
  (params (ChatMainVM :type Fort%20Text%20Chat%20Main%20VM%20Object%20Reference) (MessageID :type Name))
  (Class|TextChatMessageVM|Report (Class|FortTextChatMainVM|GetMessageVM ChatMainVM (Utilities|String|ToString(Name) MessageID))))

(fn SendTextMessage_Base
  (params (commitMethod :type ETextCommit%20Enum :default "Default") (MessageText :type Text%20%28by%20ref%29) (ChatMainVM :type Fort%20Text%20Chat%20Main%20VM%20Object%20Reference) (TextInput :type WBP%20UIKit%20Input%20Field%20Base%20Object%20Reference) (MessageList :type Fort%20Polymorphic%20List%20View%20Object%20Reference) (SelfTarget :type Common%20Activatable%20Widget%20Object%20Reference))
  (switch Utilities|FlowControl|Switch|SwitchonETextCommit commitMethod
    (:OnEnter
      (Messages|HandleSendMessageBase (Variables|GetMessageText) (Variables|GetChatMainVM) (Variables|GetTextInput) (Variables|GetMessageList) (Variables|GetSelfTarget)))))

(fn SetInputFieldEnabled_Base
  (params (ChatInput :type WBP%20UIKit%20Input%20Field%20Base%20Object%20Reference) (SendButton :type Widget%20Object%20Reference) (IsChannelReadOnly :type Boolean))
  (bind _ischannelreadonly (Variables|GetIsChannelReadOnly))
  (bind _returnvalue (not _ischannelreadonly))
  (bind _chatinput (Variables|GetChatInput))
  (Widget|SetIsEnabled (Variables|GetChatInput) _returnvalue)
  (if _returnvalue
    (Variables|/Script/UMG.Widget|SetRenderOpacity 1.0 _chatinput)
    (else
      (Variables|/Script/UMG.Widget|SetRenderOpacity 0.4 _chatinput)))
  (Channels|UpdateTabsDataBase (Variables|GetChatInput) _returnvalue)
  (ChatInput|UpdateSendButtonVisibilityBase (Variables|GetSendButton) _returnvalue))

(fn SetMessageContextMenuActions_Base
  (params (ContextMenuActionContext :type UIKit%20Context%20Menu%20Action%20Context%20Object%20Reference) (ChatVM :type MVVM%20Base%20Viewmodel%20Object%20Reference) (MessageID :type Name) (MentionEvent :type Delegate%20%28by%20ref%29) (ViewProfileEvent :type Delegate%20%28by%20ref%29) (ReportEvent :type Delegate%20%28by%20ref%29) (BlockEvent :type Delegate%20%28by%20ref%29) (UnblockEvent :type Delegate%20%28by%20ref%29) (MutePlayerEvent :type Delegate%20%28by%20ref%29) (UnmutePlayerEvent :type Delegate%20%28by%20ref%29))
  (bind _contextmenuactioncontext (Variables|GetContextMenuActionContext))
  (bind _contextmenuactions (LocalVariables|GetContextMenuActions))
  (bind _asfort_text_chat_main_vm (Utilities|Casting|CastToFortTextChatMainVM (Variables|GetChatVM)))
  (LocalVariables|SetContextMenuActions (Class|FortTextChatMainVM|GetMessageContextMenuActions _asfort_text_chat_main_vm (Utilities|String|ToString(Name) (Variables|GetMessageID))))
  (Class|UIKitContextMenuActionContext|RemoveAllInteractions _contextmenuactioncontext)
  (Class|UIKitContextMenuActionContext|AddInteraction _contextmenuactioncontext (struct :DisplayName "NSLOCTEXT(\"[DA2FC953810905D310C177B6FAA7FF88]\", \"07F4F53A4318D999555940BE0DBAA3E2\", \"Mention\")" :IconTexture "" :IconMaterial 0 :SortPriority 1000 :bIsAvailable false :bIsEnabled true :bIsSelected false :bAddDivider false :bPositionDividerAfter false :bCloseMenuOnSelect false :InteractionNameID "Mention" :Interaction "(DataTable=None,RowName=\"\")" :PressedDelegate "(null).None" :ReleasedDelegate "(null).None") (Variables|GetMentionEvent))
  (Class|UIKitContextMenuActionContext|AddInteraction _contextmenuactioncontext (Utilities|Struct|MakeUIKitContextMenuActionInfo "NSLOCTEXT(\"[DA2FC953810905D310C177B6FAA7FF88]\", \"36216A834E7F334B02EDDD9CC5487CAE\", \"View Profile\")" "" 0 1000 (Utilities|Array|ContainsItem _contextmenuactions "ViewProfile") true false false false false "ViewProfile") (Variables|GetViewProfileEvent))
  (Class|UIKitContextMenuActionContext|AddInteraction _contextmenuactioncontext (Utilities|Struct|MakeUIKitContextMenuActionInfo "NSLOCTEXT(\"[DA2FC953810905D310C177B6FAA7FF88]\", \"1F0B595A403B707899150E992CBD0B3C\", \"Block\")" "" 0 1000 (Utilities|Array|ContainsItem _contextmenuactions "Block") true false false false false "Block") (Variables|GetBlockEvent))
  (Class|UIKitContextMenuActionContext|AddInteraction _contextmenuactioncontext (Utilities|Struct|MakeUIKitContextMenuActionInfo "NSLOCTEXT(\"[DA2FC953810905D310C177B6FAA7FF88]\", \"DC156AC34F0AFFF13A9E2E8FB572F2DC\", \"Unblock\")" "" 0 1000 (Utilities|Array|ContainsItem _contextmenuactions "Unblock") true false false false false "Unblock") (Variables|GetUnblockEvent))
  (Class|UIKitContextMenuActionContext|AddInteraction _contextmenuactioncontext (Utilities|Struct|MakeUIKitContextMenuActionInfo "NSLOCTEXT(\"[DA2FC953810905D310C177B6FAA7FF88]\", \"17BE36584B76F9E80078478763687C3A\", \"Mute Player\")" "" 0 1000 (Utilities|Array|ContainsItem _contextmenuactions "Mute") true false false false false "MutePlayer") (Variables|GetMutePlayerEvent))
  (Class|UIKitContextMenuActionContext|AddInteraction _contextmenuactioncontext (Utilities|Struct|MakeUIKitContextMenuActionInfo "NSLOCTEXT(\"[DA2FC953810905D310C177B6FAA7FF88]\", \"BCCADA3B4CF8E819F29681A2241D2C19\", \"Unmute Player\")" "" 0 1000 (Utilities|Array|ContainsItem _contextmenuactions "Unmute") true false false false false "UnmutePlayer") (Variables|GetUnmutePlayerEvent)))

(fn SetMessageEntryBinding_Base
  (params (UserWidget :type User%20Widget%20Object%20Reference) (ClickedEvent :type Delegate%20%28by%20ref%29))
  (bind _aswbp_text_chat_entry_user_message (Utilities|Casting|CastToWBP_TextChat_Entry_UserMessage UserWidget))
  (Default|BindEventtoOnMessageClicked _aswbp_text_chat_entry_user_message ClickedEvent))

(fn SetSendButtonHeight_Base
  (params (Target :type Common%20Button%20Base%20Object%20Reference) (SendButton :type Common%20Button%20Base%20Object%20Reference) (TextChatInput :type Common%20Button%20Base%20Object%20Reference))
  (CommonButton|Setters|SetMinDimensions Target (Variables|/Script/CommonUI.CommonButtonBase|GetMinWidth SendButton) (Variables|/Script/CommonUI.CommonButtonBase|GetMinHeight TextChatInput)))

(fn SetTabGroupStyle_Base
  (params (TabGroup :type WBP%20UIKit%20Tab%20Group%20Base%20Object%20Reference))
  (Variables|/FortUILibrary/Systems/Tabs/WBP_UIKit_TabGroup_Base.WBP_UIKit_TabGroup_Base_C|SetBackplateClass "/FortUILibrary/UIKit/Backplates/WBP_UIKit_Backplate_Fallback.WBP_UIKit_Backplate_Fallback_C" TabGroup)
  (Variables|/FortUILibrary/Systems/Tabs/WBP_UIKit_TabGroup_Base.WBP_UIKit_TabGroup_Base_C|SetBackplateBrightness "NewEnumerator3" TabGroup)
  (Variables|/FortUILibrary/Systems/Tabs/WBP_UIKit_TabGroup_Base.WBP_UIKit_TabGroup_Base_C|SetPromptsBackplateClass "/FortUILibrary/UIKit/Backplates/WBP_UIKit_Backplate_Fallback.WBP_UIKit_Backplate_Fallback_C" TabGroup)
  (Variables|/FortUILibrary/Systems/Tabs/WBP_UIKit_TabGroup_Base.WBP_UIKit_TabGroup_Base_C|SetPromptBackplateBrightness "NewEnumerator3" TabGroup)
  (Variables|/FortUILibrary/Systems/Tabs/WBP_UIKit_TabGroup_Base.WBP_UIKit_TabGroup_Base_C|SetArrowButtonsClass "/FortniteCoreUI/UI/Social/Chat/TextChat/WBP_UIKit_Button_TabNavArrow_TextChat.WBP_UIKit_Button_TabNavArrow_TextChat_C" TabGroup)
  (Variables|/FortUILibrary/Systems/Tabs/WBP_UIKit_TabGroup_Base.WBP_UIKit_TabGroup_Base_C|SetUsePromptBackplate false TabGroup))

(fn SetWhisperContextMenuActions_Base
  (params (ContextMenu :type WBP%20UIKit%20Context%20Menu%20Anchor%20Button%20Base%20Object%20Reference) (ChatVM :type MVVM%20Base%20Viewmodel%20Object%20Reference) (UIKitContextMenuRef :type WBP%20UIKit%20Context%20Menu%20Object%20Reference) (WhisperUserClickedEvent :type Delegate%20%28by%20ref%29) (NoFriendsAvailableEvent :type Delegate%20%28by%20ref%29))
  (bind _whisperdestinations (LocalVariables|GetWhisperDestinations))
  (bind _returnvalue (Utilities|Array|IsNotEmpty _whisperdestinations))
  (bind _contextmenu (Variables|GetContextMenu))
  (bind _uikitcontextmenuref (Variables|GetUIKitContextMenuRef))
  (bind _sb_maxwidth (Variables|/FortUILibrary/Systems/ContextMenu/WBP_UIKit_ContextMenu.WBP_UIKit_ContextMenu_C|GetSB_MaxWidth _uikitcontextmenuref))
  (bind _interactioncontext (Variables|/FortUILibrary/Systems/ContextMenu/WBP_UIKit_ContextMenuAnchorButton_Base.WBP_UIKit_ContextMenuAnchorButton_Base_C|GetInteractionContext _contextmenu))
  (bind _asfort_text_chat_main_vm (Utilities|Casting|CastToFortTextChatMainVM (Variables|GetChatVM)))
  (LocalVariables|SetWhisperDestinations (Class|FortTextChatMainVM|GetWhisperDestinations _asfort_text_chat_main_vm))
  (Class|UIKitContextMenuActionContext|RemoveAllInteractions _interactioncontext)
  (LocalVariables|SetbIsWhisperListEmpty (not _returnvalue))
  (if _returnvalue
    (for (_array_element _array_index) _whisperdestinations
      (Class|UIKitContextMenuActionContext|AddInteraction _interactioncontext (Utilities|Struct|MakeUIKitContextMenuActionInfo (Utilities|Text|ToText(String) (Variables|/Script/FortniteUI.FortTextChatWhisperDestinationVM|GetWhisperDestinationDisplayName _array_element)) "" 0 1000 true true false false false false (Variables|/Script/FortniteUI.FortTextChatWhisperDestinationVM|GetWhisperDestinationId _array_element)) (Variables|GetWhisperUserClickedEvent)))
    (if (> (stale _array_index) 3)
      ;[1 1 1] Manually locking Context menu height and width
      (Layout|SizeBox|SetMaxDesiredHeight _sb_maxwidth 2.1e+02)
      (Layout|SizeBox|SetWidthOverride _sb_maxwidth 308.0)
      (else
        _))
    (return (LocalVariables|GetbIsWhisperListEmpty))
    (else
      (Class|UIKitContextMenuActionContext|AddInteraction (Variables|/FortUILibrary/Systems/ContextMenu/WBP_UIKit_ContextMenuAnchorButton_Base.WBP_UIKit_ContextMenuAnchorButton_Base_C|GetInteractionContext (Variables|GetContextMenu)) (struct :DisplayName "NSLOCTEXT(\"[DA2FC953810905D310C177B6FAA7FF88]\", \"064D9B0C4731C829FC39E8B6B2F5ABA8\", \"No friends available\")" :IconTexture "" :IconMaterial 0 :SortPriority 1000 :bIsAvailable true :bIsEnabled true :bIsSelected false :bAddDivider false :bPositionDividerAfter false :bCloseMenuOnSelect false :InteractionNameID "No friends available" :Interaction "(DataTable=None,RowName=\"\")" :PressedDelegate "(null).None" :ReleasedDelegate "(null).None") (Variables|GetNoFriendsAvailableEvent)))))

(fn SetWhisperItemVisual_Base
  (params (Object :type Object%20Reference) (IsWhisperListEmpty :type Boolean))
  (if (not IsWhisperListEmpty)
    (bind _aswbp_uikit_context_menu_list_entry (Utilities|Casting|CastToWBP_UIKit_ContextMenuListEntry Object))
    ;[1 1 1] Setting visual data icons for items
    (UIKitButtonGeneric|SetIconMaterial _aswbp_uikit_context_menu_list_entry "/FortUILibrary/Systems/StatusIndicator/MaterialInstances/MI_UIKit_StatusIndicator_Background.MI_UIKit_StatusIndicator_Background")
    (UIKitButtonGeneric|SetIconSize _aswbp_uikit_context_menu_list_entry (Math|Vector2D|MakeVector2D 12.0 12.0))
    (UIKitButtonGeneric|SetIconPadding _aswbp_uikit_context_menu_list_entry (struct :Left 0.0 :Top 0.0 :Right 8.0 :Bottom 0.0))))

(fn SetWhisperList_Base
  (params (FortChatMain :type Fort%20Text%20Chat%20Main%20VM%20Object%20Reference) (TileView :type Common%20Tile%20View%20Object%20Reference))
  (bind _whisperdestinations (LocalVariables|GetWhisperDestinations))
  (bind _returnvalue (Utilities|Array|IsNotEmpty _whisperdestinations))
  (Utilities|Array|Clear (LocalVariables|GetWhisperDestinations))
  (ListView|ClearListItems (Variables|GetTileView))
  (LocalVariables|SetWhisperDestinations (Class|FortTextChatMainVM|GetWhisperDestinations (Variables|GetFortChatMain)))
  (LocalVariables|SetbIsWhisperListEmpty (not _returnvalue))
  (if _returnvalue
    (ListView|SetListItems (Variables|GetTileView) _whisperdestinations)
    (else
      _))
  (return (LocalVariables|GetbIsWhisperListEmpty)))

(fn Unblock_Base
  (params (ChatVM :type MVVM%20Base%20Viewmodel%20Object%20Reference) (MessageID :type Name))
  (MessageActions|PerformMessageActionBase ChatVM MessageID "Unblock"))

(fn UnmutePlayer_Base
  (params (ChatVM :type MVVM%20Base%20Viewmodel%20Object%20Reference) (MessageID :type Name))
  (MessageActions|PerformMessageActionBase ChatVM MessageID "Unmute"))

(fn UpdateChatData_Base
  (params (ChatInput :type WBP%20UIKit%20Input%20Field%20Base%20Object%20Reference) (IsTabActive :type Boolean))
  (if (not IsTabActive)
    (UIKitInputField|SetText (Variables|GetChatInput) "") ; Clearing input field text when tab is switched to read-only 
  ))

(fn UpdateInputPadding_Base
  (params (ChatInput :type WBP%20UIKit%20Input%20Field%20Base%20Object%20Reference) (Hovered :type Boolean))
  (Layout|BorderSlot|SetPadding (Slot|SlotasGridSlot (Variables|/FortUILibrary/Systems/InputField/WBP_UIKit_InputField_Base.WBP_UIKit_InputField_Base_C|GetBlock_EditableText ChatInput)) (Utilities|Struct|MakeMargin 0.0 0.0 (select (Variables|GetHovered) (Utilities|Select :Index (CommonInputSubsystem|GetCurrentInputType (LocalPlayerSubsystems|GetCommonInputSubsystem 0)) :MouseAndKeyboard 0.0 :Gamepad 2e+01 :Touch 0.0 :Count 0.0) 0.0)) :!class /Script/UMG.GridSlot))

(fn UpdateSendButtonVisibility_Base
  (params (SendButton :type Widget%20Object%20Reference) (IsInputEnabled :type Boolean))
  (Widget|SetVisibility (Variables|GetSendButton) (select IsInputEnabled "Visible" "Collapsed")))

(fn UpdateTabsData_Base
  (params (ChatInput :type WBP%20UIKit%20Input%20Field%20Base%20Object%20Reference) (is%20Tab%20Active :type Boolean))
  (if (not is%20Tab%20Active)
    (UIKitInputField|SetText (Variables|GetChatInput) "") ; Clearing input field text when tab is switched to read-only 
  ))

(fn UpdateTabSelected_Base
  (params (TabAddCmd :type Name) (Tab%20ID :type Name%20%28by%20ref%29) (FortTextChatMainVM :type Fort%20Text%20Chat%20Main%20VM%20Object%20Reference) (ChatInput :type WBP%20UIKit%20Input%20Field%20Base%20Object%20Reference) (BlankMessage :type WBP%20Text%20Chat%20Blank%20Message%20Object%20Reference) (ContextAchor :type UIKit%20Context%20Menu%20Anchor%20Object%20Reference) (MessageListView :type Fort%20Polymorphic%20List%20View%20Object%20Reference) (SelfTarget :type Common%20Activatable%20Widget%20Object%20Reference) (UseContextAnchor :type Boolean))
  (bind _forttextchatmainvm (Variables|GetFortTextChatMainVM))
  (bind _biscurrentchannelreadonly (Variables|/Script/FortniteUI.FortTextChatMainVM|GetbIsCurrentChannelReadOnly _forttextchatmainvm))
  (bind _returnvalue (not _biscurrentchannelreadonly))
  (bind _messagelistview (Variables|GetMessageListView))
  (bind _listview (Variables|/Script/FortniteUI.FortPolymorphicListView|GetListView _messagelistview))
  (if (!= (Variables|GetTabAddCmd) Tab%20ID)
    (Class|FortTextChatMainVM|ChangeCurrentChannel _forttextchatmainvm Tab%20ID)
    (Channels|UpdateTabsDataBase (Variables|GetChatInput) _returnvalue)
    (Class|WBPTextChatBlankMessage|SetBlankMessageText (Variables|GetBlankMessage) (Variables|/Script/FortniteUI.FortTextChatMainVM|GetCurrentChannelType _forttextchatmainvm) _returnvalue)
    (if (not (and _returnvalue (Utilities|Enum|Equal(Enum) (CommonInputSubsystem|GetCurrentInputType (LocalPlayerSubsystems|GetCommonInputSubsystem 0)) "Gamepad")))
      (ActivatableWidget|RequestRefreshFocus (Variables|GetSelfTarget)))
    (ListViewBase|ScrolltoBottom _listview)
    (if (Utilities|Enum|NotEqual(Enum) (CommonInputSubsystem|GetCurrentInputType (LocalPlayerSubsystems|GetCommonInputSubsystem 0)) "Touch")
      (ListView|SetSelectedIndex _listview (- (ListView|GetNumItems _listview) 1)))))

(fn WhisperTabSelected_Base
  (params (Tab%20VM :type MVVM%20Base%20Viewmodel%20Object%20Reference) (ChatVM :type MVVM%20Base%20Viewmodel%20Object%20Reference))
  (bind _asfort_text_chat_channel_vm (Utilities|Casting|CastToFortTextChatChannelVM Tab%20VM))
  (if (Utilities|Enum|Equal(Enum) (Variables|/Script/FortniteUI.FortTextChatChannelVM|GetChannelType _asfort_text_chat_channel_vm) "Whisper")
    (bind _asfort_text_chat_main_vm (Utilities|Casting|CastToFortTextChatMainVM (Variables|GetChatVM)))
    (Class|FortTextChatMainVM|SetDirectMessageContext _asfort_text_chat_main_vm "ChannelTab")))
