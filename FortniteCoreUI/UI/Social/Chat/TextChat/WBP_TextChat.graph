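# WBP_TextChat is the Fortnite social text chat widget, with channel tabs, a message list, single or multiline input and speech-to-text dictation.
# Component events drive FortTextChatPlayerVM to switch channels, send messages, block, report, stop TTS and record dictation.
# Chat, whisper and message context menus use action contexts, with the first two injecting constructed header widgets.
# Focus, scrolling and padding adapt to input method, with config-gated repositioning above virtual keyboards.
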WidgetBlueprint /FortniteCoreUI/UI/Social/Chat/TextChat/WBP_TextChat
graphs [
 AddingAdditionalContentForChatContextMenu {nodes [
   n1 = AddingAdditionalContentForChatContextMenu!() -> (then, Object:Object Reference) {is=title K2Node_FunctionEntry}
   > n2 = Construct Object from Class(Class=&HorizontalBox) -> (ReturnValue:Horizontal Box Object Reference)
   > n3 = Cast To WBP_UIKit_ContextMenu!(execute, Object=@n1) -> (then, CastFailed, AsWBP UIKit Context Menu:WBP UIKit Context Menu Object Reference) {is=title K2Node_DynamicCast}
   > n4 = AddChildToGrid!(
    execute
    self:Grid Panel Object Reference=Get Root!(self:WBP UIKit Context Menu Object Reference=@n3) -> (Root:Grid Panel Object Reference)
    Content:Widget Object Reference=@n2
    InRow:int=0
    InColumn:int=0
   ) -> (then, ReturnValue:Grid Slot Object Reference)
   > SetLayer!(execute, self:Grid Slot Object Reference=@n4, InLayer:int=2) -> (then)
   > SetHorizontalAlignment(self:Grid Slot Object Reference=@n4, InHorizontalAlignment=HAlign_Fill)
   > SetVerticalAlignment(self:Grid Slot Object Reference=@n4, InVerticalAlignment=VAlign_Center)
   > SetPadding!(
    execute
    self:Grid Slot Object Reference=@n4
    InPadding_Left:float=32
    InPadding_Top:float=12
    InPadding_Right:float=32
    InPadding_Bottom:float=12
   ) -> (then)
   > n5 = Construct Object from Class!(
    execute
    Class=&CommonTextBlock
    self
    bIsScrollingEnabled:bool=True
    Style:Common Text Style Class Reference
    ScrollStyle:Common Text Scroll Style Class Reference
    ScrollOrientation:EOrientation Enum=Orient_Horizontal
   ) -> (then, ReturnValue:Common Text Object Reference)
   > n6 = Set Text_ChatContextMenu(Text_ChatContextMenu:Common Text Object Reference=@n5) -> (Output_Get:Common Text Object Reference)
   > UpdateWidgetStyleTags!(
    execute
    InWidget:Widget Object Reference=@n6
    InAddedTags:Array of Gameplay Tag Structures=Get InAddedTags2() -> (InAddedTags2:Array of Gameplay Tag Structures)
    InRemovedTags:Array of Gameplay Tag Structures=Get InRemovedTags2() -> (InRemovedTags2:Array of Gameplay Tag Structures)
   ) -> (then, ReturnValue:bool)
   > AddChildToHorizontalBox!(execute, self:Horizontal Box Object Reference=@n2, Content:Widget Object Reference=@n5) -> (then, ReturnValue:Horizontal Box Slot Object Reference)
   n7 = Get Image_ChatContextMenu() -> (Image_ChatContextMenu:Image Object Reference)
   > SetBrushFromTexture!(
    execute
    self:Image Object Reference=@n7
    Texture:Texture 2D Object Reference=/Game/Athena/UI/InGame/Creative/PropertyEditor/Icon_Picker/Textures/128x128/T_UI_IconLibrary_shield_128
    bMatchSize:bool=false
   ) -> (then)
   > SetPadding(
    self:Horizontal Box Slot Object Reference=SlotAsHorizontalBoxSlot!(Widget:Widget Object Reference=@n7) -> (ReturnValue:Horizontal Box Slot Object Reference)
    InPadding_Left=0
    InPadding_Top=0
    InPadding_Right=10
    InPadding_Bottom=0
   )
   n8 = Get Text_ChatContextMenu() -> (Text_ChatContextMenu:Common Text Object Reference)
   > "SetText (Text)"(self=@n8, InText="LOCTEXT(\"SuperLongText\")")
   > SetColorAndOpacity!(
    execute
    self=@n8
    InColorAndOpacity_SpecifiedColor:Linear Color Structure=Get RecordingColorOff() -> (RecordingColorOff:Linear Color Structure)
    InColorAndOpacity_ColorUseRule:ESlateColorStylingMode Enum
   ) -> (then)
   n9 = Construct Object from Class!(execute, Class=&Image, self) -> (then, ReturnValue:Image Object Reference)
   > n10 = Set Image_ChatContextMenu(Image_ChatContextMenu:Image Object Reference=@n9) -> (Output_Get:Image Object Reference)
   > AddChildToHorizontalBox(self=@n2, Content=@n10)
  ]}
 AddingAdditionalContentForWhisperContextMenu {nodes [
   n1 = AddingAdditionalContentForWhisperContextMenu!() -> (then, Object:Object Reference) {is=title K2Node_FunctionEntry}
   > n2 = Construct Object from Class!(
    execute
    Class=&CommonTextBlock
    self
    bIsScrollingEnabled:bool=True
    Style:Common Text Style Class Reference
    ScrollStyle:Common Text Scroll Style Class Reference
    ScrollOrientation:EOrientation Enum=Orient_Horizontal
   ) -> (then, ReturnValue:Common Text Object Reference)
   > UpdateWidgetStyleTags(
    InWidget=@n2
    InAddedTags=Get InAddedTags() -> (InAddedTags:Array of Gameplay Tag Structures)
    InRemovedTags=Get InRemovedTags() -> (InRemovedTags:Array of Gameplay Tag Structures)
   )
   > SetColorAndOpacity!(execute, self=@n2, InColorAndOpacity:Slate Color Structure=Get TextColor() -> (TextColor:Slate Color Structure)) -> (then)
   > n3 = Cast To WBP_UIKit_ContextMenu(Object=@n1)
   > n4 = Set WBP UIKit Context Menu_Ref(WBP UIKit Context Menu_Ref:WBP UIKit Context Menu Object Reference=@n3) -> (Output_Get:WBP UIKit Context Menu Object Reference)
   > n5 = AddChildToGrid(self=Get Root(self=@n4), Content=@n2, InRow=0, InColumn=0)
   > SetHorizontalAlignment(self=@n5, InHorizontalAlignment=HAlign_Left)
   > SetVerticalAlignment(self=@n5, InVerticalAlignment=VAlign_Center)
   > SetPadding!(
    execute
    self:Grid Slot Object Reference=@n5
    InPadding:Margin Structure=Get WhisperHeadersMargin() -> (WhisperHeadersMargin:Margin Structure)
   ) -> (then)
   > SetMaxDesiredWidth!(
    execute
    self:Size Box Object Reference=Get SB_MaxWidth!(self:WBP UIKit Context Menu Object Reference=@n4) -> (SB_MaxWidth:Size Box Object Reference)
    InMaxDesiredWidth:float=450
   ) -> (then)
   > Set Font!(
    execute
    Font:Slate Font Info Structure=Get Whisper_ContextMenu_AdditionalContent_FontStyle() -> (Whisper_ContextMenu_AdditionalContent_FontStyle:Slate Font Info Structure)
    self:Text Object Reference=@n2
   ) -> (then, Output_Get:Slate Font Info Structure)
   > Set Text(
    Text=GetDestinationsTitle!(self:Fort Text Chat Main VM Object Reference=Get FortTextChatPlayerVM() -> (FortTextChatPlayerVM:Fort Text Chat Player VM Object Reference)) -> (ReturnValue:Text)
    self=@n2
   )
   > Bind Event to On Entry Generated(
    self:List View Base Object Reference=Get Content!(self:WBP UIKit Context Menu Object Reference=@n4) -> (Content:Common List View Object Reference)
    Delegate=Create Event()
   ) {is=title K2Node_AddDelegate}
  ]}
 BP_GetDesiredFocusTarget {nodes [
   Get Desired Focus Target()
   > n1 = Branch(Condition=HasAnyUserFocusedDescendants!(self:Widget Object Reference=Get Message_ContextMenu() -> (Message_ContextMenu:WBP UIKit Context Menu Anchor Button Object Reference)) -> (ReturnValue:bool))
   > Return Node!(
    execute
    ReturnValue:Widget Object Reference=Get ListView!(self:Fort Polymorphic List View Object Reference=Get MessageList() -> (MessageList:WBP Text Chat Message List Object Reference)) -> (ListView:Common List View Object Reference)
   )
   n2 = Get FortTextChatPlayerVM()
   n3 = Branch(
    execute=@n1.else
    Condition=GetInputFieldEnabled!(
     self
     IsChannelReadOnly:bool=IsCurrentChannelReadOnly!(self:Fort Text Chat Main VM Object Reference=@n2) -> (ReturnValue:bool)
     RestrictionType:EFortTextChatRestriction Enum=GetTextChatSettingRestriction!(self:Fort Text Chat Main VM Object Reference=@n2) -> (ReturnValue:EFortTextChatRestriction Enum=Unrestricted)
    ) -> (IsInputFieldEnabled:bool)
   )
   n4 = isInputFieldNotEmpty!(self) -> (NotEmpty:bool)
   n5 = "Not Equal (Enum)"!(A:ECommonInputType Enum=Get CurrentInputType() -> (CurrentInputType:ECommonInputType Enum), B:ECommonInputType Enum=MouseAndKeyboard) -> (ReturnValue:bool)
   n6 = Get is SpeechToText Enabled() -> (is SpeechToText Enabled:bool)
   n7 = Get Multiline InputField Enabled!(self) -> (isEnabled:bool) {is=title K2Node_CallFunction}
   n8 = Get EditableText!(self:WBP UIKit Block Editable Text Object Reference=Get Block_EditableText!(self:WBP UIKit Input Field Base Object Reference=Get WBP_TextChat_Input() -> (WBP_TextChat_Input:WBP Text Chat Input Object Reference)) -> (Block_EditableText:WBP UIKit Block Editable Text Object Reference)) -> (EditableText:UIKit Editable Text Object Reference)
   n9 = Get EditableText(self:WBP Text Chat Block Multiline Editable Text Object Reference=Get Block_EditableText(self:WBP Text Chat Multiline Input Field Object Reference=Get WBP_TextChat_MultilineInputField() -> (WBP_TextChat_MultilineInputField:WBP Text Chat Multiline Input Field Object Reference)) -> (Block_EditableText:WBP Text Chat Block Multiline Editable Text Object Reference)) -> (EditableText:"Text Box (Multi- Line) Object Reference")
   > Return Node(ReturnValue=Select(
     Option 0:Widget Object Reference=Select(Option 0=@n8, Option 1=@n9, Index=@n7)
     Option 1:Widget Object Reference=Get Send_ButtonIcon() -> (Send_ButtonIcon:WBP UIKit Button Generic Object Reference)
     Index=AND Boolean!(A=@n4, B=@n6, C:bool=@n5) -> (ReturnValue)
    ) -> (ReturnValue:Widget Object Reference))
   n10 = Get Multiline InputField Enabled()
   n11 = Get WBP_TextChat_Input()
   n12 = Get WBP_TextChat_MultilineInputField()
   n13 = Branch(execute=@n3.else, Condition=IsHovered(self=Select(Option 0=@n11, Option 1=@n12, Index=@n10)))
   > n14 = RequestTransitionOnBlocks!(
    execute
    self:UIKit Modular Button Object Reference=Select(
     Option 0:UIKit Modular Button Object Reference=Get WBP_TextChat_Input()
     Option 1:UIKit Modular Button Object Reference=Get WBP_TextChat_MultilineInputField()
     Index=Get Multiline InputField Enabled()
    ) -> (ReturnValue:UIKit Modular Button Object Reference)
    State:EUIKitBlockVisualState Enum=Unfocused
   ) -> (then)
   Return Node(execute=[@n14.then, @n13.else], ReturnValue=Get Chat_ContextMenu() -> (Chat_ContextMenu:WBP Text Chat Context Menu Button Object Reference))
  ]}
 Block {nodes [
   Block!() -> (then, InteractionNameID:Name) {is=title K2Node_FunctionEntry}
   > n1 = Create Widget!(
    execute:Exec
    Class:User Widget Class Reference=/FortniteCoreUI/UI/Social/Chat/TextChat/Dialogs/WBP_DialogVM_TextChat_Block.WBP_DialogVM_TextChat_Block_C
    OwningPlayer:Player Controller Object Reference
    UIKitDialogViewModel:UI Kit Dialog Viewmodel Object Reference
   ) -> (then:Exec, ReturnValue:WBP Dialog VM Text Chat Block Object Reference)
   > SetDescriptionText!(
    execute
    self:WBP Dialog VM Text Chat Block Object Reference=@n1
    Username:Text="To Text (String)"(InString=Get Username() -> (Username:String))
   ) -> (then)
   > Bind Event to On Clicked Button Confirm(self:WBP Dialog VM Text Chat Block Object Reference=@n1, Delegate=Create Event()) {is=title K2Node_AddDelegate}
   > TryPushWidgetAsModalPanel!(execute, PushWidget:User Widget Object Reference=@n1) -> (then)
  ]}
 Button Dictate Button Visuals {nodes [
   Button Dictate Button Visuals!() -> (
    then
    Icon:Texture 2D Object Reference
    Material Instance:Material Instance Object Reference
    Icon Size:Vector 2D Structure
    InColorAndOpacity:Linear Color Structure
   ) {is=title K2Node_FunctionEntry}
   n1 = Get Send_Dictation() -> (Send_Dictation:WBP UIKit Button Generic Object Reference)
   > SetIconTexture!(
    execute
    self:WBP UIKit Button Generic Base Object Reference=@n1
    Icon:Texture 2D Object Reference=Get Icon() -> (Icon:Texture 2D Object Reference)
   ) -> (then)
   > SetIconMaterial!(
    execute
    self:WBP UIKit Button Generic Base Object Reference=@n1
    Material Instance:Material Instance Object Reference=Get Material Instance() -> (Material Instance:Material Instance Object Reference)
   ) -> (then)
   > SetIconSize!(
    execute
    self:WBP UIKit Button Generic Base Object Reference=@n1
    Icon:Vector 2D Structure=Get Icon Size() -> (Icon Size:Vector 2D Structure)
   ) -> (then)
   > SetColorAndOpacity(
    self:User Widget Object Reference=Get Block_Icon!(self:WBP UIKit Button Generic Base Object Reference=@n1) -> (Block_Icon:WBP UIKit Block Image Object Reference)
    InColorAndOpacity:Linear Color Structure=Get InColorAndOpacity() -> (InColorAndOpacity:Linear Color Structure)
   )
  ]}
 CF_InactiveMessageVisibility {nodes [
   n1 = CF_InactiveMessageVisibility!() -> (then, CurrentChannelType:EFortTextChatChannelType Enum=All, IsChannelReadOnly:bool=~, IsRestrictionActive:bool=~) {is=title K2Node_FunctionEntry}
   n2 = NOT Boolean(A=Get IsRestrictionActive() -> (IsRestrictionActive:bool))
   n3 = "Equal (Enum)"(A:EFortTextChatChannelType Enum=@n1, B:EFortTextChatChannelType Enum=Persistent)
   n4 = Get IsChannelReadOnly() -> (IsChannelReadOnly:bool)
   > Return Node!(
    execute
    InactiveMessageVisibility:ESlateVisibility Enum=Select(Option 0:ESlateVisibility Enum=Collapsed, Option 1:ESlateVisibility Enum=SelfHitTestInvisible, Index=AND Boolean(A=@n4, B=@n2, C=@n3)) -> (ReturnValue:ESlateVisibility Enum)
   )
  ]}
 CF_InactiveText {nodes [
   n1 = CF_InactiveText!() -> (then, CurrentChannelType:EFortTextChatChannelType Enum=All, IsPersistentDM:bool=~) {is=title K2Node_FunctionEntry}
   > n2 = Branch(Condition="Equal (Enum)"(A=@n1, B=Persistent))
   n3 = Get FortTextChatPlayerVM()
   > Return Node!(
    execute
    InactiveMessageText:Text=CF_InactiveMessageText!(
     self:WBP Text Chat Inactive Message Object Reference=Get WBP_TextChat_InactiveMessage() -> (WBP_TextChat_InactiveMessage:WBP Text Chat Inactive Message Object Reference)
     ConversationType:EFortTextChatPersistentConversationType Enum=GetCurrentPersistentConversationType!(self:Fort Text Chat Main VM Object Reference=@n3) -> (ReturnValue:EFortTextChatPersistentConversationType Enum=EpicParty)
    ) -> (InactiveMessageText:Text)
   )
   Return Node(execute=@n2.else)
  ]}
 CreateReportModal {nodes [
   CreateReportModal() {is=title K2Node_FunctionEntry}
   n1 = GetChannels!(self:Fort Text Chat Main VM Object Reference=Get FortTextChatPlayerVM()) -> (ReturnValue:Array of Fort Text Chat Channel VM Object References)
   > Branch(Condition=Is Not Empty!(TargetArray:Array of Fort Text Chat Channel VM Object References=@n1) -> (ReturnValue:bool))
   n2 = Get WBP_UIKit_TabGroup() -> (WBP_UIKit_TabGroup:WBP UIKit Tab Group Object Reference)
   > Branch(Condition="Not Equal (Name)"(
     A=Get TabAddCmd() -> (TabAddCmd:Name=None)
     B=GetSelectedTabId!(self:Common Tab List Widget Base Object Reference=@n2) -> (ReturnValue:Name=None)
    ) -> (ReturnValue=~))
   > n3 = Create Widget!(execute, Class=/FortniteCoreUI/UI/Social/Chat/TextChat/Dialogs/WBP_Dialog_VM_TextChat_Report.WBP_Dialog_VM_TextChat_Report_C, OwningPlayer) -> (then, ReturnValue:WBP Dialog VM Text Chat Report Object Reference)
   n4 = Get WBP_UIKit_TabGroup()
   > SetReportedChannelVM!(
    execute
    self:WBP Dialog VM Text Chat Report Object Reference=@n3
    ReportedChannelVM:Fort Text Chat Channel VM Object Reference="Get (a copy)"(
     Array:Array of Fort Text Chat Channel VM Object References=@n1
     Dimension 1=GetSelectedTabIndex!(self:UIKit Tab Group Object Reference=@n4) -> (ReturnValue:int=0)
    ) -> (Output:Fort Text Chat Channel VM Object Reference)
    ReportMostRecentMessages:bool=true
   ) -> (then)
   > Bind Event to On Widget Deactivated(self:Common Activatable Widget Object Reference=@n3, Delegate=Create Event()) {is=title K2Node_AddDelegate}
   > TryPushWidgetAsModalPanel(PushWidget=@n3)
   > AddTagToInputMode!(
    execute
    self:Enhanced Input Subsystem Interface Interface=Get EnhancedInputLocalPlayerSubsystem() -> (ReturnValue:Enhanced Input Local Player Subsystem Object Reference) {is=title K2Node_GetSubsystem}
    TagToAdd:"Gameplay Tag Structure (by ref)"=MakeLiteralGameplayTag!(Value:Gameplay Tag Structure=(TagName="UI.TextChat.Disabled")) -> (ReturnValue:Gameplay Tag Structure)
    Options:"Modify Context Options Structure (by ref)"=(bIgnoreAllPressedKeysUntilRelease=True,bForceImmediately=False,bNotifyUserSettings=False)
   ) -> (then)
  ]}
 EventGraph {nodes [
   "On Clicked (Button_ReportChat)"!() -> (OutputDelegate, then, Button:Common Button Base Object Reference) {is=title K2Node_ComponentBoundEvent}
   > CreateReportModal() {is=title K2Node_CallFunction}
   "On Clicked (Send_Dictation)"!() -> (OutputDelegate, then, Button:Common Button Base Object Reference) {is=title K2Node_ComponentBoundEvent}
   > n1 = Switch on EFortSpeechToTextState!(
    execute
    Selection:EFortSpeechToTextState Enum=Get CurrentSpeechToTextState!(self:Fort Text Chat Main VM Object Reference=Get FortTextChatPlayerVM()) -> (CurrentSpeechToTextState:EFortSpeechToTextState Enum=Idle)
   ) -> (Idle:Exec, Initializing:Exec, Recording:Exec, Transcribing:Exec, Deinitializing:Exec, Unavailable:Exec) {is=title K2Node_SwitchEnum}
   > StopSpeakingMessages(self:Fort Text Chat Main VM Object Reference=Get FortTextChatPlayerVM())
   > StartOrStopSpeechToTextRecording!(execute, self:Fort Text Chat Main VM Object Reference=Get FortTextChatPlayerVM(), bWantsToStart:bool=true) -> (then)
   StartOrStopSpeechToTextRecording(
    execute=[@n63.then, @n49.else, @n1.Deinitializing, @n1.Initializing, @n1.Recording, @n1.Transcribing]
    self=Get FortTextChatPlayerVM()
    bWantsToStart=false
   )
   > Clear and Invalidate Timer by Handle(Handle=Get Timer Handle() -> (Timer Handle:Timer Handle Structure))
   > Set Record Time(Record Time:double=Get SpeechToTextRecordTime!(self:Fort Text Chat Main VM Object Reference=Get FortTextChatPlayerVM()) -> (SpeechToTextRecordTime:float)) -> (Output_Get:double)
   > SetText!(execute, self:WBP UIKit Button Generic Base Object Reference=Get Send_Dictation(), Text:Text) -> (then)
   n2 = "On Clicked (Button_BackChat)"!() -> (OutputDelegate, then, Button:Common Button Base Object Reference) {is=title K2Node_ComponentBoundEvent}
   > Call On Button Close Clicked!(execute:Exec, self:WBP Text Chat Object Reference, Button:Common Button Base Object Reference=@n2.Button) -> (then:Exec) {is=title K2Node_CallDelegate}
   n3 = "On Clicked (Button_StopTTS)"!() -> (OutputDelegate, then, Button:Common Button Base Object Reference) {is=title K2Node_ComponentBoundEvent}
   StopSpeakingMessages(execute=[@n3.then, @n9.then], self=Get FortTextChatPlayerVM())
   n4 = "On Hovered (WBP_TextChat_Input)"!() -> (OutputDelegate, then, Button:Common Button Base Object Reference) {is=title K2Node_ComponentBoundEvent}
   UpdateInputFieldPadding(execute=[@n4.then, @n10.then])
   "On Unhovered (WBP_TextChat_Input)"!() -> (OutputDelegate, then, Button:Common Button Base Object Reference) {is=title K2Node_ComponentBoundEvent}
   n5 = Get WBP_TextChat_Input()
   > SetPadding!(
    execute
    self=SlotAsGridSlot(Widget=Get Block_EditableText(self:WBP UIKit Input Field Base Object Reference=@n5) -> (Block_EditableText:WBP UIKit Block Editable Text Object Reference))
    InPadding_Left:float=0
    InPadding_Top:float=0
    InPadding_Right:float=0
    InPadding_Bottom:float=0
   ) -> (then)
   "On Context Menu Opened (Chat_ContextMenu)"() {is=title K2Node_ComponentBoundEvent}
   > SetChatContextMenuActionContext()
   > SetBackButtonText!(execute, self, Close:bool=false) -> (then)
   n6 = "On UIKit Tab Clicked (WBP_UIKit_TabGroup)"!() -> (OutputDelegate, then, TabID:Name=None, TabVM:MVVM Base Viewmodel Object Reference) {is=title K2Node_ComponentBoundEvent}
   > n7 = Cast To FortTextChatChannelVM!(execute, Object=@n6.TabVM) -> (then, CastFailed, AsFort Text Chat Channel VM:Fort Text Chat Channel VM Object Reference) {is=title K2Node_DynamicCast}
   > Branch(Condition="Equal (Enum)"(A=Get ChannelType!(self:Fort Text Chat Channel VM Object Reference=@n7) -> (ChannelType:EFortTextChatChannelType Enum=All), B=Whisper))
   > SetDirectMessageContext!(execute, self:Fort Text Chat Main VM Object Reference=Get FortTextChatPlayerVM(), Context:EFortDirectMessageContext Enum=ChannelTab) -> (then)
   "On Text Changed (WBP_TextChat_Input)"!() -> (OutputDelegate, then, Text:Text) {is=title K2Node_ComponentBoundEvent}
   > UpdateSendButtonEnabled()
   > PlayTextChatSound!(execute, self, Sound:Sound Base Object Reference=Get TextInputSound() -> (TextInputSound:Sound Base Object Reference)) -> (then)
   "On Clicked (Button_CompactChat_RestrictionAction)"!() -> (OutputDelegate, then, Button:Common Button Base Object Reference) {is=title K2Node_ComponentBoundEvent}
   > n8 = HandleRestrictionActionButtonClicked(self:WBP Text Chat Restrictions Object Reference=Get WBP_TextChat_Restrictions() -> (WBP_TextChat_Restrictions:WBP Text Chat Restrictions Object Reference))
   Call On Button Close Clicked(execute=[@n8.then, @n16.then])
   n9 = "On Clicked (Button_StopTTS_Decoy)"!() -> (OutputDelegate, then, Button:Common Button Base Object Reference) {is=title K2Node_ComponentBoundEvent}
   n10 = "On Hovered (WBP_TextChat_MultilineInputField)"!() -> (OutputDelegate, then, Button:Common Button Base Object Reference) {is=title K2Node_ComponentBoundEvent}
   "On Lock Clicked (Button_ReportChat)"!() -> (OutputDelegate, then, Button:Common Button Base Object Reference) {is=title K2Node_ComponentBoundEvent}
   > Branch(Condition=ShouldShowOnboarding!(self:Fort Sidebar Onboard Tooltip Widget Object Reference=Get WBP_TextChat_NothingToReport_CoachMark() -> (WBP_TextChat_NothingToReport_CoachMark:WBP Text Chat Nothing To Report Coach Mark Object Reference)) -> (ReturnValue:bool))
   > n11 = GetChatViewportQuadrantForCoachmarkLayout!(execute, self) -> (then, IsContextButtonOnLeftHalfOfScreen:bool, IsContextButtonOnTopHalfOfScreen:bool)
   > SetDynamicCoachmarkLayout!(
    execute
    self
    IsChatContextButtonOnLeftHalfOfScreen:bool=@n11
    IsChatContextButtonOnTopHalfOfScreen:bool=@n11.IsContextButtonOnTopHalfOfScreen
   ) -> (then)
   n12 = Get WBP_TextChat_NothingToReport_CoachMark()
   > ActivateWidget(self=@n12)
   > SetFocus(self=Get DismissButton!(self:WBP UIKit Coach Mark Base Object Reference=Get WBP_UIKit_CoachMark!(self:WBP Text Chat Nothing To Report Coach Mark Object Reference=@n12) -> (WBP_UIKit_CoachMark:WBP UIKit Coach Mark Object Reference)) -> (DismissButton:WBP UIKit Button Regular Object Reference))
   "On Unhovered (WBP_TextChat_MultilineInputField)"!() -> (OutputDelegate, then, Button:Common Button Base Object Reference) {is=title K2Node_ComponentBoundEvent}
   n13 = Get WBP_TextChat_MultilineInputField()
   > SetPadding(
    self=SlotAsGridSlot(Widget=Get Block_EditableText(self:WBP Text Chat Multiline Input Field Object Reference=@n13) -> (Block_EditableText:WBP Text Chat Block Multiline Editable Text Object Reference))
    InPadding_Left=0
    InPadding_Top=0
    InPadding_Right=0
    InPadding_Bottom=0
   )
   n14 = "On Text Commited (WBP_TextChat_MultilineInputField)"!() -> (OutputDelegate, then, Text:Text, Commit method:ETextCommit Enum=Default) {is=title K2Node_ComponentBoundEvent}
   > SendTextMessage!(execute, self, Text:"Text (by ref)"=@n14.Text, CommitMethod:ETextCommit Enum=@n14.Commit method) -> (then)
   "On Text Changed (WBP_TextChat_MultilineInputField)"!() -> (OutputDelegate, then, Text:Text) {is=title K2Node_ComponentBoundEvent}
   > UpdateSendButtonEnabled()
   > PlayTextChatSound(Sound=Get TextInputSound())
   "On Tab Selected (WBP_UIKit_TabGroup)"!() -> (OutputDelegate, then, TabId:Name=None) {is=title K2Node_ComponentBoundEvent}
   n15 = Get GP_BottomContent() -> (GP_BottomContent:Grid Panel Object Reference)
   > Branch(Condition="Equal (Enum)"(
     A:ESlateVisibility Enum=GetVisibility!(self:Widget Object Reference=@n15) -> (ReturnValue:ESlateVisibility Enum=Visible)
     B:ESlateVisibility Enum=Hidden
    ))
   > Delay(Duration=0.15)
   > SetVisibility(self=Get GP_BottomContent(), InVisibility=SelfHitTestInvisible)
   > RequestRefreshFocus(self:Common Activatable Widget Object Reference)
   n16 = "On Button Clicked (WBP_TextChat_Restrictions)"() {is=title K2Node_ComponentBoundEvent}
   n17 = "On UIKit Tab Selected (WBP_UIKit_TabGroup)"!() -> (OutputDelegate, then, TabID:Name=None, TabVM:MVVM Base Viewmodel Object Reference) {is=title K2Node_ComponentBoundEvent}
   n18 = Get WBP_TextChat_ChatSettings_CoachMark() -> (WBP_TextChat_ChatSettings_CoachMark:WBP Text Chat Chat Settings Coach Mark Object Reference)
   n19 = Get WBP_TextChat_STT_CoachMark() -> (WBP_TextChat_STT_CoachMark:WBP Text Chat STT Coach Mark Object Reference)
   n20 = Get WBP_TextChat_NothingToReport_CoachMark()
   > n21 = Branch(Condition=OR Boolean!(
     A=IsActivated!(self:Common Activatable Widget Object Reference=@n20) -> (ReturnValue:bool)
     B=IsActivated(self=@n18)
     C:bool=IsActivated(self=@n19)
    ) -> (ReturnValue))
   n22 = Branch(execute=@n21.else, Condition="Not Equal (Name)"(A=Get TabAddCmd() -> (TabAddCmd=None), B=@n17.TabID) -> (ReturnValue=~))
   > ChangeCurrentChannel!(execute, self:Fort Text Chat Main VM Object Reference=Get FortTextChatPlayerVM(), InChannelId:"Name (by ref)"=@n17.TabID) -> (then)
   > Cancel Speech To Text Recording()
   n23 = Get FortTextChatPlayerVM()
   n24 = NOT Boolean(A=IsCurrentChannelReadOnly(self=@n23))
   > UpdateTabsData!(execute, self, IsTabActive:bool=@n24) -> (then)
   > SetBlankMessageText!(
    execute
    self:WBP Text Chat Blank Message Object Reference=Get WBP_TextChat_BlankMessage() -> (WBP_TextChat_BlankMessage:WBP Text Chat Blank Message Object Reference)
    ChannelType:EFortTextChatChannelType Enum=GetCurrentChannelType!(self:Fort Text Chat Main VM Object Reference=@n23) -> (ReturnValue:EFortTextChatChannelType Enum=All)
    ChannelEnabled:bool=@n24
   ) -> (then)
   n25 = GetCurrentInputType(self=Get CommonInputSubsystem())
   > n26 = Branch(Condition=AND Boolean!(A=@n24, B="Equal (Enum)"(A:ECommonInputType Enum=@n25, B:ECommonInputType Enum=Gamepad)) -> (ReturnValue))
   Close(
    execute=[@n29.then, @n26.then]
    self:Menu Anchor Object Reference=Get Anchor!(self:UIKit Context Menu Anchor Object Reference=Get Whisper_ContextMenu() -> (Whisper_ContextMenu:WBP UIKit Context Menu Anchor Button Object Reference)) -> (Anchor:Menu Anchor Object Reference)
   )
   > ScrollToBottom(self:List View Base Object Reference=Get ListView(self=Get MessageList()))
   n27 = Get CommonInputSubsystem()
   > Branch(Condition="Not Equal (Enum)"(A=GetCurrentInputType(self=@n27), B=Touch))
   n28 = Get ListView(self=Get MessageList())
   > SetSelectedIndex!(
    execute
    self:List View Object Reference=@n28
    Index:int=int - int(A=GetNumItems!(self:List View Object Reference=@n28) -> (ReturnValue:int=0))
   ) -> (then)
   n29 = RequestRefreshFocus(execute=@n26.else)
   Open!(execute=@n22.else, self:Menu Anchor Object Reference=Get Anchor(self=Get Whisper_ContextMenu()), bFocusMenu:bool=true) -> (then)
   n30 = "On Text Commited (WBP_TextChat_Input)"!() -> (OutputDelegate, then, Text:Text, Commit method:ETextCommit Enum=Default) {is=title K2Node_ComponentBoundEvent}
   > SendTextMessage(Text=@n30.Text, CommitMethod=@n30.Commit method)
   "On Context Menu Opened (Whisper_ContextMenu)"() {is=title K2Node_ComponentBoundEvent}
   > SetWhisperContextMenuActionContext()
   > SetBackButtonText(Close=false)
   n31 = "On List View Initialized (MessageList)"!() -> (OutputDelegate, then, Item:Object Reference, Widget:User Widget Object Reference) {is=title K2Node_ComponentBoundEvent}
   > n32 = Cast To WBP_TextChat_Entry_UserMessage!(execute, Object=@n31.Widget) -> (then, CastFailed, AsWBP Text Chat Entry User Message:WBP Text Chat Entry User Message Object Reference) {is=title K2Node_DynamicCast}
   > Bind Event to On Message Clicked(self:WBP Text Chat Entry User Message Object Reference=@n32, Delegate=Create Event()) {is=title K2Node_AddDelegate}
   n33 = Cast To WBP_TextChat_MessageSystem!(execute=@n32.CastFailed, Object=@n31.Widget) -> (then, CastFailed, AsWBP Text Chat Message System:WBP Text Chat Message System Object Reference) {is=title K2Node_DynamicCast}
   > Bind Event to On Message Clicked(self:WBP Text Chat Message System Object Reference=@n33, Delegate=Create Event())
   "On Context Menu Closed (Whisper_ContextMenu)"() {is=title K2Node_ComponentBoundEvent}
   > SelectTabByID!(
    execute
    self:Common Tab List Widget Base Object Reference=Get WBP_UIKit_TabGroup()
    TabNameID:Name=Get CurrentChannelId!(self:Fort Text Chat Main VM Object Reference=Get FortTextChatPlayerVM()) -> (CurrentChannelId:Name=None)
    bSuppressClickFeedback:bool=true
   ) -> (then, ReturnValue:bool)
   "On Clicked (Send_ButtonIcon)"!() -> (OutputDelegate, then, Button:Common Button Base Object Reference) {is=title K2Node_ComponentBoundEvent}
   n34 = Get EditableText(self:WBP UIKit Block Editable Text Object Reference=Get Block_EditableText(self:WBP UIKit Input Field Base Object Reference=Get WBP_TextChat_Input()) -> (Block_EditableText:WBP UIKit Block Editable Text Object Reference)) -> (EditableText:UIKit Editable Text Object Reference)
   n35 = Get EditableText(self:WBP Text Chat Block Multiline Editable Text Object Reference=Get Block_EditableText(self:WBP Text Chat Multiline Input Field Object Reference=Get WBP_TextChat_MultilineInputField()) -> (Block_EditableText:WBP Text Chat Block Multiline Editable Text Object Reference)) -> (EditableText:"Text Box (Multi- Line) Object Reference")
   > SendTextMessage(
    Text=SelectText(
     A=Get Text(self:"Text Box (Multi- Line) Object Reference"=@n35)
     B="GetText (Editable Text)"!(self:Editable Text Object Reference=@n34) -> (ReturnValue:Text)
     bPickA=Get Multiline InputField Enabled()
    )
    CommitMethod=OnEnter
   )
   "On Context Menu Closed (Message_ContextMenu)"() {is=title K2Node_ComponentBoundEvent}
   > n36 = Branch(Condition=Get bSystemMessageClicked() -> (bSystemMessageClicked:bool))
   n37 = Get WBP Text Chat Message System() -> (WBP Text Chat Message System:WBP Text Chat Message System Object Reference)
   > n38 = Set bIsContextIconVisible!(execute, bIsContextIconVisible:bool=false, self:WBP Text Chat Message System Object Reference=@n37) -> (then, Output_Get:bool)
   > SetContextMenuIconVisibility!(execute, self:WBP Text Chat Message System Object Reference=@n37, IsContextMenuVisible:bool=@n38) -> (then)
   > Set bSystemMessageClicked(bSystemMessageClicked:bool=false) -> (Output_Get:bool)
   n39 = Get Message_ContextMenu()
   n40 = Get WBP Text Chat Entry User Message() -> (WBP Text Chat Entry User Message:WBP Text Chat Entry User Message Object Reference)
   SetBackgroundVisibility!(
    execute=@n36.else
    self:WBP Text Chat Entry User Message Object Reference=@n40
    ShouldBeVisible:bool=IsOpen!(self:Menu Anchor Object Reference=Get Anchor(self=@n39)) -> (ReturnValue:bool)
   ) -> (then)
   > n41 = Set bIsContextIconVisible(bIsContextIconVisible=false, self:WBP Text Chat Entry User Message Object Reference=@n40)
   > SetContextMenuIconVisibility(self:WBP Text Chat Entry User Message Object Reference=@n40, IsContextMenuVisible=@n41)
   n42 = OnGenWhisperContextMenuItem!() -> (OutputDelegate, then, UserWidget:User Widget Object Reference) {is=title K2Node_CustomEvent}
   > SetWhisperItemVisual!(execute, self, Object:Object Reference=@n42.UserWidget) -> (then)
   n43 = OnInitWhisperContextMenu!() -> (OutputDelegate, then, Object:Object Reference) {is=title K2Node_CustomEvent}
   > AddingAdditionalContentForWhisperContextMenu!(execute, self, Object:Object Reference=@n43.Object) -> (then) {is=title K2Node_CallFunction}
   OnCoachmarkDismissReEvaluateReportStatus() {is=title K2Node_CustomEvent}
   n44 = Get Button_ReportChat() -> (Button_ReportChat:WBP UIKit Button Quiet Object Reference)
   > SetIsLocked!(execute, self:Common Button Base Object Reference=@n44, bInIsLocked:bool=false) -> (then)
   > SetIsEnabled(
    self=@n44
    bInIsEnabled=Get bCanReportTextChat!(self:Fort Text Chat Main VM Object Reference=Get FortTextChatPlayerVM()) -> (bCanReportTextChat:bool)
   )
   n45 = HandleOnVirtualKeyboardShown!() -> (OutputDelegate, then, RectTop:int, RectLeft:int, RectBottom:int, RectRight:int, VirtualKeyboardInputHeight:int) {is=title K2Node_CustomEvent}
   n46 = GetConfigDataBoolWithDefault!(EntryName:String=EnableCompactChatResizeAboveKeyboard, bDefaultValue:bool=false) -> (ReturnValue:bool)
   > Branch(Condition=@n46)
   > n47 = MoveChatAboveVirtualKeyboard!(execute, self, RectTop:int=@n45.RectTop, RectBottom:int=@n45.RectBottom, VirtualKeyboardInputHeight:int=@n45.VirtualKeyboardInputHeight) -> (then)
   ScrollToBottom(execute=[@n47.then, @n60.then], self=Get ListView(self=Get MessageList()))
   Dictation Timer() {is=title K2Node_CustomEvent}
   n48 = Get Record Time() -> (Record Time:double)
   > n49 = Branch(Condition="Not Equal (Float)"(A=@n48))
   > n50 = Set Record Time(Record Time=float - float(A=@n48, B=1))
   n51 = Truncate(A=@n50)
   n52 = SelectText(B="LOCTEXT(\"0\")", bPickA="float >= float"(A=@n50, B=10))
   > SetText(self=Get Send_Dictation(), Text=Format Text(Format=Format Text!(Format="LOCTEXT(\"0:{zero}{n}\")", zero:Text=@n52, n:int=@n51) -> (Result)))
   OnMessageListFocused() {is=title K2Node_CustomEvent}
   > SetBackButtonText(Close=false)
   n53 = OnContextButtonFocused() {is=title K2Node_CustomEvent}
   SetBackButtonText(execute=[@n53.then, @n61.then], Close=true)
   OnMessageSent() {is=title K2Node_CustomEvent}
   > PlayTextChatSound(Sound=Get MessageSentSound() -> (MessageSentSound:Sound Base Object Reference))
   OnMessageReceived() {is=title K2Node_CustomEvent}
   > PlayTextChatSound(Sound=Get MessageReceivedSound() -> (MessageReceivedSound:Sound Base Object Reference))
   n54 = OnMessageContextMenuInit!() -> (OutputDelegate, then, Object:Object Reference) {is=title K2Node_CustomEvent}
   > n55 = Cast To WBP_UIKit_ContextMenu(Object=@n54.Object)
   > Bind Event to On Entry Generated(self=Get Content(self=@n55), Delegate=Create Event())
   n56 = OnContextMenuMessageGenerated!() -> (OutputDelegate, then, Widget:User Widget Object Reference) {is=title K2Node_CustomEvent}
   > n57 = Cast To WBP_UIKit_ContextMenuListEntry!(execute, Object=@n56.Widget) -> (then, CastFailed, AsWBP UIKit Context Menu List Entry:WBP UIKit Context Menu List Entry Object Reference) {is=title K2Node_DynamicCast}
   > Branch(Condition=IsValid(Object=Get Icon!(self:WBP UIKit Button Generic Base Object Reference=@n57) -> (Icon)))
   > SetIconMaterial(self=@n57, Material Instance=/FortniteCoreUI/UI/Social/Chat/TextChat/MaterialInstances/MI_UI_TextChat_TTS)
   > SetIconPadding!(
    execute
    self:WBP UIKit Button Generic Base Object Reference=@n57
    Padding_Left:float=0
    Padding_Top:float=0
    Padding_Right:float=-8
    Padding_Bottom:float=0
   ) -> (then)
   n58 = OnClicked_ConfirmButton!() -> (OutputDelegate, then, Button:Common Button Base Object Reference, DialogRef:Object Reference) {is=title K2Node_CustomEvent}
   > PerformMessageContextMenuAction!(
    execute
    self:Fort Text Chat Main VM Object Reference=Get FortTextChatPlayerVM()
    MessageId:String="To String (Name)"(InName=Get MessageId() -> (MessageId:Name=None))
    MenuAction:EFortTextChatMessageContextMenuAction Enum=Block
   ) -> (then)
   > n59 = Cast To WBP_DialogVM_TextChat_Block!(execute, Object=@n58.DialogRef) -> (then, CastFailed, AsWBP Dialog VM Text Chat Block:WBP Dialog VM Text Chat Block Object Reference) {is=title K2Node_DynamicCast}
   > DeactivateWidget(self=@n59)
   > RequestRefreshFocus()
   HandleOnVirtualKeyboardHidden() {is=title K2Node_CustomEvent}
   > Branch(Condition=@n46)
   > n60 = ResetChatPosition()
   n61 = OnInputFieldFocused() {is=title K2Node_CustomEvent}
   Cancel Speech To Text Recording() {is=title K2Node_CustomEvent}
   n62 = Get CurrentSpeechToTextState(self=Get FortTextChatPlayerVM()) -> (CurrentSpeechToTextState=Idle)
   > n63 = Branch(Condition=AND Boolean(A="Not Equal (Enum)"(A:EFortSpeechToTextState Enum=@n62, B:EFortSpeechToTextState Enum=Idle), B=Get is SpeechToText Enabled()))
   On STT Tooltip Open Settings() {is=title K2Node_CustomEvent}
   > Delay(Duration=0.2)
   > RequestOpenSettings!(
    execute
    self:Fort Settings Context Object Reference=Get FortSettingsContext!() -> (ReturnValue:Fort Settings Context Object Reference) {is=title K2Node_GetBlueprintContextEx}
    ContextWidget:Common User Widget Object Reference=Self-Reference()
    SettingsTabId:Name=Select(Option 0:Name=CommunicationCollection, Option 1:Name=Communication, Index=IsSaveTheWorldUI!() -> (IsSaveTheWorld:bool)) -> (ReturnValue:Name)
    SettingToSelect:String=ShowTextToSpeech
   ) -> (then)
   > Call On Button Close Clicked()
   n64 = SystemMessageClickedEvent!() -> (OutputDelegate, then, Message ID:Name, Button:Common Button Base Object Reference) {is=title K2Node_CustomEvent}
   > Set MessageId(MessageId:Name=@n64.Message ID) -> (Output_Get:Name=None)
   > SetMessageContextMenuActionContext()
   > Branch(Condition=HasAvailableInteractions!(self:UIKit Context Menu Action Context Object Reference=Get Message_ContextMenuActionContext() -> (Message_ContextMenuActionContext:UIKit Context Menu Action Context Object Reference)) -> (ReturnValue:bool))
   > Open(self=Get Anchor(self=Get Message_ContextMenu()), bFocusMenu=true)
   > Set bSystemMessageClicked(bSystemMessageClicked=true)
   > n65 = Cast To WBP_TextChat_MessageSystem(Object=@n64.Button)
   > Set WBP Text Chat Message System(WBP Text Chat Message System:WBP Text Chat Message System Object Reference=@n65) -> (Output_Get:WBP Text Chat Message System Object Reference)
   OnCoachmarkGoToChatSettings() {is=title K2Node_CustomEvent}
   > OpenChatSettings!(execute, self, InteractionNameID:Name=None) -> (then)
   n66 = OnInitChatContextMenu!() -> (OutputDelegate, then, Object:Object Reference) {is=title K2Node_CustomEvent}
   > AddingAdditionalContentForChatContextMenu!(execute, self, Object:Object Reference=@n66.Object) -> (then) {is=title K2Node_CallFunction}
   n67 = MessageClickedEvent!() -> (OutputDelegate, then, MessageID:Name, Username:String, Button:Common Button Base Object Reference) {is=title K2Node_CustomEvent}
   > Set MessageId(MessageId=@n67.MessageID) -> (Output_Get=None)
   > Set Username(Username:String=@n67.Username) -> (Output_Get:String)
   > SetMessageContextMenuActionContext()
   > Branch(Condition=HasAvailableInteractions(self=Get Message_ContextMenuActionContext()))
   > Open(self=Get Anchor(self=Get Message_ContextMenu()), bFocusMenu=true)
   > n68 = Cast To WBP_TextChat_Entry_UserMessage(Object=@n67.Button)
   > n69 = Set WBP Text Chat Entry User Message(WBP Text Chat Entry User Message:WBP Text Chat Entry User Message Object Reference=@n68) -> (Output_Get:WBP Text Chat Entry User Message Object Reference)
   n70 = Get Message_ContextMenu()
   > SetBackgroundVisibility(self=@n69, ShouldBeVisible=IsOpen(self=Get Anchor(self=@n70)))
   n71 = OnListViewIEntryGenerated!() -> (OutputDelegate, then, Widget:User Widget Object Reference) {is=title K2Node_CustomEvent}
   > n72 = Cast To WBP_TextChat_Entry_MessageBase!(execute, Object=@n71.Widget) -> (then, CastFailed, AsWBP Text Chat Entry Message Base:WBP Text Chat Entry Message Base Object Reference) {is=title K2Node_DynamicCast}
   n73 = GetNumItems(self=Get ListView(self=Get MessageList())) -> (ReturnValue=0)
   n74 = Get IndexOfItem!(self:WBP Text Chat Entry Message Base Object Reference=@n72) -> (IndexOfItem:int=0)
   > Branch(Condition="Equal (Integer)"(A=@n74, B=int - int(A=@n73)))
   n75 = Get CommonInputSubsystem()
   > n76 = Branch(Condition="Not Equal (Enum)"(A:ECommonInputType Enum=GetCurrentInputType(self=@n75), B:ECommonInputType Enum=Gamepad))
   n77 = Get ListView(self=Get MessageList())
   ScrollToBottom(execute=[@n78.then, @n76.then], self=@n77)
   n78 = SetSelectedIndex(execute=@n76.else, self=@n77, Index=@n74)
   n79 = OnInputMethodChanged!() -> (OutputDelegate, then, InputType:ECommonInputType Enum) {is=title K2Node_CustomEvent}
   > Set CurrentInputType(CurrentInputType:ECommonInputType Enum=@n79.InputType) -> (Output_Get:ECommonInputType Enum)
   > UpdateSendButtonVisibility()
   > UpdateSendButtonEnabled()
   n80 = Get Multiline InputField Enabled()
   n81 = Get WBP_TextChat_MultilineInputField()
   n82 = Get WBP_TextChat_Input()
   > Branch(Condition=IsHovered(self=Select(Option 0:Widget Object Reference=@n82, Option 1:Widget Object Reference=@n81, Index=@n80) -> (ReturnValue:Widget Object Reference)))
   > UpdateInputFieldPadding()
   OnBlockModalDeactivated() {is=title K2Node_CustomEvent}
   > RequestRefreshFocus()
   > RemoveTagFromInputMode!(
    execute
    self:Enhanced Input Subsystem Interface Interface=Get EnhancedInputLocalPlayerSubsystem()
    TagToRemove:"Gameplay Tag Structure (by ref)"=MakeLiteralGameplayTag(Value=(TagName="UI.TextChat.Disabled"))
    Options:"Modify Context Options Structure (by ref)"=(bIgnoreAllPressedKeysUntilRelease=True,bForceImmediately=False,bNotifyUserSettings=False)
   ) -> (then)
   OnOpenTextChat() {is=title K2Node_CustomEvent}
   > Delay(Duration=0.5)
   > ActivateWidget()
   Event OnInitialized()
   > InitializeTextChatVM(self:Fort Text Chat Main VM Object Reference=Get FortTextChatPlayerVM())
   > n83 = Construct Object from Class!(
    execute
    Class=/FortniteCoreUI/UI/Social/Chat/TextChat/Data/BP_VM_TextChat_Welcome_Entry.BP_VM_TextChat_Welcome_Entry_C
    self
    IsCompactChat:bool=true
   ) -> (then, ReturnValue:BP VM Text Chat Welcome Entry Object Reference)
   > Set WelcomeMessageEntry(WelcomeMessageEntry:BP VM Text Chat Welcome Entry Object Reference=@n83) -> (Output_Get:BP VM Text Chat Welcome Entry Object Reference)
   > InitializeChatContextMenu()
   > InitializeMessageContextMenu()
   > InitializeWhisperContextMenu()
   > Assign Init Finished!(
    execute:Exec
    self:WBP UIKit Context Menu Anchor Button Base Object Reference=Get Whisper_ContextMenu()
    Delegate:"Delegate (by ref)"=Create Event()
   ) -> (then:Exec) {is=title K2Node_AssignDelegate}
   > Assign Init Finished(self=Get Message_ContextMenu(), Delegate=Create Event())
   > Bind Event to Init Finished(self:WBP UIKit Context Menu Anchor Button Base Object Reference=Get Chat_ContextMenu(), Delegate=Create Event()) {is=title K2Node_AddDelegate}
   > Bind Event to On Entry Generated(self=Get ListView(self=Get MessageList()), Delegate=Create Event())
   > SetInGameAdjustments()
   n84 = Get GP_BottomContent()
   > Set InitialGPContentGridRow(InitialGPContentGridRow:int=Get Row!(self:Grid Slot Object Reference=SlotAsGridSlot(Widget=@n84)) -> (Row:int=0)) -> (Output_Get:int=0)
   n85 = Get WBP_UIKit_TabGroup()
   > Set BackplateClass!(
    execute
    BackplateClass:WBP UIKit Backplate Base Class Reference=/FortUILibrary/UIKit/Backplates/WBP_UIKit_Backplate_Fallback.WBP_UIKit_Backplate_Fallback_C
    self:WBP UIKit Tab Group Base Object Reference=@n85
   ) -> (then, Output_Get:WBP UIKit Backplate Base Class Reference)
   > Set BackplateBrightness!(execute, BackplateBrightness:E_UI_BackplateBrightness Enum=NewEnumerator3, self:WBP UIKit Tab Group Base Object Reference=@n85) -> (then, Output_Get:E_UI_BackplateBrightness Enum=NewEnumerator0)
   > Set PromptsBackplateClass!(
    execute
    PromptsBackplateClass:WBP UIKit Backplate Base Class Reference=&WBP_UIKit_Backplate_Fallback_C
    self:WBP UIKit Tab Group Base Object Reference=@n85
   ) -> (then, Output_Get:WBP UIKit Backplate Base Class Reference)
   > Set PromptBackplateBrightness!(execute, PromptBackplateBrightness:E_UI_BackplateBrightness Enum=NewEnumerator3, self:WBP UIKit Tab Group Base Object Reference=@n85) -> (then, Output_Get:E_UI_BackplateBrightness Enum=NewEnumerator0)
   > Set ArrowButtonsClass!(
    execute
    ArrowButtonsClass:WBP UIKit Button Generic Base Class Reference=&/WBP_UIKit_Button_TabNavArrow_TextChat.WBP_UIKit_Button_TabNavArrow_TextChat_C
    self:WBP UIKit Tab Group Base Object Reference=@n85
   ) -> (then, Output_Get:WBP UIKit Button Generic Base Class Reference)
   > Set UsePromptBackplate!(execute, UsePromptBackplate:bool=false, self:WBP UIKit Tab Group Base Object Reference=@n85) -> (then, Output_Get:bool)
   > SetMobileLayout()
   Event PreConstruct()
   > SetSendButtonHeight()
   > SetDynamicCoachmarkLayout(IsChatContextButtonOnLeftHalfOfScreen=false, IsChatContextButtonOnTopHalfOfScreen=false)
   Event On Activated()
   n86 = Get FortTextChatPlayerVM()
   > OnOpen(self:Fort Text Chat Main VM Object Reference=@n86)
   > PlayTextChatSound(Sound=Get TextChatOpenSound() -> (TextChatOpenSound:Sound Base Object Reference))
   > SetVisibility(
    self=Get WBP_TextChat_Input()
    InVisibility=Select(Option 0:ESlateVisibility Enum=SelfHitTestInvisible, Option 1:ESlateVisibility Enum=Collapsed, Index=Get Multiline InputField Enabled()) -> (ReturnValue:ESlateVisibility Enum)
   )
   > SetVisibility(
    self=Get WBP_TextChat_MultilineInputField()
    InVisibility=Select(Option 0=Collapsed, Option 1=SelfHitTestInvisible, Index=Get Multiline InputField Enabled())
   )
   > SetVisibility(InVisibility=SelfHitTestInvisible)
   > SetEnableClippingMask!(execute, self:Fort Maskable Widget Object Reference, bEnableClipMask:bool=true) -> (then)
   > SelectTabByID(
    self=Get WBP_UIKit_TabGroup()
    TabNameID=Get CurrentChannelId(self=Get FortTextChatPlayerVM()) -> (CurrentChannelId=None)
    bSuppressClickFeedback=true
   )
   > Call On Active State Changed!(execute:Exec, self:WBP Text Chat Object Reference, IsActivated:bool=true) -> (then:Exec) {is=title K2Node_CallDelegate}
   > Broadcast Event!(
    execute
    self:Gameplay Event Router Component Object Reference=GetGlobalEventRouter()
    EventData:BPS Text Chat Active State Change Structure=Make BPS Text Chat Active State Change!(bActive_4_#1:bool=true) -> (BPS_TextChatActiveStateChange:BPS Text Chat Active State Change Structure) {is=title K2Node_MakeStruct}
    EventContext:Object Reference=GetOwningPlayer()
   ) -> (then, ReturnValue:bool)
   > Get TTS and STT Settings() {is=title K2Node_CallFunction}
   > n87 = Branch(Condition=ShouldShowOnboarding(self=Get WBP_TextChat_STT_CoachMark()))
   > n88 = ActivateWidget(self=Get WBP_TextChat_STT_CoachMark())
   DelayUntilNextTick!(execute=[@n88.then, @n94.then, @n93.else]) -> (then)
   > DelayUntilNextTick()
   > n89 = GetChatViewportQuadrantForCoachmarkLayout()
   > SetDynamicCoachmarkLayout(IsChatContextButtonOnLeftHalfOfScreen=@n89, IsChatContextButtonOnTopHalfOfScreen=@n89.IsContextButtonOnTopHalfOfScreen)
   n90 = Branch(execute=@n87.else, Condition=IsVisible(self=Get WBP_TextChat_STT_CoachMark()))
   > n91 = SetVisibility(self=Get WBP_TextChat_STT_CoachMark(), InVisibility=Collapsed)
   n92 = Get WBP_TextChat_ChatSettings_CoachMark()
   n93 = Branch(execute=[@n91.then, @n90.else], Condition=ShouldShowOnboarding(self=@n92))
   > n94 = ActivateWidget(self=@n92)
   Event On Deactivated()
   > OnClose(self:Fort Text Chat Main VM Object Reference=@n86)
   > PlayTextChatSound(Sound=Get TextChatClosedSound() -> (TextChatClosedSound:Sound Base Object Reference))
   > SetEnableClippingMask(bEnableClipMask=false)
   > Call On Active State Changed()
   > Broadcast Event(
    self=GetGlobalEventRouter()
    EventData=Make BPS Text Chat Active State Change(bActive_4_#1=False)
    EventContext=GetOwningPlayer()
   )
   Event Construct()
   n95 = Create Event()
   n96 = Get CommonInputSubsystem()
   > Bind Event to On Input Method Changed(self=@n96, Delegate=@n95)
   > OnInputMethodChanged!(execute, self, InputType:ECommonInputType Enum=GetCurrentInputType(self=@n96)) -> (then) {is=title K2Node_CallFunction}
   n97 = Create Event()
   n98 = Get FortTextChatPlayerVM()
   > Bind Event to On Open Text Chat(self:Fort Text Chat Player VM Object Reference=@n98, Delegate=@n97) {is=title K2Node_AddDelegate}
   n99 = Create Event()
   > Bind Event to On Message Sent(self:Fort Text Chat Main VM Object Reference=@n98, Delegate=@n99) {is=title K2Node_AddDelegate}
   n100 = Create Event()
   > Bind Event to On Message Received(self:Fort Text Chat Main VM Object Reference=@n98, Delegate=@n100) {is=title K2Node_AddDelegate}
   n101 = Create Event()
   n102 = Get WBP_TextChat_NothingToReport_CoachMark()
   > Bind Event to On Other Action Button Clicked(self:WBP Text Chat Nothing To Report Coach Mark Object Reference=@n102, Delegate=@n101) {is=title K2Node_AddDelegate}
   n103 = Create Event()
   > Bind Event to On Dismiss Action Button Clicked(self:WBP Text Chat Nothing To Report Coach Mark Object Reference=@n102, Delegate=@n103) {is=title K2Node_AddDelegate}
   n104 = Create Event()
   n105 = Get Chat_ContextMenu()
   > Bind Event to On Focused(self:WBP Text Chat Context Menu Button Object Reference=@n105, Delegate=@n104) {is=title K2Node_AddDelegate}
   n106 = Create Event()
   n107 = Get MessageList()
   > Bind Event to On Focused(self:WBP Text Chat Message List Object Reference=@n107, Delegate=@n106)
   n108 = Create Event()
   n109 = Get WBP_TextChat_Input()
   > Bind Event to On Focused(self:WBP Text Chat Input Object Reference=@n109, Delegate=@n108)
   n110 = Create Event()
   n111 = Get WBP_TextChat_STT_CoachMark()
   > Bind Event to On Other Action Button Clicked(self:WBP Text Chat STT Coach Mark Object Reference=@n111, Delegate=@n110)
   n112 = GetRuntimeOptions()
   > Set IsPersistentChatOnInSidebar(IsPersistentChatOnInSidebar:bool=AND Boolean(
     A=Get bEnableVMDrivenSidebar!(self:Fort Runtime Options Object Reference=@n112) -> (bEnableVMDrivenSidebar:bool)
     B=GetIsPersistentTextChatEnabled!(self:Fort Runtime Options Object Reference=@n112) -> (ReturnValue:bool)
    )) -> (Output_Get:bool)
   n113 = Create Event()
   n114 = Get FortPlatformEventVM() -> (FortPlatformEventVM:Fort Platform Event VM Object Reference)
   > Bind Event to On Virtual Keyboard Shown(self:Fort Platform Event VM Object Reference=@n114, Delegate=@n113) {is=title K2Node_AddDelegate}
   n115 = Create Event()
   n116 = Get FortPlatformEventVM()
   > Bind Event to On Virtual Keyboard Hidden(self:Fort Platform Event VM Object Reference=@n116, Delegate=@n115) {is=title K2Node_AddDelegate}
   Event Destruct()
   > Unbind Event from On Input Method Changed(self=@n96, Delegate=@n95)
   > Unbind Event from On Open Text Chat(self:Fort Text Chat Player VM Object Reference=@n98, Delegate=@n97) {is=title K2Node_RemoveDelegate}
   > Unbind Event from On Message Sent(self:Fort Text Chat Main VM Object Reference=@n98, Delegate=@n99) {is=title K2Node_RemoveDelegate}
   > Unbind Event from On Message Received(self:Fort Text Chat Main VM Object Reference=@n98, Delegate=@n100) {is=title K2Node_RemoveDelegate}
   > Unbind Event from On Other Action Button Clicked(self:WBP Text Chat Nothing To Report Coach Mark Object Reference=@n102, Delegate=@n101) {is=title K2Node_RemoveDelegate}
   > Unbind Event from On Dismiss Action Button Clicked(self:WBP Text Chat Nothing To Report Coach Mark Object Reference=@n102, Delegate=@n103) {is=title K2Node_RemoveDelegate}
   > Unbind Event from On Focused(self:WBP Text Chat Context Menu Button Object Reference=@n105, Delegate=@n104) {is=title K2Node_RemoveDelegate}
   > Unbind Event from On Focused(self:WBP Text Chat Message List Object Reference=@n107, Delegate=@n106)
   > Unbind Event from On Focused(self:WBP Text Chat Input Object Reference=@n109, Delegate=@n108)
   > Unbind Event from On Other Action Button Clicked(self:WBP Text Chat STT Coach Mark Object Reference=@n111, Delegate=@n110)
   > Unbind Event from On Virtual Keyboard Shown(self:Fort Platform Event VM Object Reference=@n114, Delegate=@n113) {is=title K2Node_RemoveDelegate}
   > Unbind Event from On Virtual Keyboard Hidden(self:Fort Platform Event VM Object Reference=@n116, Delegate=@n115) {is=title K2Node_RemoveDelegate}
  ]}
 FillOutChannelsData {nodes [
   n1 = FillOutChannelsData!() -> (then, Channels:Array of Fort Text Chat Channel VM Object References) {is=title K2Node_FunctionEntry}
   > Set CurrentlySelectedTabID(CurrentlySelectedTabID:Name=GetSelectedTabId(self=Get WBP_UIKit_TabGroup()) -> (ReturnValue=None)) -> (Output_Get:Name=None)
   > RemoveAllTabs(self:Common Tab List Widget Base Object Reference=Get WBP_UIKit_TabGroup())
   > Branch(Condition=Is Not Empty(TargetArray=@n1))
   > n2 = For Each Loop(Array:Array of Fort Text Chat Channel VM Object References=@n1) -> (Array Element:Fort Text Chat Channel VM Object Reference)
   > AddTab!(
    execute
    self:WBP UIKit Tab Group Base Object Reference=Get WBP_UIKit_TabGroup()
    Tab VM:MVVM Base Viewmodel Object Reference=@n2
    Tab ID:Name=Get ChannelId!(self:Fort Text Chat Channel VM Object Reference=@n2) -> (ChannelId:Name=None)
    Index:int=@n2.Array Index
   ) -> (then)
   n3 = Get CurrentlySelectedTabID() -> (CurrentlySelectedTabID:Name=None)
   n4 = "Equal (Name)"(A=@n3, B=Get TabAddCmd() -> (TabAddCmd=None))
   Branch(execute=@n2.Completed, Condition=AND Boolean(A="Not Equal (Name)"(A=@n3) -> (ReturnValue=~), B=NOT Boolean(A=@n4)))
   > SelectTabByID(self=Get WBP_UIKit_TabGroup(), TabNameID=@n3, bSuppressClickFeedback=true)
   > RequestRefreshFocus()
  ]}
 FillOutMessageData {nodes [
   n1 = FillOutMessageData!() -> (then, Messages:Array of Text Chat Message VM Object References) {is=title K2Node_FunctionEntry}
   > n2 = InsertDividersIntoMessageList_Base!(execute, MessageVMs:Array of Text Chat Message VM Object References=@n1, OwningWidget:Object Reference=Self-Reference()) -> (then, MessageList:Array of MVVM Base Viewmodel Object References)
   > Set MessageListWithDividers(MessageListWithDividers:Array of MVVM Base Viewmodel Object References=@n2) -> (Output_Get:Array of MVVM Base Viewmodel Object References)
   n3 = Get FortTextChatPlayerVM()
   n4 = GetCurrentChannelType(self=@n3) -> (ReturnValue=All)
   > n5 = Branch(Condition=AND Boolean(
     A="Equal (Enum)"(A:EFortTextChatChannelType Enum=@n4, B:EFortTextChatChannelType Enum=Persistent)
     B=Get IsPersistentChatOnInSidebar() -> (IsPersistentChatOnInSidebar:bool)
    ))
   n6 = Get WelcomeMessageEntry() -> (WelcomeMessageEntry:BP VM Text Chat Welcome Entry Object Reference)
   > Set with Broadcast ConvoType!(
    execute
    ConvoType:EFortTextChatPersistentConversationType Enum=Get CurrentPersistentConversationType!(self:Fort Text Chat Main VM Object Reference=@n3) -> (CurrentPersistentConversationType:EFortTextChatPersistentConversationType Enum=EpicParty)
    self:BP VM Text Chat Welcome Entry Object Reference=@n6
   ) -> (then, Output_Get:EFortTextChatPersistentConversationType Enum=EpicParty)
   > n7 = Insert!(
    execute
    TargetArray:Array of MVVM Base Viewmodel Object References=Get MessageListWithDividers() -> (MessageListWithDividers:Array of MVVM Base Viewmodel Object References)
    NewItem:"MVVM Base Viewmodel Object Reference (by ref)"=@n6
    Index:int=0
   ) -> (then) {is=title K2Node_CallArrayFunction}
   n8 = Get ListView(self=Get MessageList())
   n9 = int - int(A=GetNumItems(self=@n8) -> (ReturnValue=0))
   Set bScrollToBottom(
    execute=[@n7.then, @n5.else]
    bScrollToBottom:bool=Is Item Visible!(
     self:List View Object Reference=@n8
     Item:Object Reference=GetItemAt!(self:List View Object Reference=@n8, Index:int=@n9) -> (ReturnValue:Object Reference)
    ) -> (ReturnValue:bool)
   ) -> (Output_Get:bool)
   > ClearListItems(self=@n8)
   n10 = Is Empty!(TargetArray:Array of MVVM Base Viewmodel Object References=Get MessageListWithDividers()) -> (ReturnValue:bool)
   > SetActiveWidget(
    self=Get Switcher_Content() -> (Switcher_Content:Common Visibility Switcher Object Reference)
    Widget=Select(
     Option 0:Widget Object Reference=Get O_ClippingBox() -> (O_ClippingBox:Overlay Object Reference)
     Option 1:Widget Object Reference=Get WBP_TextChat_BlankMessage()
     Index=@n10
    ) -> (ReturnValue:Widget Object Reference)
   )
   > Branch(Condition=NOT Boolean(A=@n10))
   > Set List Items!(execute, self:List View Object Reference=@n8, InListItems:Array of Object References=Get MessageListWithDividers()) -> (then) {is=title K2Node_CallFunction}
   > Branch(Condition=Get bScrollToBottom() -> (bScrollToBottom:bool))
   > ScrollToBottom(self=@n8)
  ]}
 Get Multiline InputField Enabled {nodes [
   Get Multiline InputField Enabled() {is=title K2Node_FunctionEntry}
   > n1 = GetConsoleVariableBoolValue!(execute, VariableName:String=TextChat.EnableMultilineInputField) -> (then, ReturnValue:bool)
   > Return Node!(execute, isEnabled:bool=@n1)
  ]}
 Get TTS and STT Settings {nodes [
   Get TTS and STT Settings() {is=title K2Node_FunctionEntry}
   > n1 = Get!(execute:Exec) -> (then:Exec, else:Exec, FortSocialSettingDiscreteVM:Fort Social Setting Discrete VM Object Reference)
   > SetSettingId!(execute, self:Fort Setting Value VM Object Reference=@n1, InSettingId:String=ShowTextToSpeech) -> (then)
   n2 = Get FortSocialSettingDiscreteVM() -> (FortSocialSettingDiscreteVM:Fort Social Setting Discrete VM Object Reference)
   > Set is TextToSpeech Enabled(is TextToSpeech Enabled:bool=Select(
     Option 0:bool=false
     Option 1:bool=true
     Index:int=GetSettingValueIndex!(self:Fort Social Setting Discrete VM Object Reference=@n2) -> (ReturnValue:int=0)
    ) -> (ReturnValue:bool)) -> (Output_Get:bool)
   > SetSettingId(self=Get FortSocialSettingDiscreteVM(), InSettingId=ShowSpeechToText)
   n3 = Get FortSocialSettingDiscreteVM()
   > Set is SpeechToText Enabled(is SpeechToText Enabled:bool=Select(Option 0=false, Option 1=true, Index=GetSettingValueIndex(self=@n3) -> (ReturnValue=0))) -> (Output_Get:bool)
   > UpdateSendButtonEnabled()
  ]}
 GetChatViewportQuadrantForCoachmarkLayout {nodes [
   GetChatViewportQuadrantForCoachmarkLayout() {is=title K2Node_FunctionEntry}
   > n1 = GetViewportWidgetGeometry!(execute) -> (then, ReturnValue:Geometry Structure)
   n2 = BreakVector2D(InVec=GetAbsoluteSize!(Geometry:"Geometry Structure (by ref)"=@n1) -> (ReturnValue:Vector 2D Structure))
   > Set XPos_CenterOfViewport(XPos_CenterOfViewport:double=SafeDivide(A=@n2, B=2)) -> (Output_Get:double)
   > Set YPos_CenterOfViewport(YPos_CenterOfViewport:double=SafeDivide(A=@n2.Y, B=2)) -> (Output_Get:double)
   n3 = GetCachedGeometry!(self:Widget Object Reference=Get Chat_ContextMenu()) -> (ReturnValue:"Geometry Structure (by ref)")
   n4 = BreakVector2D(InVec=LocalToViewport!(Geometry:"Geometry Structure (by ref)"=@n3, LocalCoordinate:Vector 2D Structure) -> (PixelPosition:Vector 2D Structure, ViewportPosition:Vector 2D Structure))
   n5 = BreakVector2D(InVec=GetAbsoluteSize(Geometry=@n3))
   > Set XPos_CenterOfContextButton(XPos_CenterOfContextButton:double=float + float(A=@n4, B=SafeDivide(A=@n5, B=2))) -> (Output_Get:double)
   > Set YPos_CenterOfContextButton(YPos_CenterOfContextButton:double=float + float(A=@n4.Y, B=SafeDivide(A=@n5.Y, B=2))) -> (Output_Get:double)
   > Return Node!(
    execute
    IsContextButtonOnLeftHalfOfScreen:bool="float <= float"(A=Get XPos_CenterOfContextButton() -> (XPos_CenterOfContextButton:double), B=Get XPos_CenterOfViewport() -> (XPos_CenterOfViewport:double)) -> (ReturnValue=~)
    IsContextButtonOnTopHalfOfScreen:bool="float <= float"(A=Get YPos_CenterOfContextButton() -> (YPos_CenterOfContextButton:double), B=Get YPos_CenterOfViewport() -> (YPos_CenterOfViewport:double)) -> (ReturnValue=~)
   )
  ]}
 GetInputFieldEnabled {nodes [
   n1 = GetInputFieldEnabled!() -> (then, IsChannelReadOnly:bool=~, RestrictionType:EFortTextChatRestriction Enum=Unrestricted) {is=title K2Node_FunctionEntry}
   > Return Node!(
    execute
    IsInputFieldEnabled:bool=AND Boolean(A=NOT Boolean(A=@n1), B="Equal (Enum)"(A:EFortTextChatRestriction Enum=@n1.RestrictionType, B:EFortTextChatRestriction Enum=Unrestricted))
   )
  ]}
 GetInputFieldTextHint {nodes [
   n1 = GetInputFieldTextHint!() -> (
    then
    IsRecordingCurrentChannel:bool=~
    IsCurrentChannelReadOnly:bool=~
    CurrentChannelName:String
    CurrentChannelType:EFortTextChatChannelType Enum=All
   ) {is=title K2Node_FunctionEntry}
   > n2 = Branch(Condition=@n1.IsCurrentChannelReadOnly)
   n3 = IsInZone!(self:Fort Global UIContext Object Reference=Get FortGlobalUIContext() -> (ReturnValue:Fort Global UIContext Object Reference) {is=title K2Node_GetSubsystem}) -> (ReturnValue:bool)
   > n4 = Branch(Condition=AND Boolean(A=NOT Boolean(A=@n3), B="Equal (Enum)"(A:EFortTextChatChannelType Enum=@n1.CurrentChannelType, B:EFortTextChatChannelType Enum=Game)))
   > Return Node!(execute, HintText:Text="LOCTEXT(\"You are not in a game\")")
   Return Node(
    execute=@n4.else
    HintText=Format Text!(Format="LOCTEXT(\"{ChannelName} not available\")", ChannelName:String=@n1.CurrentChannelName) -> (Result)
   )
   Return Node(
    execute=@n2.else
    HintText=Select(Option 0:Text="LOCTEXT(\"Text reporting off\")", Option 1:Text="LOCTEXT(\"Text reporting on\")", Index:bool=@n1) -> (ReturnValue:Text)
   )
  ]}
 HandleSendMessage {nodes [
   n1 = HandleSendMessage!() -> (then, Text:"Text (by ref)") {is=title K2Node_FunctionEntry}
   > Branch(Condition=NOT Boolean(A=TextIsEmpty(InText=@n1)))
   n2 = Get FortTextChatPlayerVM()
   > Set MessageToSend!(
    execute
    MessageToSend:Text=@n1
    self:Fort Text Chat Sending Message VM Object Reference=Get MessageToSend!(self:Fort Text Chat Main VM Object Reference=@n2) -> (MessageToSend:Fort Text Chat Sending Message VM Object Reference)
   ) -> (then, Output_Get:Text)
   > Send(self:Fort Text Chat Main VM Object Reference=@n2)
   > SetText(self:WBP UIKit Input Field Base Object Reference=Get WBP_TextChat_Input())
   > SetText(self:WBP Text Chat Multiline Input Field Object Reference=Get WBP_TextChat_MultilineInputField())
   > UpdateSendButtonEnabled()
   > ScrollToBottom(self=Get ListView(self=Get MessageList()))
   > RequestRefreshFocus()
  ]}
 HideBlankMessageWhenHasRestriction {nodes [
   n1 = HideBlankMessageWhenHasRestriction!() -> (then, RestrictionType:EFortTextChatRestriction Enum=Unrestricted) {is=title K2Node_FunctionEntry}
   > SetRenderOpacity(
    self=Get WBP_TextChat_BlankMessage()
    InOpacity=SelectFloat(A=1, B=0, bPickA="Equal (Enum)"(A:EFortTextChatRestriction Enum=@n1, B:EFortTextChatRestriction Enum=Unrestricted))
   )
  ]}
 InitializeChatContextMenu {nodes [
   InitializeChatContextMenu() {is=title K2Node_FunctionEntry}
   > n1 = Construct Object from Class!(execute, Class=/Script/FortniteUI.UIKitContextMenuActionContext, self) -> (then, ReturnValue:UIKit Context Menu Action Context Object Reference)
   > Set Chat_ContextMenuActionContext(Chat_ContextMenuActionContext:UIKit Context Menu Action Context Object Reference=@n1) -> (Output_Get:UIKit Context Menu Action Context Object Reference)
   > SetContext!(
    execute
    self:WBP UIKit Context Menu Anchor Button Base Object Reference=Get Chat_ContextMenu()
    InteractionContext_Data:UIKit Context Menu Action Context Object Reference=Get Chat_ContextMenuActionContext() -> (Chat_ContextMenuActionContext:UIKit Context Menu Action Context Object Reference)
   ) -> (then)
  ]}
 InitializeMessageContextMenu {nodes [
   InitializeMessageContextMenu() {is=title K2Node_FunctionEntry}
   > n1 = Construct Object from Class(Class=&UIKitContextMenuActionContext)
   > Set Message_ContextMenuActionContext(Message_ContextMenuActionContext:UIKit Context Menu Action Context Object Reference=@n1) -> (Output_Get:UIKit Context Menu Action Context Object Reference)
   > SetContext(self=Get Message_ContextMenu(), InteractionContext_Data=Get Message_ContextMenuActionContext())
  ]}
 InitializeWhisperContextMenu {nodes [
   InitializeWhisperContextMenu() {is=title K2Node_FunctionEntry}
   > n1 = Construct Object from Class(Class=&UIKitContextMenuActionContext)
   > Set Whisper_ContextMenuActionContext(Whisper_ContextMenuActionContext:UIKit Context Menu Action Context Object Reference=@n1) -> (Output_Get:UIKit Context Menu Action Context Object Reference)
   > SetContext(
    self=Get Whisper_ContextMenu()
    InteractionContext_Data=Get Whisper_ContextMenuActionContext() -> (Whisper_ContextMenuActionContext:UIKit Context Menu Action Context Object Reference)
   )
  ]}
 Mention {nodes [
   Mention!() -> (then, InteractionNameID:Name) {is=title K2Node_FunctionEntry}
   n1 = GetMessageVM!(
    self:Fort Text Chat Main VM Object Reference=Get FortTextChatPlayerVM()
    MessageId:String="To String (Name)"(InName=Get MessageId() -> (MessageId=None))
   ) -> (ReturnValue:Text Chat Message VM Object Reference)
   > Is Valid(InputObject=@n1)
   > n2 = Branch(Condition=Get Multiline InputField Enabled() {is=title K2Node_CallFunction})
   n3 = Format Text!(
    Format="LOCTEXT(\"@{DisplayName}\")"
    DisplayName:String=Get DisplayName!(self:Text Chat Message VM Object Reference=@n1) -> (DisplayName:String)
   ) -> (Result)
   > SetText(self=Get WBP_TextChat_MultilineInputField(), Text=@n3)
   SetText(execute=@n2.else, self:WBP UIKit Input Field Base Object Reference=Get WBP_TextChat_Input(), Text=@n3)
  ]}
 MoveChatAboveVirtualKeyboard {nodes [
   MoveChatAboveVirtualKeyboard!() -> (then, RectTop:int, RectBottom:int, VirtualKeyboardInputHeight:int) {is=title K2Node_FunctionEntry}
   > SetVisibility(
    self=[Get WBP_ActionBarContainer() -> (WBP_ActionBarContainer:WBP Action Bar Container Object Reference), Get WBP_UIKit_TabGroup()]
    InVisibility=Collapsed
   )
   n1 = Get FortPlatformEventVM()
   n2 = Get WBP_TextChat_Input()
   > n3 = Branch(Condition=AND Boolean(
     A=IsWidgetAllowedToEnableIntegratedKeyboard!(self:Fort Platform Event VM Object Reference=@n1) -> (ReturnValue:bool)
     B=Get MobileEnableIntegratedKeyboard!(self:WBP UIKit Input Field Base Object Reference=@n2) -> (MobileEnableIntegratedKeyboard:bool)
    ))
   SetPadding!(
    execute=[@n6.then, @n3.then]
    self:Border Slot Object Reference=SlotAsBorderSlot!(Widget:Widget Object Reference=Get Mobile_SafeZone() -> (Mobile_SafeZone:Safe Zone Object Reference)) -> (ReturnValue:Border Slot Object Reference)
    InPadding:Margin Structure=Get OpenVirtualKeyboardSafeZoneMargin() -> (OpenVirtualKeyboardSafeZoneMargin:Margin Structure)
   ) -> (then)
   > n4 = Branch(Condition=GetConfigDataBoolWithDefault(EntryName=UseFortFunctionLibraryCalculateChatPadding, bDefaultValue=true))
   > SetPadding(
    self:Size Box Slot Object Reference=SlotAsSizeBoxSlot!(Widget:Widget Object Reference=Get GP_TextChat_Wrapper() -> (GP_TextChat_Wrapper:Grid Panel Object Reference)) -> (ReturnValue:Size Box Slot Object Reference)
    InPadding=CalculateChatPadding!(
     TargetWidget:Widget Object Reference=Get GP_TextChat_Wrapper()
     RectTop:int=Get RectTop() -> (RectTop:int=0)
     RectBottom:int=Get RectBottom() -> (RectBottom:int=0)
     VirtualKeyboardInputHeight:int=Get VirtualKeyboardInputHeight() -> (VirtualKeyboardInputHeight:int=0)
    ) -> (ReturnValue:Margin Structure)
   )
   UpdateChatPadding!(
    execute=@n4.else
    self:Fort Text Chat V3 Object Reference
    TargetWidget:Widget Object Reference=Get GP_TextChat_Wrapper()
    RectTop:int=Get RectTop() -> (RectTop=0)
    RectBottom:int=Get RectBottom() -> (RectBottom=0)
    VirtualKeyboardInputHeight:int=Get VirtualKeyboardInputHeight() -> (VirtualKeyboardInputHeight=0)
   ) -> (then)
   n5 = Get GP_BottomContent()
   SetRenderScale!(execute=@n3.else, self:Widget Object Reference=@n5, Scale:Vector 2D Structure) -> (then)
   > n6 = SetRow!(execute, self:Grid Slot Object Reference=SlotAsGridSlot(Widget=@n5), InRow:int=0) -> (then)
  ]}
 MuteChat {nodes [
   MuteChat!() -> (then, InteractionID:Name) {is=title K2Node_FunctionEntry}
   > ToggleMuteTextChatPreview(self:Fort Text Chat Main VM Object Reference=Get FortTextChatPlayerVM())
  ]}
 MutePlayer {nodes [
   MutePlayer!() -> (then, InteractionID:Name) {is=title K2Node_FunctionEntry}
   > PerformMessageContextMenuAction(self=Get FortTextChatPlayerVM(), MessageId="To String (Name)"(InName=Get MessageId() -> (MessageId=None)), MenuAction=Mute)
  ]}
 Navigate Left from Context Menu {nodes [
   Navigate Left from Context Menu!() -> (then, Navigation:EUINavigation Enum) {is=title K2Node_FunctionEntry}
   n1 = Get FortTextChatPlayerVM()
   > Branch(Condition=NOT Boolean(A=IsCurrentChannelReadOnly(self=@n1)))
   n2 = Get Send_ButtonIcon()
   n3 = IsVisible(self=Get Send_Dictation())
   n4 = Select(
    Option 0:Widget Object Reference=Get WBP_TextChat_Input()
    Option 1:Widget Object Reference=Get WBP_TextChat_MultilineInputField()
    Index=Get Multiline InputField Enabled()
   ) -> (ReturnValue:Widget Object Reference)
   n5 = Get Send_Dictation()
   > Return Node!(
    execute
    ReturnValue:Widget Object Reference=Select(Option 0=Select(Option 0=@n4, Option 1=@n5, Index=@n3), Option 1=Get Send_ButtonIcon(), Index=IsVisible(self=@n2))
   )
  ]}
 Navigate Left from Send Button {nodes [
   Navigate Left from Send Button!() -> (then, Navigation:EUINavigation Enum) {is=title K2Node_FunctionEntry}
   n1 = IsCurrentChannelReadOnly(self=Get FortTextChatPlayerVM())
   n2 = Get Multiline InputField Enabled()
   n3 = Get WBP_TextChat_Input()
   n4 = Get WBP_TextChat_MultilineInputField()
   n5 = "Equal (Enum)"(
    A:ESlateVisibility Enum=Get Visibility!(self:Widget Object Reference=Select(Option 0=@n3, Option 1=@n4, Index=@n2)) -> (Visibility:ESlateVisibility Enum=Visible)
    B:ESlateVisibility Enum=HitTestInvisible
   )
   n6 = Get Multiline InputField Enabled()
   n7 = Get WBP_TextChat_Input()
   n8 = Get WBP_TextChat_MultilineInputField()
   > Return Node(ReturnValue=Select(Option 0=Select(Option 0=@n7, Option 1=@n8, Index=@n6), Option 1=Get Chat_ContextMenu(), Index=OR Boolean!(A=@n1, B=@n5) -> (ReturnValue)))
  ]}
 Navigate Right from Input Field {nodes [
   Navigate Right from Input Field!() -> (then, Navigation:EUINavigation Enum) {is=title K2Node_FunctionEntry}
   n1 = Get Send_ButtonIcon()
   n2 = IsVisible(self=Get Send_Dictation())
   n3 = Get Send_Dictation()
   n4 = Get Chat_ContextMenu()
   > Return Node(ReturnValue=Select(Option 0=Select(Option 0=@n4, Option 1=@n3, Index=@n2), Option 1=Get Send_ButtonIcon(), Index=IsVisible(self=@n1)))
  ]}
 Navigate Right from Send Button {nodes [
   Navigate Right from Send Button!() -> (then, Navigation:EUINavigation Enum) {is=title K2Node_FunctionEntry}
   > Return Node(ReturnValue=Get Chat_ContextMenu())
  ]}
 Navigate Up to ActionButton {nodes [
   n1 = Navigate Up to ActionButton!() -> (then, Navigation:EUINavigation Enum) {is=title K2Node_FunctionEntry}
   n2 = Get WBP_Social_Restriction_Generic_Internal!(self:WBP Text Chat Restrictions Object Reference=Get WBP_TextChat_Restrictions()) -> (WBP_Social_Restriction_Generic_Internal:WBP Social Restriction Generic Internal Object Reference)
   > n3 = Branch(Condition=IsVisible(self=@n2))
   > Return Node(ReturnValue=@n2)
   n4 = Branch(
    execute=@n3.else
    Condition=IsVisible(self=Get Button_CompactChat_RestrictionAction() -> (Button_CompactChat_RestrictionAction:WBP UIKit Button Generic Object Reference))
   )
   > Return Node(ReturnValue=Get Button_CompactChat_RestrictionAction())
   n5 = Navigate Up to Message List!(execute=@n4.else, self, Navigation:EUINavigation Enum=@n1) -> (then, ReturnValue:Widget Object Reference)
   > Return Node(ReturnValue=@n5)
  ]}
 Navigate Up to Message List {nodes [
   Navigate Up to Message List!() -> (then, Navigation:EUINavigation Enum) {is=title K2Node_FunctionEntry}
   n1 = Get Switcher_Content()
   > Branch(Condition="Equal (Object)"(A=GetActiveWidget!(self:Common Visibility Switcher Object Reference=@n1) -> (ReturnValue:Widget Object Reference), B=Get O_ClippingBox()))
   n2 = Get ListView(self=Get MessageList())
   > SetSelectedIndex(self=@n2, Index=int - int(A=GetNumItems(self=@n2) -> (ReturnValue=0)))
   > Return Node(ReturnValue=@n2)
  ]}
 NavigateDownToInputRow {nodes [
   NavigateDownToInputRow!() -> (then, Navigation:EUINavigation Enum) {is=title K2Node_FunctionEntry}
   n1 = Get FortTextChatPlayerVM()
   > n2 = Branch(Condition=NOT Boolean(A=IsCurrentChannelReadOnly(self=@n1)))
   n3 = Get Block_EditableText(self:WBP UIKit Input Field Base Object Reference=Get WBP_TextChat_Input()) -> (Block_EditableText:WBP UIKit Block Editable Text Object Reference)
   n4 = Get Block_EditableText(self:WBP Text Chat Multiline Input Field Object Reference=Get WBP_TextChat_MultilineInputField()) -> (Block_EditableText:WBP Text Chat Block Multiline Editable Text Object Reference)
   > Return Node(ReturnValue=Select(
     Option 0=Get EditableText(self:WBP UIKit Block Editable Text Object Reference=@n3) -> (EditableText:UIKit Editable Text Object Reference)
     Option 1=Get EditableText(self:WBP Text Chat Block Multiline Editable Text Object Reference=@n4) -> (EditableText:"Text Box (Multi- Line) Object Reference")
     Index=Get Multiline InputField Enabled()
    ))
   Return Node(execute=@n2.else, ReturnValue=Get Chat_ContextMenu())
  ]}
 NavigateFromListMessage {nodes [
   NavigateFromListMessage!() -> (then, Navigation:EUINavigation Enum) {is=title K2Node_FunctionEntry}
   n1 = Get WBP_Social_Restriction_Generic_Internal(self=Get WBP_TextChat_Restrictions())
   > n2 = Branch(Condition=IsVisible(self=@n1))
   > Return Node(ReturnValue=@n1)
   n3 = Branch(execute=@n2.else, Condition=IsVisible(self=Get Button_CompactChat_RestrictionAction()))
   > Return Node(ReturnValue=Get Button_CompactChat_RestrictionAction())
   n4 = NavigateDownToInputRow!(execute=@n3.else, self, Navigation:EUINavigation Enum=Left) -> (then, ReturnValue:Widget Object Reference) {is=title K2Node_CallFunction}
   > Return Node(ReturnValue=@n4)
  ]}
 NoFriendsAvailable {nodes [NoFriendsAvailable!() -> (then, InteractionNameID:Name) {is=title K2Node_FunctionEntry}]}
 OnActiveStateChanged {nodes [OnActiveStateChanged!() -> (then, IsActivated:bool=~) {is=title K2Node_FunctionEntry}]}
 OnButtonCloseClicked {nodes [OnButtonCloseClicked!() -> (then, Button:Common Button Base Object Reference) {is=title K2Node_FunctionEntry}]}
 OnMouseWheel {nodes [
   OnMouseWheel!() -> (then, MyGeometry:Geometry Structure, MouseEvent:"Pointer Event Structure (by ref)") {is=title K2Node_FunctionEntry}
   > n1 = Branch(Condition=HasAnyUserFocus!(self:Widget Object Reference=Self-Reference()) -> (ReturnValue:bool))
   > Return Node(ReturnValue:Event Reply Structure=Unhandled())
   Return Node(execute=@n1.else, ReturnValue=Handled())
  ]}
 OnSpeechToTextMicStatusChanged {nodes [
   n1 = OnSpeechToTextMicStatusChanged!() -> (then, Mic Status:EFortSpeechToTextMicrophoneStatus Enum=Available) {is=title K2Node_FunctionEntry}
   > SetIsLocked(
    self=Get Send_Dictation()
    bInIsLocked="Not Equal (Enum)"(A:EFortSpeechToTextMicrophoneStatus Enum=@n1, B:EFortSpeechToTextMicrophoneStatus Enum=Available)
   )
  ]}
 OnViewProfileRequested {nodes [OnViewProfileRequested() {is=title K2Node_FunctionEntry}]}
 OnWhisperUserClicked {nodes [
   n1 = OnWhisperUserClicked!() -> (then, InteractionNameID:Name) {is=title K2Node_FunctionEntry}
   > CreateWhisperChannelIfNeeded!(execute, self:Fort Text Chat Main VM Object Reference=Get FortTextChatPlayerVM(), InWhisperChannelId:"Name (by ref)"=@n1) -> (then)
   > SelectTabByID(self=Get WBP_UIKit_TabGroup(), TabNameID=@n1, bSuppressClickFeedback=true)
   > SetDirectMessageContext(self=Get FortTextChatPlayerVM(), Context=TextChat)
  ]}
 OpenChatSettings {nodes [
   OpenChatSettings!() -> (then, InteractionNameID:Name) {is=title K2Node_FunctionEntry}
   n1 = Get FortTextChatPlayerVM()
   > Set SettingsScreenClass!(
    execute
    SettingsScreenClass:Fort Settings Screen Soft Class Reference=/FortniteCoreUI/Settings/SettingScreen.SettingScreen_C
    self:Fort Text Chat Main VM Object Reference=@n1
   ) -> (then, Output_Get:Fort Settings Screen Soft Class Reference)
   > Set OptionsMenuClass!(
    execute
    OptionsMenuClass:Fort Options Menu Soft Class Reference=/FortniteCoreUI/Settings/OptionsMenu.OptionsMenu_C
    self:Fort Text Chat Main VM Object Reference=@n1
   ) -> (then, Output_Get:Fort Options Menu Soft Class Reference)
   > ViewSettings(self:Fort Text Chat Main VM Object Reference=@n1)
   > Call On Button Close Clicked()
  ]}
 PlayDictationSound {nodes [
   n1 = PlayDictationSound!() -> (then, Curret Speech To Text State:EFortSpeechToTextState Enum=Idle) {is=title K2Node_FunctionEntry}
   > n2 = Switch on EFortSpeechToTextState(Selection=@n1)
   PlaySound2D(execute=@n2.Recording, Sound=/FortniteCoreUI/UI/Social/Chat/TextChat/Audio/UI_TTS_Activate_A)
   PlaySound2D(execute=@n2.Transcribing, Sound=/FortniteCoreUI/UI/Social/Chat/TextChat/Audio/UI_TTS_Deactivate_A)
  ]}
 PlayTextChatSound {nodes [
   n1 = PlayTextChatSound!() -> (then, Sound:Sound Base Object Reference) {is=title K2Node_FunctionEntry}
   > Branch(Condition=IsTextChatSoundEnabled!(self:Fort Text Chat Main VM Object Reference=Get FortTextChatPlayerVM()) -> (ReturnValue:bool))
   > Is Valid(InputObject=@n1)
   > PlaySound2D(Sound=@n1)
  ]}
 Report {nodes [
   Report!() -> (then, InteractionNameID:Name) {is=title K2Node_FunctionEntry}
   n1 = Get MessageId() -> (MessageId=None)
   > Report(self:Text Chat Message VM Object Reference=GetMessageVM(self=Get FortTextChatPlayerVM(), MessageId="To String (Name)"(InName=@n1))) {is=title K2Node_CallFunction}
  ]}
 ResetChatPosition {nodes [
   ResetChatPosition() {is=title K2Node_FunctionEntry}
   > SetVisibility(self=[Get WBP_ActionBarContainer(), Get WBP_UIKit_TabGroup()], InVisibility=SelfHitTestInvisible)
   n1 = Get GP_BottomContent()
   > SetRenderScale(self=@n1, Scale=(X=1,Y=1))
   > SetRow(self=SlotAsGridSlot(Widget=@n1), InRow=Get InitialGPContentGridRow() -> (InitialGPContentGridRow:int=0))
   > SetPadding(self=SlotAsSizeBoxSlot(Widget=Get GP_TextChat_Wrapper()))
   > SetPadding(
    self:Border Slot Object Reference=SlotAsBorderSlot(Widget=Get Mobile_SafeZone())
    InPadding=Get HiddenVirtualKeyboardSafeZoneMargin() -> (HiddenVirtualKeyboardSafeZoneMargin:Margin Structure)
   )
  ]}
 SendTextMessage {nodes [
   n1 = SendTextMessage!() -> (then, Text:"Text (by ref)", CommitMethod:ETextCommit Enum=Default) {is=title K2Node_FunctionEntry}
   > n2 = Switch on ETextCommit!(execute, Selection:ETextCommit Enum=@n1.CommitMethod) -> (Default:Exec, OnEnter:Exec, OnUserMovedFocus:Exec, OnCleared:Exec) {is=title K2Node_SwitchEnum}
   n3 = "Equal (Enum)"(A:ECommonInputType Enum=Get CurrentInputType(), B:ECommonInputType Enum=MouseAndKeyboard)
   n4 = Branch(execute=@n2.OnCleared, Condition=@n3)
   Call On Button Close Clicked(execute=[@n6.then, @n4.then])
   > SetVisibility(self=Get GP_BottomContent(), InVisibility=Hidden)
   n5 = Get Text!() -> (Text:"Text (by ref)")
   n6 = Branch(execute=@n2.OnEnter, Condition=AND Boolean(A=TextIsEmpty(InText=@n5), B=@n3))
   HandleSendMessage!(execute=@n6.else, self, Text:"Text (by ref)"=Get Text()) -> (then) {is=title K2Node_CallFunction}
  ]}
 Set Transcribed Text {nodes [
   n1 = Set Transcribed Text!() -> (then, Text:Text) {is=title K2Node_FunctionEntry}
   > n2 = Branch(Condition=TextIsEmpty(InText=@n1))
   SetText(execute=@n2.else, self=Get WBP_TextChat_Input(), Text=Get Text() -> (Text:Text))
   n3 = Get WBP_TextChat_MultilineInputField()
   > "SetText (Multi-Line Text Box)"!(execute, self:"Text Box (Multi- Line) Object Reference"=Get EditableText(self=Get Block_EditableText(self=@n3)), InText:Text=Get Text()) -> (then)
  ]}
 SetBackButtonText {nodes [
   n1 = SetBackButtonText!() -> (then, Close:bool=~) {is=title K2Node_FunctionEntry}
   > SetText!(
    execute
    self:Fort CTAButton Object Reference=Get Button_BackChat() -> (Button_BackChat:WBP UIKit Button Quiet Object Reference)
    ButtonText:"Text (by ref)"=SelectText(A="LOCTEXT(\"Close\")", B="LOCTEXT(\"Back\")", bPickA=@n1)
   ) -> (then)
  ]}
 SetCanReportChat {nodes [
   n1 = SetCanReportChat!() -> (then, CanReport:bool=~) {is=title K2Node_FunctionEntry}
   > n2 = Branch(Condition=@n1)
   n3 = Get Button_ReportChat()
   > SetIsLocked(self=@n3, bInIsLocked=false)
   > SetIsEnabled(self=@n3, bInIsEnabled=true)
   n4 = Branch(execute=@n2.else, Condition=ShouldShowOnboarding(self=Get WBP_TextChat_NothingToReport_CoachMark()))
   n5 = Get Button_ReportChat()
   > SetIsLocked(self=@n5, bInIsLocked=true)
   > SetIsEnabled(self=@n5, bInIsEnabled=true)
   n6 = Get Button_ReportChat()
   SetIsLocked(execute=@n4.else, self=@n6, bInIsLocked=false)
   > SetIsEnabled(self=@n6, bInIsEnabled=false)
  ]}
 SetChatContextMenuActionContext {nodes [
   SetChatContextMenuActionContext() {is=title K2Node_FunctionEntry}
   n1 = Get bIsRecordingCurrentChannel!(self:Fort Text Chat Main VM Object Reference=Get FortTextChatPlayerVM()) -> (bIsRecordingCurrentChannel:bool)
   > "SetText (Text)"(
    self=Get Text_ChatContextMenu()
    InText=Select(Option 0:Text="LOCTEXT(\"Text reporting off\")", Option 1:Text="LOCTEXT(\"Text reporting on\")", Index=@n1) -> (ReturnValue:Text)
   )
   > SetBrushTintColor!(
    execute
    self:Image Object Reference=Get Image_ChatContextMenu()
    TintColor_SpecifiedColor:Linear Color Structure=Select(
     Option 0:Linear Color Structure=Get RecordingColorOff()
     Option 1:Linear Color Structure=Get RecordingColorOn() -> (RecordingColorOn:Linear Color Structure)
     Index=@n1
    ) -> (ReturnValue:Linear Color Structure)
    TintColor_ColorUseRule:ESlateColorStylingMode Enum
   ) -> (then)
   n2 = Get Chat_ContextMenuActionContext()
   > RemoveAllInteractions(self:UIKit Context Menu Action Context Object Reference=@n2)
   n3 = Get FortTextChatPlayerVM()
   > AddInteraction!(
    execute
    self:UIKit Context Menu Action Context Object Reference=@n2
    InActionInfo_DisplayName:Text=Select(
     Option 0:Text="LOCTEXT(\"Hide all chat previews\")"
     Option 1:Text="LOCTEXT(\"Show all chat previews\")"
     Index=IsTextChatPreviewMuted!(self:Fort Text Chat Main VM Object Reference=@n3) -> (ReturnValue:bool)
    ) -> (ReturnValue:Text)
    InActionInfo_IconTexture:Texture 2D Soft Object Reference
    InActionInfo_IconMaterial:Material Instance Soft Object Reference
    InActionInfo_SortPriority:int=1000
    InActionInfo_bIsAvailable:bool=True
    InActionInfo_bIsEnabled:bool=True
    InActionInfo_bIsSelected:bool=False
    InActionInfo_bAddDivider:bool=False
    InActionInfo_bPositionDividerAfter:bool=False
    InActionInfo_bCloseMenuOnSelect:bool=False
    InActionInfo_InteractionNameID:Name=MuteChat
    InActionInfo_Interaction:Data Table Row Handle Structure=(DataTable=None,RowName="")
    InActionInfo_PressedDelegate:Delegate="(null).None"
    InActionInfo_ReleasedDelegate:Delegate="(null).None"
    InInteractionDelegate:"Delegate (by ref)"=Create Event()
   ) -> (then)
   > AddInteraction(
    self=@n2
    InActionInfo_DisplayName="LOCTEXT(\"Chat settings\")"
    InActionInfo_SortPriority=1000
    InActionInfo_bIsAvailable=True
    InActionInfo_bIsEnabled=True
    InActionInfo_bIsSelected=False
    InActionInfo_bAddDivider=False
    InActionInfo_bPositionDividerAfter=False
    InActionInfo_bCloseMenuOnSelect=False
    InActionInfo_InteractionNameID=ChatSettings
    InActionInfo_Interaction=(DataTable=None,RowName="")
    InActionInfo_PressedDelegate="(null).None"
    InActionInfo_ReleasedDelegate="(null).None"
    InInteractionDelegate=Create Event()
   )
  ]}
 SetDictationVisuals {nodes [
   SetDictationVisuals!() -> (then, Current Speech to Text State:EFortSpeechToTextState Enum=Idle) {is=title K2Node_FunctionEntry}
   > n1 = Branch(Condition=Get Multiline InputField Enabled())
   n2 = Select!(
    Idle:bool=false
    Initializing:bool=true
    Recording:bool=true
    Transcribing:bool=true
    Deinitializing:bool=true
    Unavailable:bool=false
    Index:EFortSpeechToTextState Enum=Get Current Speech to Text State() -> (Current Speech to Text State:EFortSpeechToTextState Enum=Idle)
   ) -> (ReturnValue:bool)
   n3 = Get WBP_TextChat_MultilineInputField()
   > "SetIsReadOnly (Multi-Line Text Box)"!(execute, self:"Text Box (Multi- Line) Object Reference"=Get EditableText(self=Get Block_EditableText(self=@n3)), bReadOnly:bool=@n2) -> (then)
   n4 = Select(
    Idle:ESlateVisibility Enum=SelfHitTestInvisible
    Initializing:ESlateVisibility Enum=HitTestInvisible
    Recording:ESlateVisibility Enum=HitTestInvisible
    Transcribing:ESlateVisibility Enum=HitTestInvisible
    Deinitializing:ESlateVisibility Enum=HitTestInvisible
    Unavailable:ESlateVisibility Enum=SelfHitTestInvisible
    Index=Get Current Speech to Text State() -> (Current Speech to Text State=Idle)
   ) -> (ReturnValue:ESlateVisibility Enum)
   > n5 = SetVisibility(self=Get WBP_TextChat_MultilineInputField(), InVisibility=@n4)
   n6 = Switch on EFortSpeechToTextState(execute=[@n18.then, @n5.then], Selection=Get Current Speech to Text State() -> (Current Speech to Text State=Idle))
   n7 = Get Timer Handle()
   n8 = Branch(
    execute=[@n6.Deinitializing, @n6.Transcribing]
    Condition=Is Timer Active by Handle!(Handle:Timer Handle Structure=@n7) -> (ReturnValue:bool)
   )
   > Clear and Invalidate Timer by Handle(Handle=@n7)
   > n9 = Set Record Time(Record Time=Get SpeechToTextRecordTime(self=Get FortTextChatPlayerVM()))
   SetHintText!(
    execute=[@n9.then, @n8.else]
    self:WBP UIKit Input Field Base Object Reference=Get WBP_TextChat_Input()
    Hint Text:Text=Get Hint Text Please Wait() -> (Hint Text Please Wait:Text)
   ) -> (then)
   > SetHintText(self:WBP Text Chat Multiline Input Field Object Reference=Get WBP_TextChat_MultilineInputField(), Hint Text=Get Hint Text Please Wait())
   > Button Dictate Button Visuals!(
    execute
    self
    Icon:Texture 2D Object Reference
    Material Instance:Material Instance Object Reference=/FortUILibrary/UIKit/Throbber/MI_UIKit_Monochrome_Throbber
    Icon Size:Vector 2D Structure=Get DictateDefaultIconSize() -> (DictateDefaultIconSize:Vector 2D Structure)
    InColorAndOpacity:Linear Color Structure=(R=1,G=1,B=1,A=1)
   ) -> (then) {is=title K2Node_CallFunction}
   > SetText!(execute, self:WBP UIKit Button Generic Base Object Reference=Get Send_Dictation(), Text:Text) -> (then)
   n10 = Get WBP_TextChat_MultilineInputField()
   > Set WrapTextAt!(execute, WrapTextAt:float=385, self:Text Layout Widget Object Reference=Get EditableText(self=Get Block_EditableText(self=@n10))) -> (then, Output_Get:float)
   RequestRefreshFocus(execute=@n6.Idle)
   > UpdateSendButtonEnabled()
   > n11 = Branch(Condition=TextIsEmpty(InText=Get Current Hint() -> (Current Hint:Text)))
   Button Dictate Button Visuals(
    execute=[@n12.then, @n11.then]
    Material Instance=/FortniteCoreUI/UI/Social/Chat/TextChat/MaterialInstances/MI_UI_TextChat_Dictation
    Icon Size=Get DictateDefaultIconSize()
    InColorAndOpacity=(R=0,G=0,B=0,A=1)
   )
   SetHintText(execute=@n11.else, self:WBP UIKit Input Field Base Object Reference=Get WBP_TextChat_Input(), Hint Text=Get Current Hint())
   > n12 = SetHintText(self:WBP Text Chat Multiline Input Field Object Reference=Get WBP_TextChat_MultilineInputField(), Hint Text=Get Current Hint())
   n13 = Set Timer by Event(execute=@n6.Recording, Delegate=Create Event(), Time=1, bLooping=true)
   > Set Timer Handle(Timer Handle:Timer Handle Structure=@n13) -> (Output_Get:Timer Handle Structure)
   n14 = Get WBP_TextChat_MultilineInputField()
   > Set WrapTextAt(WrapTextAt=360, self=Get EditableText(self=Get Block_EditableText(self=@n14)))
   > SetHintText(self:WBP UIKit Input Field Base Object Reference=Get WBP_TextChat_Input(), Hint Text=Get Hint Text Listening() -> (Hint Text Listening:Text))
   > SetHintText(self:WBP Text Chat Multiline Input Field Object Reference=Get WBP_TextChat_MultilineInputField(), Hint Text=Get Hint Text Listening())
   > Button Dictate Button Visuals(
    Material Instance=/FortniteCoreUI/UI/Social/Chat/TextChat/MaterialInstances/MI_UI_TextChat_Dictation_Stop_Icon
    Icon Size=Get DictateStopIconSize() -> (DictateStopIconSize:Vector 2D Structure)
    InColorAndOpacity=(R=0.0048,G=0.004,B=0.0116,A=1)
   )
   n15 = Get SpeechToTextRecordTime(self=Get FortTextChatPlayerVM())
   n16 = "float >= float"(A=@n15, B=10)
   > SetText(
    self=Get Send_Dictation()
    Text=Format Text!(Format="LOCTEXT(\"0:{zero}{n}\")", zero:Text=SelectText(B="LOCTEXT(\"0\")", bPickA=@n16), n:int=Truncate(A=@n15)) -> (Result)
   )
   Set Current Hint(
    execute=@n6.Initializing
    Current Hint:Text=Get Hint Text!(self:WBP UIKit Input Field Base Object Reference=Get WBP_TextChat_Input()) -> (Hint Text:Text)
   ) -> (Output_Get:Text)
   > SetHintText(self:WBP UIKit Input Field Base Object Reference=Get WBP_TextChat_Input(), Hint Text=Get Hint Text Please Wait())
   > SetHintText(self:WBP Text Chat Multiline Input Field Object Reference=Get WBP_TextChat_MultilineInputField(), Hint Text=Get Hint Text Please Wait())
   > Button Dictate Button Visuals(
    Material Instance=&MI_UIKit_Monochrome_Throbber
    Icon Size=Get DictateDefaultIconSize()
    InColorAndOpacity=(R=1,G=1,B=1,A=1)
   )
   n17 = Get WBP_TextChat_Input()
   "SetIsReadOnly (Editable Text)"!(
    execute=@n1.else
    self:Editable Text Object Reference=Get EditableText(self:WBP UIKit Block Editable Text Object Reference=Get Block_EditableText(self:WBP UIKit Input Field Base Object Reference=@n17) -> (Block_EditableText:WBP UIKit Block Editable Text Object Reference)) -> (EditableText:UIKit Editable Text Object Reference)
    InbIsReadyOnly:bool=@n2
   ) -> (then)
   > n18 = SetVisibility(self=Get WBP_TextChat_Input(), InVisibility=@n4)
  ]}
 SetDynamicCoachmarkLayout {nodes [
   n1 = SetDynamicCoachmarkLayout!() -> (then, IsChatContextButtonOnLeftHalfOfScreen:bool=~, IsChatContextButtonOnTopHalfOfScreen:bool=~) {is=title K2Node_FunctionEntry}
   n2 = MakeVector2D(X=SelectFloat(A=0, B=1, bPickA=@n1), Y=SelectFloat(A=0.5, B=1, bPickA=@n1.IsChatContextButtonOnTopHalfOfScreen))
   n3 = Get VisAttach_CoachmarkReport() -> (VisAttach_CoachmarkReport:Common Visual Attachment Object Reference)
   > Set ContentAnchor!(execute, ContentAnchor:Vector 2D Structure=@n2, self:Common Visual Attachment Object Reference=@n3) -> (then, Output_Get:Vector 2D Structure)
   n4 = Get VisAttach_CoachmarkSettings() -> (VisAttach_CoachmarkSettings:Common Visual Attachment Object Reference)
   > Set ContentAnchor(ContentAnchor=@n2, self=@n4)
   n5 = Get IsChatContextButtonOnLeftHalfOfScreen() -> (IsChatContextButtonOnLeftHalfOfScreen:bool)
   n6 = Get IsChatContextButtonOnLeftHalfOfScreen()
   > Set ContentAnchor(
    ContentAnchor=MakeVector2D(X=SelectFloat(A=-0.2, B=0.9, bPickA=@n6), Y=SelectFloat(A=1.3, B=0.95, bPickA=@n5))
    self=Get VisAttach_CoachmarkSTT() -> (VisAttach_CoachmarkSTT:Common Visual Attachment Object Reference)
   )
   n7 = Select!(
    Option 0:Vector 2D Structure=MakeVector2D(X=SelectFloat(A=-20, B=-80, bPickA=@n1), Y=16)
    Option 1:Vector 2D Structure=MakeVector2D(X=SelectFloat(A=0, B=-80, bPickA=@n1), Y=-36)
    Index:bool=@n1.IsChatContextButtonOnTopHalfOfScreen
   ) -> (ReturnValue:Vector 2D Structure)
   > SetGridSlotNudge!(execute, Widget:Widget Object Reference=@n3, Nudge:Vector 2D Structure=@n7) -> (then)
   > SetGridSlotNudge(Widget=@n4, Nudge=@n7)
   n8 = Select(
    Option 0:E_UI_OutBoardPosition Enum=Select(Option 0=NewEnumerator12, Option 1=NewEnumerator11, Index=@n1.IsChatContextButtonOnTopHalfOfScreen)
    Option 1:E_UI_OutBoardPosition Enum=Select(Option 0=NewEnumerator9, Option 1=NewEnumerator8, Index=@n1.IsChatContextButtonOnTopHalfOfScreen)
    Index=@n1
   ) -> (ReturnValue:E_UI_OutBoardPosition Enum)
   > SetCoachmarkArrowPosition!(
    execute
    self:WBP Text Chat Nothing To Report Coach Mark Object Reference=Get WBP_TextChat_NothingToReport_CoachMark()
    ArrowPosition:E_UI_OutBoardPosition Enum=@n8
   ) -> (then)
   > SetCoachmarkArrowPosition(self:WBP Text Chat Chat Settings Coach Mark Object Reference=Get WBP_TextChat_ChatSettings_CoachMark(), ArrowPosition=@n8)
   > SetCoachmarkArrowPosition(
    self:WBP Text Chat STT Coach Mark Object Reference=Get WBP_TextChat_STT_CoachMark()
    ArrowPosition=Select(Option 0=NewEnumerator12, Option 1=NewEnumerator4, Index=Get IsChatContextButtonOnLeftHalfOfScreen())
   )
  ]}
 SetInGameAdjustments {nodes [
   SetInGameAdjustments() {is=title K2Node_FunctionEntry}
   > n1 = Set isInZone(isInZone:bool=IsInZone(self=Get FortGlobalUIContext!() -> (ReturnValue:Fort Global UIContext Object Reference) {is=title K2Node_GetBlueprintContextEx})) -> (Output_Get:bool)
   > Branch(Condition=@n1)
   > SetHorizontalAlignment(
    self=SlotAsGridSlot(Widget=Get CommonVisualAttachment() -> (CommonVisualAttachment:Common Visual Attachment Object Reference))
    InHorizontalAlignment=HAlign_Left
   )
   > Set ContentAnchor(self=Get CommonVisualAttachment())
  ]}
 SetInputFieldEnabled {nodes [
   SetInputFieldEnabled!() -> (then, IsInputFieldEnabled:bool=~) {is=title K2Node_FunctionEntry}
   > SetIsEnabled(
    self=[Get WBP_TextChat_Input(), Get WBP_TextChat_MultilineInputField()]
    bInIsEnabled=Get IsInputFieldEnabled() -> (IsInputFieldEnabled:bool)
   )
   > n1 = Branch(Condition=Get IsInputFieldEnabled())
   > Set RenderOpacity!(execute, RenderOpacity:float=1, self:Widget Object Reference=Get WBP_TextChat_Input()) -> (then, Output_Get:float)
   > n2 = Set RenderOpacity(RenderOpacity=1, self=Get WBP_TextChat_MultilineInputField())
   UpdateTabsData(execute=[@n2.then, @n3.then], IsTabActive=Get IsInputFieldEnabled())
   > UpdateSendButtonVisibility()
   > UpdateSendButtonEnabled()
   Set RenderOpacity(execute=@n1.else, RenderOpacity=0.4, self=Get WBP_TextChat_Input())
   > n3 = Set RenderOpacity(RenderOpacity=0.4, self=Get WBP_TextChat_MultilineInputField())
  ]}
 SetMessageContextMenuActionContext {nodes [
   SetMessageContextMenuActionContext() {is=title K2Node_FunctionEntry}
   n1 = Get Message_ContextMenuActionContext()
   > RemoveAllInteractions(self=@n1)
   > AddInteraction(
    self=@n1
    InActionInfo_DisplayName="LOCTEXT(\"Read aloud\")"
    InActionInfo_IconTexture=/FortniteCoreUI/UI/Icons/Icon-Settings/T-Icon-Audio-64
    InActionInfo_SortPriority=1
    InActionInfo_bIsAvailable=Get is TextToSpeech Enabled() -> (is TextToSpeech Enabled:bool)
    InActionInfo_bIsEnabled=True
    InActionInfo_bIsSelected=False
    InActionInfo_bAddDivider=False
    InActionInfo_bPositionDividerAfter=False
    InActionInfo_bCloseMenuOnSelect=False
    InActionInfo_InteractionNameID=SpeakMessage
    InActionInfo_Interaction=(DataTable=None,RowName="")
    InActionInfo_PressedDelegate="(null).None"
    InActionInfo_ReleasedDelegate="(null).None"
    InInteractionDelegate=Create Event()
   )
   > AddInteraction(
    self=@n1
    InActionInfo_DisplayName="LOCTEXT(\"Read aloud from here\")"
    InActionInfo_IconTexture=&T-Icon-Audio-64
    InActionInfo_SortPriority=2
    InActionInfo_bIsAvailable=Get is TextToSpeech Enabled()
    InActionInfo_bIsEnabled=True
    InActionInfo_bIsSelected=False
    InActionInfo_bAddDivider=true
    InActionInfo_bPositionDividerAfter=true
    InActionInfo_bCloseMenuOnSelect=False
    InActionInfo_InteractionNameID=SpeakMessagesFromHere
    InActionInfo_Interaction=(DataTable=None,RowName="")
    InActionInfo_PressedDelegate="(null).None"
    InActionInfo_ReleasedDelegate="(null).None"
    InInteractionDelegate=Create Event()
   )
   > AddInteraction(
    self=@n1
    InActionInfo_DisplayName="LOCTEXT(\"Mention\")"
    InActionInfo_SortPriority=1000
    InActionInfo_bIsAvailable=false
    InActionInfo_bIsEnabled=True
    InActionInfo_bIsSelected=False
    InActionInfo_bAddDivider=False
    InActionInfo_bPositionDividerAfter=False
    InActionInfo_bCloseMenuOnSelect=False
    InActionInfo_InteractionNameID=Mention
    InActionInfo_Interaction=(DataTable=None,RowName="")
    InActionInfo_PressedDelegate="(null).None"
    InActionInfo_ReleasedDelegate="(null).None"
    InInteractionDelegate=Create Event()
   )
   n2 = GetMessageContextMenuActions!(
    self:Fort Text Chat Main VM Object Reference=Get FortTextChatPlayerVM()
    MessageId:String="To String (Name)"(InName=Get MessageId() -> (MessageId=None))
   ) -> (ReturnValue:Array of EFortTextChatMessageContextMenuAction Enums=Mention)
   > AddInteraction(
    self=@n1
    InActionInfo_DisplayName="LOCTEXT(\"View Profile\")"
    InActionInfo_SortPriority=1000
    InActionInfo_bIsAvailable=Contains Item(
     TargetArray:Array of EFortTextChatMessageContextMenuAction Enums=@n2
     ItemToFind:"EFortTextChatMessageContextMenuAction Enum (by ref)"=ViewProfile
    )
    InActionInfo_bIsEnabled=True
    InActionInfo_bIsSelected=False
    InActionInfo_bAddDivider=False
    InActionInfo_bPositionDividerAfter=False
    InActionInfo_bCloseMenuOnSelect=False
    InActionInfo_InteractionNameID=ViewProfile
    InActionInfo_Interaction=(DataTable=None,RowName="")
    InActionInfo_PressedDelegate="(null).None"
    InActionInfo_ReleasedDelegate="(null).None"
    InInteractionDelegate=Create Event()
   )
   > AddInteraction(
    self=@n1
    InActionInfo_DisplayName="LOCTEXT(\"Block\")"
    InActionInfo_SortPriority=1000
    InActionInfo_bIsAvailable=Contains Item(TargetArray=@n2, ItemToFind=Block)
    InActionInfo_bIsEnabled=True
    InActionInfo_bIsSelected=False
    InActionInfo_bAddDivider=False
    InActionInfo_bPositionDividerAfter=False
    InActionInfo_bCloseMenuOnSelect=False
    InActionInfo_InteractionNameID=Block
    InActionInfo_Interaction=(DataTable=None,RowName="")
    InActionInfo_PressedDelegate="(null).None"
    InActionInfo_ReleasedDelegate="(null).None"
    InInteractionDelegate=Create Event()
   )
   > AddInteraction(
    self=@n1
    InActionInfo_DisplayName="LOCTEXT(\"Unblock\")"
    InActionInfo_SortPriority=1000
    InActionInfo_bIsAvailable=Contains Item(TargetArray=@n2, ItemToFind=Unblock)
    InActionInfo_bIsEnabled=True
    InActionInfo_bIsSelected=False
    InActionInfo_bAddDivider=False
    InActionInfo_bPositionDividerAfter=False
    InActionInfo_bCloseMenuOnSelect=False
    InActionInfo_InteractionNameID=Unblock
    InActionInfo_Interaction=(DataTable=None,RowName="")
    InActionInfo_PressedDelegate="(null).None"
    InActionInfo_ReleasedDelegate="(null).None"
    InInteractionDelegate=Create Event()
   )
   n3 = IsMuteLimitedToGameChannel!(self:Fort Text Chat Main VM Object Reference=Get FortTextChatPlayerVM()) -> (ReturnValue:bool)
   > AddInteraction(
    self=@n1
    InActionInfo_DisplayName=Select(Option 0:Text="LOCTEXT(\"Mute Player\")", Option 1:Text="LOCTEXT(\"Mute Text\")", Index=@n3) -> (ReturnValue:Text)
    InActionInfo_SortPriority=1000
    InActionInfo_bIsAvailable=Contains Item(TargetArray=@n2, ItemToFind=Mute)
    InActionInfo_bIsEnabled=True
    InActionInfo_bIsSelected=False
    InActionInfo_bAddDivider=False
    InActionInfo_bPositionDividerAfter=False
    InActionInfo_bCloseMenuOnSelect=False
    InActionInfo_InteractionNameID=MutePlayer
    InActionInfo_Interaction=(DataTable=None,RowName="")
    InActionInfo_PressedDelegate="(null).None"
    InActionInfo_ReleasedDelegate="(null).None"
    InInteractionDelegate=Create Event()
   )
   > AddInteraction(
    self=@n1
    InActionInfo_DisplayName=Select(Option 0="LOCTEXT(\"Unmute Player\")", Option 1="LOCTEXT(\"Unmute Text\")", Index=@n3)
    InActionInfo_SortPriority=1000
    InActionInfo_bIsAvailable=Contains Item(TargetArray=@n2, ItemToFind=Unmute)
    InActionInfo_bIsEnabled=True
    InActionInfo_bIsSelected=False
    InActionInfo_bAddDivider=False
    InActionInfo_bPositionDividerAfter=False
    InActionInfo_bCloseMenuOnSelect=False
    InActionInfo_InteractionNameID=UnmutePlayer
    InActionInfo_Interaction=(DataTable=None,RowName="")
    InActionInfo_PressedDelegate="(null).None"
    InActionInfo_ReleasedDelegate="(null).None"
    InInteractionDelegate=Create Event()
   )
   Contains Item(TargetArray=@n2, ItemToFind=Mention)
   AddInteraction(
    self=@n1
    InActionInfo_DisplayName="LOCTEXT(\"Report\")"
    InActionInfo_SortPriority=1000
    InActionInfo_bIsAvailable=Contains Item(TargetArray=@n2, ItemToFind=Report)
    InActionInfo_bIsEnabled=True
    InActionInfo_bIsSelected=False
    InActionInfo_bAddDivider=False
    InActionInfo_bPositionDividerAfter=False
    InActionInfo_bCloseMenuOnSelect=False
    InActionInfo_InteractionNameID=Report
    InActionInfo_Interaction=(DataTable=None,RowName="")
    InActionInfo_PressedDelegate="(null).None"
    InActionInfo_ReleasedDelegate="(null).None"
    InInteractionDelegate=Create Event()
   )
  ]}
 SetMobileLayout {nodes [
   SetMobileLayout() {is=title K2Node_FunctionEntry}
   > Branch(Condition=IsUsingCompactUI())
   > n1 = Sequence!(execute) -> (then_0, then_1, then_2:Exec)
   Set PresenceIconSize(execute=@n1.then_2, PresenceIconSize:double=18) -> (Output_Get:double)
   n2 = Break Slate Font Info!(SlateFontInfo:"Slate Font Info Structure (by ref)"=Get Whisper_ContextMenu_AdditionalContent_FontStyle()) -> (
    FontObject:Object Reference
    FontMaterial:Object Reference
    OutlineSettings:Font Outline Settings Structure
    TypefaceFontName:Name
    Size:float=~
    LetterSpacing:int
    SkewAmount:float=~
    bForceMonospaced:bool=~
    bMaterialIsStencil:bool=~
    MonospacedWidth:float=~
   ) {is=title K2Node_BreakStruct}
   > Set Whisper_ContextMenu_AdditionalContent_FontStyle(Whisper_ContextMenu_AdditionalContent_FontStyle:Slate Font Info Structure=MakeSlateFontInfo!(
     FontObject:Object Reference=@n2
     FontMaterial:Object Reference=@n2.FontMaterial
     OutlineSettings:Font Outline Settings Structure=@n2.OutlineSettings
     TypefaceFontName:Name=@n2.TypefaceFontName
     Size:float=18
     LetterSpacing:int=@n2.LetterSpacing
     SkewAmount:float=@n2.SkewAmount
     bForceMonospaced:bool=@n2.bForceMonospaced
     bMaterialIsStencil:bool=@n2.bMaterialIsStencil
     MonospacedWidth:float=@n2.MonospacedWidth
    ) -> (ReturnValue:Slate Font Info Structure)) -> (Output_Get:Slate Font Info Structure)
   > Set MenuPlacement!(
    execute
    MenuPlacement:EMenuPlacement Enum=MenuPlacement_AboveRightAnchor
    self:WBP UIKit Context Menu Anchor Button Base Object Reference=Get Chat_ContextMenu()
   ) -> (then, Output_Get:EMenuPlacement Enum=MenuPlacement_BelowAnchor)
   > Set WhisperContextMenuMaxHeight(WhisperContextMenuMaxHeight:float=600) -> (Output_Get:float)
   > Set!(
    execute
    WhisperHeadersMargin_Left:float=24
    WhisperHeadersMargin_Top:float=24
    WhisperHeadersMargin_Right:float=24
    WhisperHeadersMargin_Bottom:float=12
   ) -> (then, Output_Get:Margin Structure)
   SetRowFill!(execute=@n1.then_0, self:Grid Panel Object Reference=Get GP_TextChat_Wrapper(), RowIndex:int=1, Coefficient:float=1) -> (then)
   n3 = Get Mobile_SafeZone()
   > SetSidesToPad!(execute, self:Safe Zone Object Reference=@n3, InPadLeft:bool=true, InPadRight:bool=true, InPadTop:bool=true, InPadBottom:bool=true) -> (then)
   > SetPadding(self=SlotAsBorderSlot(Widget=@n3), InPadding=Get HiddenVirtualKeyboardSafeZoneMargin())
   > ClearHeightOverride(self:Size Box Object Reference=Get SB_ChatHeight() -> (SB_ChatHeight:Size Box Object Reference))
   n4 = GetOwningPlayer()
   Branch(
    execute=@n1.then_1
    Condition=NAND Boolean(
     A=Get isInZone() -> (isInZone:bool)
     B=IsUsingInputType!(InputType:ECommonInputType Enum=Gamepad, PlayerController:Player Controller Object Reference=@n4) -> (ReturnValue:bool)
    ) {is=title K2Node_CommutativeAssociativeBinaryOperator}
   )
   n5 = Get WBP_UIKit_TabGroup()
   > Set IconSize!(execute, IconSize:Vector 2D Structure=(X=48,Y=48), self:WBP UIKit Tab Group Base Object Reference=@n5) -> (then, Output_Get:Vector 2D Structure)
   > Set!(execute, MinButtonSize_X:int=80, MinButtonSize_Y:int=80, self:WBP UIKit Tab Group Base Object Reference=@n5) -> (then, Output_Get:Int Point Structure)
   > SetMinDimensions!(
    execute
    self:Common Button Base Object Reference=[Get WBP_TextChat_Input(), Get WBP_TextChat_MultilineInputField()]
    InMinWidth:int=80
    InMinHeight:int=80
   ) -> (then)
   > Set DictateDefaultIconSize(DictateDefaultIconSize:Vector 2D Structure=(X=48,Y=48)) -> (Output_Get:Vector 2D Structure)
   > Set DictateStopIconSize(DictateStopIconSize:Vector 2D Structure=(X=32,Y=32)) -> (Output_Get:Vector 2D Structure)
   > n6 = Set!(
    execute
    MobileButtonPadding_Left:float=16
    MobileButtonPadding_Top:float=0
    MobileButtonPadding_Right:float=16
    MobileButtonPadding_Bottom:float=0
    self:WBP UIKit Button Generic Base Object Reference=Get Send_Dictation()
   ) -> (then, Output_Get:Margin Structure)
   n7 = Get Send_ButtonIcon()
   > Set MobileButtonPadding!(execute, MobileButtonPadding:Margin Structure=@n6, self:WBP UIKit Button Generic Base Object Reference=@n7) -> (then, Output_Get:Margin Structure)
   > SetIconSize(self=@n7, Icon=(X=48,Y=48))
  ]}
 SetSendButtonHeight {nodes [
   SetSendButtonHeight() {is=title K2Node_FunctionEntry}
   n1 = Get Send_ButtonIcon()
   > SetMinDimensions(
    self=@n1
    InMinWidth=Get MinHeight!(self:Common Button Base Object Reference=@n1) -> (MinHeight:int=0)
    InMinHeight=Get MinHeight(self=Get WBP_TextChat_Input()) -> (MinHeight=0)
   )
  ]}
 SetWhisperContextMenuActionContext {nodes [
   SetWhisperContextMenuActionContext() {is=title K2Node_FunctionEntry}
   > n1 = Set WhisperDestinations(WhisperDestinations:Array of Fort Text Chat Whisper Destination VM Object References=GetWhisperDestinations!(self:Fort Text Chat Main VM Object Reference=Get FortTextChatPlayerVM()) -> (ReturnValue:Array of Fort Text Chat Whisper Destination VM Object References)) -> (Output_Get:Array of Fort Text Chat Whisper Destination VM Object References)
   n2 = Get InteractionContext!(self:WBP UIKit Context Menu Anchor Button Base Object Reference=Get Whisper_ContextMenu()) -> (InteractionContext:UIKit Context Menu Action Context Object Reference)
   > RemoveAllInteractions(self=@n2)
   n3 = Is Not Empty(TargetArray:Array of Fort Text Chat Whisper Destination VM Object References=@n1)
   > Set bIsWhisperListEmpty(bIsWhisperListEmpty:bool=NOT Boolean(A=@n3)) -> (Output_Get:bool)
   > n4 = Branch(Condition=@n3)
   > n5 = For Each Loop(Array:Array of Fort Text Chat Whisper Destination VM Object References=@n1) -> (Array Element:Fort Text Chat Whisper Destination VM Object Reference)
   Branch(execute=@n5.Completed, Condition="integer > integer"(A=@n5.Array Index, B=3))
   > SetMaxDesiredHeight!(
    execute
    self:Size Box Object Reference=Get SB_MaxWidth(self=Get WBP UIKit Context Menu_Ref() -> (WBP UIKit Context Menu_Ref:WBP UIKit Context Menu Object Reference))
    InMaxDesiredHeight:float=Get WhisperContextMenuMaxHeight() -> (WhisperContextMenuMaxHeight:float)
   ) -> (then)
   AddInteraction(
    execute=@n5.LoopBody
    self=@n2
    InActionInfo_DisplayName="To Text (String)"(InString=Get WhisperDestinationDisplayName!(self:Fort Text Chat Whisper Destination VM Object Reference=@n5) -> (WhisperDestinationDisplayName:String))
    InActionInfo_SortPriority=1000
    InActionInfo_bIsAvailable=True
    InActionInfo_bIsEnabled=True
    InActionInfo_bIsSelected=False
    InActionInfo_bAddDivider=False
    InActionInfo_bPositionDividerAfter=False
    InActionInfo_bCloseMenuOnSelect=False
    InActionInfo_InteractionNameID=Get WhisperDestinationId!(self:Fort Text Chat Whisper Destination VM Object Reference=@n5) -> (WhisperDestinationId:Name=None)
    InActionInfo_Interaction=(DataTable=None,RowName="")
    InActionInfo_PressedDelegate="(null).None"
    InActionInfo_ReleasedDelegate="(null).None"
    InInteractionDelegate=Create Event()
   )
   AddInteraction(
    execute=@n4.else
    self=Get InteractionContext(self=Get Whisper_ContextMenu())
    InActionInfo_DisplayName="LOCTEXT(\"No friends available\")"
    InActionInfo_SortPriority=1000
    InActionInfo_bIsAvailable=True
    InActionInfo_bIsEnabled=True
    InActionInfo_bIsSelected=False
    InActionInfo_bAddDivider=False
    InActionInfo_bPositionDividerAfter=False
    InActionInfo_bCloseMenuOnSelect=False
    InActionInfo_InteractionNameID=No friends available
    InActionInfo_Interaction=(DataTable=None,RowName="")
    InActionInfo_PressedDelegate="(null).None"
    InActionInfo_ReleasedDelegate="(null).None"
    InInteractionDelegate=Create Event()
   )
  ]}
 SetWhisperItemVisual {nodes [
   n1 = SetWhisperItemVisual!() -> (then, Object:Object Reference) {is=title K2Node_FunctionEntry}
   > Branch(Condition=NOT Boolean(A=Get bIsWhisperListEmpty() -> (bIsWhisperListEmpty:bool)))
   > n2 = Cast To WBP_UIKit_ContextMenuListEntry(Object=@n1)
   n3 = GetActionInfo!(self:UIKit Context Menu Action Wrapper Object Reference=Get Action!(self:WBP UIKit Context Menu List Entry Object Reference=@n2) -> (Action:UIKit Context Menu Action Wrapper Object Reference)) -> (ReturnValue:"UIKit Context Menu Action Info Structure (by ref)")
   > n4 = Branch(Condition="Not Equal (Name)"(A=Break UIKit Context Menu Action Info!(UIKitContextMenuActionInfo:"UIKit Context Menu Action Info Structure (by ref)"=@n3) -> (
      DisplayName:Text
      IconTexture:Texture 2D Soft Object Reference
      IconMaterial:Material Instance Soft Object Reference
      SortPriority:int
      bIsAvailable:bool=~
      bIsEnabled:bool=~
      bIsSelected:bool=~
      bAddDivider:bool=~
      bPositionDividerAfter:bool=~
      bCloseMenuOnSelect:bool=~
      InteractionNameID:Name
      Interaction:Data Table Row Handle Structure
      PressedDelegate:Delegate
      ReleasedDelegate:Delegate
     ) {is=title K2Node_BreakStruct}.InteractionNameID) -> (ReturnValue=~))
   > SetIconMaterial(self=@n2, Material Instance=/FortUILibrary/Systems/StatusIndicator/MaterialInstances/MI_UIKit_StatusIndicator_Background)
   n5 = Get PresenceIconSize() -> (PresenceIconSize:double)
   > SetIconSize(self=@n2, Icon=MakeVector2D(X=@n5, Y=@n5))
   > n6 = SetIconPadding(self=@n2, Padding_Left=0, Padding_Top=0, Padding_Right=8, Padding_Bottom=0)
   n7 = Get!(self:WBP UIKit Button Generic Base Object Reference=@n2) -> (
    Font_FontObject:Object Reference
    Font_FontMaterial:Object Reference
    Font_OutlineSettings:Font Outline Settings Structure
    Font_TypefaceFontName:Name
    Font_Size:float=~
    Font_LetterSpacing:int
    Font_SkewAmount:float=~
    Font_bForceMonospaced:bool=~
    Font_bMaterialIsStencil:bool=~
    Font_MonospacedWidth:float=~
   )
   SetFont!(
    execute=[@n6.then, @n4.else]
    self:WBP UIKit Button Generic Base Object Reference=@n2
    In Font:Slate Font Info Structure=MakeSlateFontInfo(
     FontObject=@n7
     FontMaterial=@n7.Font_FontMaterial
     OutlineSettings=@n7.Font_OutlineSettings
     TypefaceFontName=@n7.Font_TypefaceFontName
     Size=15.75
     LetterSpacing=@n7.Font_LetterSpacing
     SkewAmount=@n7.Font_SkewAmount
     bForceMonospaced=@n7.Font_bForceMonospaced
     bMaterialIsStencil=@n7.Font_bMaterialIsStencil
     MonospacedWidth=@n7.Font_MonospacedWidth
    )
   ) -> (then)
  ]}
 SpeakMessage {nodes [
   SpeakMessage!() -> (then, InteractionID:Name) {is=title K2Node_FunctionEntry}
   > Cancel Speech To Text Recording() {is=title K2Node_CallFunction}
   > SpeakMessage!(
    execute
    self:Fort Text Chat Main VM Object Reference=Get FortTextChatPlayerVM()
    MessageId:"Name (by ref)"=Get MessageId() -> (MessageId=None)
   ) -> (then) {is=title K2Node_CallFunction}
  ]}
 SpeakMessagesFromHere {nodes [
   SpeakMessagesFromHere!() -> (then, InteractionID:Name) {is=title K2Node_FunctionEntry}
   > Cancel Speech To Text Recording()
   > SpeakMessagesInclusiveAfter!(
    execute
    self:Fort Text Chat Main VM Object Reference=Get FortTextChatPlayerVM()
    MessageId:"Name (by ref)"=Get MessageId() -> (MessageId=None)
   ) -> (then)
  ]}
 Unblock {nodes [
   Unblock!() -> (then, InteractionNameID:Name) {is=title K2Node_FunctionEntry}
   > PerformMessageContextMenuAction(self=Get FortTextChatPlayerVM(), MessageId="To String (Name)"(InName=Get MessageId() -> (MessageId=None)), MenuAction=Unblock)
  ]}
 UnmutePlayer {nodes [
   UnmutePlayer!() -> (then, InteractionID:Name) {is=title K2Node_FunctionEntry}
   > PerformMessageContextMenuAction(self=Get FortTextChatPlayerVM(), MessageId="To String (Name)"(InName=Get MessageId() -> (MessageId=None)), MenuAction=Unmute)
  ]}
 UpdateInputFieldPadding {nodes [
   UpdateInputFieldPadding() {is=title K2Node_FunctionEntry}
   n1 = Get Multiline InputField Enabled()
   n2 = Get Block_EditableText(self:WBP Text Chat Multiline Input Field Object Reference=Get WBP_TextChat_MultilineInputField()) -> (Block_EditableText:WBP Text Chat Block Multiline Editable Text Object Reference)
   n3 = Get Block_EditableText(self:WBP UIKit Input Field Base Object Reference=Get WBP_TextChat_Input()) -> (Block_EditableText:WBP UIKit Block Editable Text Object Reference)
   n4 = Get CommonInputSubsystem()
   > SetPadding!(
    execute
    self:Grid Slot Object Reference=SlotAsGridSlot(Widget=Select(Option 0:Widget Object Reference=@n3, Option 1:Widget Object Reference=@n2, Index=@n1) -> (ReturnValue:Widget Object Reference))
    InPadding_Left:float=0
    InPadding_Top:float=0
    InPadding_Right:float=Select!(MouseAndKeyboard:float=0, Gamepad:float=20, Touch:float=0, Count:float=0, Index:ECommonInputType Enum=GetCurrentInputType(self=@n4)) -> (ReturnValue:float)
    InPadding_Bottom:float=0
   ) -> (then)
  ]}
 UpdateSendButtonEnabled {nodes [
   UpdateSendButtonEnabled() {is=title K2Node_FunctionEntry}
   n1 = Get Block_EditableText(self=Get WBP_TextChat_Input())
   n2 = TextIsEmpty(InText="GetText (Editable Text)"(self=Get EditableText(self=@n1)))
   n3 = Get WBP_TextChat_Input()
   n4 = Get Block_EditableText(self:WBP Text Chat Multiline Input Field Object Reference=Get WBP_TextChat_MultilineInputField()) -> (Block_EditableText:WBP Text Chat Block Multiline Editable Text Object Reference)
   n5 = TextIsEmpty(InText=Get Text!(self:"Text Box (Multi- Line) Object Reference"=Get EditableText(self:WBP Text Chat Block Multiline Editable Text Object Reference=@n4) -> (EditableText:"Text Box (Multi- Line) Object Reference")) -> (Text))
   n6 = Get WBP_TextChat_MultilineInputField()
   n7 = Select!(
    Option 0:bool=AND Boolean(A=NOT Boolean(A=@n2), B=Get bIsEnabled!(self:Widget Object Reference=@n3) -> (bIsEnabled:bool))
    Option 1:bool=AND Boolean(A=NOT Boolean(A=@n5), B=Get bIsEnabled(self=@n6))
    Index:bool=Get Multiline InputField Enabled()
   ) -> (ReturnValue:bool)
   > SetIsEnabled(self=Get Send_ButtonIcon(), bInIsEnabled=@n7)
   > n8 = Branch(Condition=Get is SpeechToText Enabled())
   > SetVisibility(
    self=Get Send_ButtonIcon()
    InVisibility=Select(Option 0:ESlateVisibility Enum=Collapsed, Option 1:ESlateVisibility Enum=Visible, Index=@n7) -> (ReturnValue:ESlateVisibility Enum)
   )
   > SetVisibility(self=Get Send_Dictation(), InVisibility=Select(Option 0=Visible, Option 1=Collapsed, Index=@n7))
   SetVisibility(execute=@n8.else, self=Get Send_Dictation(), InVisibility=Collapsed)
   n9 = Get CurrentInputType()
   > SetVisibility(
    self=Get Send_ButtonIcon()
    InVisibility=Select(Option 0=Collapsed, Option 1=Visible, Index="Not Equal (Enum)"(A:ECommonInputType Enum=@n9, B:ECommonInputType Enum=Gamepad))
   )
  ]}
 UpdateSendButtonVisibility {nodes [
   UpdateSendButtonVisibility() {is=title K2Node_FunctionEntry}
   n1 = "Not Equal (Enum)"(A=GetCurrentInputType(self=Get CommonInputSubsystem()), B=Gamepad)
   n2 = Get bEnableVMDrivenSidebar(self=GetRuntimeOptions())
   > SetVisibility(self=Get Send_ButtonIcon(), InVisibility=Select(Option 0=Collapsed, Option 1=Visible, Index=AND Boolean(A=@n2, B=@n1)))
  ]}
 UpdateTabsData {nodes [
   n1 = UpdateTabsData!() -> (then, IsTabActive:bool=~) {is=title K2Node_FunctionEntry}
   > n2 = Branch(Condition=@n1)
   SetText(execute=@n2.else, self:WBP UIKit Input Field Base Object Reference=Get WBP_TextChat_Input())
   > SetText(self:WBP Text Chat Multiline Input Field Object Reference=Get WBP_TextChat_MultilineInputField())
  ]}
 ViewProfile {nodes [
   ViewProfile!() -> (then, InteractionNameID:Name) {is=title K2Node_FunctionEntry}
   n1 = Get FortTextChatPlayerVM()
   > Set OpenProfileActionMenuClass!(
    execute
    OpenProfileActionMenuClass:Social Action Menu Class Reference=/Game/UI/Sidebar/SidebarProfile.SidebarProfile_C
    self:Fort Text Chat Main VM Object Reference=@n1
   ) -> (then, Output_Get:Social Action Menu Class Reference)
   > PerformMessageContextMenuAction(self=@n1, MessageId="To String (Name)"(InName=Get MessageId() -> (MessageId=None)), MenuAction=ViewProfile)
   > Call On View Profile Requested!(execute:Exec, self:WBP Text Chat Object Reference) -> (then:Exec) {is=title K2Node_CallDelegate}
  ]}
 isInputFieldNotEmpty {nodes [
   isInputFieldNotEmpty() {is=title K2Node_FunctionEntry}
   n1 = Get Block_EditableText(self=Get WBP_TextChat_MultilineInputField())
   n2 = TextIsEmpty(InText="GetText (Multi-Line Text Box)"!(self:"Text Box (Multi- Line) Object Reference"=Get EditableText(self=@n1)) -> (ReturnValue:Text))
   n3 = Get Block_EditableText(self:WBP UIKit Input Field Base Object Reference=Get WBP_TextChat_Input()) -> (Block_EditableText:WBP UIKit Block Editable Text Object Reference)
   n4 = TextIsEmpty(InText="GetText (Editable Text)"(self=Get EditableText(self:WBP UIKit Block Editable Text Object Reference=@n3) -> (EditableText:UIKit Editable Text Object Reference)))
   > Return Node!(execute, NotEmpty:bool=OR Boolean(A=NOT Boolean(A=@n4), B=NOT Boolean(A=@n2)))
  ]}
]
variables [
 bIsDMButtonAvailable {type=boolean}
 bIsWhisperListEmpty {type=boolean}
 bSystemMessageClicked {type=boolean}
 Chat_ContextMenuActionContext {type=UIKitContextMenuActionContext, default=null}
 Current Hint {type=string, default=""}
 CurrentInputType {type=ECommonInputType, default=MouseAndKeyboard}
 DictateDefaultIconSize {type=Vector2D, default=<32 32>}
 DictateStopIconSize {type=Vector2D, default=<28 28>}
 HiddenVirtualKeyboardSafeZoneMargin {type=Margin, default=<ltrb 24 24 24 124>}
 Hint Text Listening {type=string, default=Listening...}
 Hint Text Please Wait {type=string, default=Please wait...}
 Image_ChatContextMenu {type=Image, default=null}
 InitialGPContentGridRow {type=integer}
 is SpeechToText Enabled {type=boolean}
 is TextToSpeech Enabled {type=boolean}
 isInZone {type=boolean}
 IsPersistentChatOnInSidebar {type=boolean}
 Message_ContextMenuActionContext {type=UIKitContextMenuActionContext, default=null}
 MessageId {type=string, default=None}
 MessageReceivedSound {type=SoundBase, default=/Game/Sounds/Fort_UI_Sounds/Chat_UI/MSS/MSS_UI_FrontEnd_Chat_MessageReceived}
 MessageSentSound {type=SoundBase, default=/Game/Sounds/Fort_UI_Sounds/Chat_UI/MSS/MSS_UI_FrontEnd_Chat_MessageSent}
 OnActiveStateChanged {}
 OnButtonCloseClicked {}
 OnViewProfileRequested {}
 OpenVirtualKeyboardSafeZoneMargin {type=Margin, default=<ltrb 24 0 24 8>}
 PresenceIconSize {type=number, default=12}
 Record Time {type=number, default=10}
 RecordingColorOff {type=LinearColor, default=<rgba 0.5776 0.5776 0.5776 0.5>}
 RecordingColorOn {type=LinearColor, default=<rgba 0.0395 1 0.2307 0.502>}
 TabAddCmd {type=string, default=_AddCmdId}
 Text_ChatContextMenu {type=CommonTextBlock, default=null}
 TextChatClosedSound {type=SoundBase, default=/Game/Sounds/Fort_UI_Sounds/Chat_UI/MSS/MSS_UI_FrontEnd_Chat_Close}
 TextChatOpenSound {type=SoundBase, default=/Game/Sounds/Fort_UI_Sounds/Chat_UI/MSS/MSS_UI_FrontEnd_Chat_Open}
 TextInputSound {type=SoundBase, default=/Game/Sounds/Fort_UI_Sounds/Chat_UI/MSS/MSS_UI_FrontEnd_Chat_Typing}
 Timer Handle {type=TimerHandle}
 Username {type=string, default=""}
 WBP Text Chat Entry User Message {type=WBP_TextChat_Entry_UserMessage_C, default=null}
 WBP Text Chat Message System {type=WBP_TextChat_MessageSystem_C, default=null}
 WBP UIKit Context Menu_Ref {type=WBP_UIKit_ContextMenu_C, default=null}
 WelcomeMessageEntry {type=BP_VM_TextChat_Welcome_Entry_C, default=null}
 Whisper_ContextMenu_AdditionalContent_FontStyle {
  type=SlateFontInfo
  default {
   bForceMonospaced=false
   bMaterialIsStencil=false
   fontMaterial=null
   fontObject=&HeadingNow
   letterSpacing=0
   monospacedWidth=1
   size=13.5
   skewAmount=0
   typefaceFontName=75 Medium
   outlineSettings {
    bApplyOutlineToDropShadows=false
    bMiteredCorners=false
    bSeparateFillAlpha=false
    outlineMaterial=null
    outlineSize=0
    outlineColor=<rgba 0 0 0 1>
   }
  }
 }
 Whisper_ContextMenuActionContext {type=UIKitContextMenuActionContext, default=null}
 WhisperContextMenuMaxHeight {type=number, default=275}
 WhisperHeadersMargin {type=Margin, default=<ltrb 24 24 32 12>}
]
widgets {
 info {inheritedWidgetCount=0, namedSlotCount=3, parentClass=/Script/FortniteUI.FortTextChatV3, rootWidgetClass=&Border, widgetCount=41}
 widgets [
  Border_Background {
   widgetClassPath=&Border
   bIsVariable=true
   widget=Border_Background
   properties {brushColor=<rgba 0 0 0 0>, padding=<ltrb 0 0 0 0>}
  }
  Mobile_SafeZone {
   widgetClassPath=&SafeZone
   bIsVariable=true
   parent=Border_Background
   widget=Mobile_SafeZone
   properties {padBottom=false, padLeft=false, padRight=false, padTop=false}
   slot {class=BorderSlot, properties {padding=<ltrb 0 0 0 0>}}
  }
  SB_ChatWidth {
   widgetClassPath=&SizeBox
   parent=Mobile_SafeZone
   widget=SB_ChatWidth
   properties {bOverride_WidthOverride=true, widthOverride=590}
   slot {class=SafeZoneSlot}
  }
  GP_TextChat_Wrapper {
   widgetClassPath=&GridPanel
   bIsVariable=true
   parent=SB_ChatWidth
   widget=GP_TextChat_Wrapper
   properties {flowDirectionPreference=Culture, columnFill [1], rowFill [0,0,0]}
   slot {class=SizeBoxSlot}
  }
  GP_TabGroup {
   widgetClassPath=&GridPanel
   parent=GP_TextChat_Wrapper
   widget=GP_TabGroup
   properties {columnFill [1], rowFill [1]}
   slot {class=GridSlot, properties {horizontalAlignment=HAlign_Left, verticalAlignment=VAlign_Top, padding=<ltrb 0 0 0 8>}}
  }
  WBP_UIKit_TabGroup {
   widgetClassPath=/FortUILibrary/UIKit/Tabs/WBP_UIKit_TabGroup
   bIsVariable=true
   parent=GP_TabGroup
   widget=WBP_UIKit_TabGroup
   properties {
    arrowButtonsAllowWrap=false
    arrowButtonsClass=&WBP_UIKit_Button_TabNavArrow_TextChat_C
    bIsVolatile=true
    backplateBrightness=Dark
    backplateClass=&WBP_UIKit_Backplate_Fallback_C
    backplateCornerRadius=Tab
    buttonClass=&/WBP_TextChat_Button_Tab.WBP_TextChat_Button_Tab_C
    nextPromptHorizontalAlignment=HAlign_Right
    prevPromptHorizontalAlignment=HAlign_Left
    promptsBackplateClass=&WBP_UIKit_Backplate_Fallback_C
    reservedPromptSpace=32
    tabSpacing=4
    useArrowNavButtons=true
    iconSize=<24 24>
    initialTabsData [
     {bEnableTabId=false, tabId=None, tabViewModel=null}
     {bEnableTabId=false, tabId=None, tabViewModel=null}
     {bEnableTabId=false, tabId=None, tabViewModel=null}
     {bEnableTabId=false, tabId=None, tabViewModel=null}
     {bEnableTabId=false, tabId=None, tabViewModel=null}
     {bEnableTabId=false, tabId=None, tabViewModel=null}
     {bEnableTabId=false, tabId=None, tabViewModel=null}
    ]
    minButtonSize=<0 40>
    nextPromptPadding=<ltrb 16 0 0 0>
    prevPromptPadding=<ltrb 0 0 16 0>
    tabGroupPadding=<ltrb 4 4 4 4>
   }
   slot {class=GridSlot}
  }
  Whisper_ContextMenu {
   widgetClassPath=/FortUILibrary/UIKit/ContextMenu/WBP_UIKit_ContextMenuAnchorButton
   bIsVariable=true
   parent=GP_TabGroup
   widget=Whisper_ContextMenu
   properties {
    button Background Material=null
    button Outline Material=null
    buttonClass=/Game/UI/Frontend/Lobby/Nameplates/WBP_EmptyButton_ContextMenu.WBP_EmptyButton_ContextMenu_C
    buttonText=""
    closeMenuOnContextClicked=true
    isMenuModal=false
    menuPlacement=MenuPlacement_BelowAnchor
    min Height=1
    button Padding=<ltrb 0 0 0 0>
   }
   slot {class=GridSlot, properties {horizontalAlignment=HAlign_Left, verticalAlignment=VAlign_Bottom, nudge=<0 17>, padding=<ltrb 32 0 0 0>}}
  }
  SB_ChatHeight {
   widgetClassPath=&SizeBox
   bIsVariable=true
   parent=GP_TextChat_Wrapper
   widget=SB_ChatHeight
   properties {bOverride_HeightOverride=true, heightOverride=315, widthOverride=590}
   slot {class=GridSlot, properties {row=1}}
  }
  GP_TextChat_Container {
   widgetClassPath=&GridPanel
   parent=SB_ChatHeight
   widget=GP_TextChat_Container
   properties {clipping=ClipToBounds, columnFill [1], rowFill [1,0,0]}
   slot {class=SizeBoxSlot}
  }
  Backplate_ClickCatcher {
   widgetClassPath=&WBP_UIKit_Backplate
   parent=GP_TextChat_Container
   widget=Backplate_ClickCatcher
   properties {
    backplateMaterial=/FortUILibrary/Systems/Backplate/MaterialInstances/MI_UIKit_Backplate_Fallback
    brightness=Obscure
    cornerRadius=Interactive Component
    visibility=Visible
   }
   slot {class=GridSlot, properties {layer=-10, rowSpan=3}}
  }
  Switcher_Content {
   widgetClassPath=&CommonVisibilitySwitcher
   bIsVariable=true
   parent=GP_TextChat_Container
   widget=Switcher_Content
   slot {class=GridSlot}
  }
  WBP_TextChat_BlankMessage {
   widgetClassPath=&/WBP_TextChat_BlankMessage
   bIsVariable=true
   parent=Switcher_Content
   widget=WBP_TextChat_BlankMessage
   slot {class=CommonVisibilitySwitcherSlot, properties {verticalAlignment=VAlign_Bottom, padding=<ltrb 20 0 20 20>}}
  }
  O_ClippingBox {
   widgetClassPath=&Overlay
   bIsVariable=true
   parent=Switcher_Content
   widget=O_ClippingBox
   properties {clipping=ClipToBounds}
   slot {class=CommonVisibilitySwitcherSlot}
  }
  WrapperNavigation {
   widgetClassPath=&Overlay
   parent=O_ClippingBox
   widget=WrapperNavigation
   properties {navigation=&self:WidgetTree.WrapperNavigation.WidgetNavigation_0}
   slot {class=OverlaySlot, properties {horizontalAlignment=HAlign_Fill, verticalAlignment=VAlign_Fill, padding=<ltrb 12 12 12 12>}}
  }
  MessageList {
   widgetClassPath=&/WBP_TextChat_MessageList
   bIsVariable=true
   parent=WrapperNavigation
   widget=MessageList
   properties {bIsFocusable=true}
   slot {class=OverlaySlot, properties {horizontalAlignment=HAlign_Fill, verticalAlignment=VAlign_Bottom}}
  }
  Message_ContextMenu {
   widgetClassPath=&WBP_UIKit_ContextMenuAnchorButton
   bIsVariable=true
   parent=O_ClippingBox
   widget=Message_ContextMenu
   properties {
    button Background Material=null
    button Outline Material=null
    buttonClass=&WBP_EmptyButton_ContextMenu_C
    buttonIcon=null
    closeMenuOnContextClicked=true
    menuPlacement=MenuPlacement_AboveRightAnchor
    min Height=1
    button Padding=<ltrb 0 0 0 0>
   }
   slot {class=OverlaySlot, properties {horizontalAlignment=HAlign_Right, verticalAlignment=VAlign_Bottom}}
  }
  GP_BottomContent {
   widgetClassPath=&GridPanel
   bIsVariable=true
   parent=GP_TextChat_Container
   widget=GP_BottomContent
   properties {
    navigation=&self:WidgetTree.GP_BottomContent.WidgetNavigation_0
    columnFill [1,0,0]
    rowFill [1]
   }
   slot {class=GridSlot, properties {row=2, padding=<ltrb 12 12 12 12>}}
  }
  NavigationWrapper_InputText {
   widgetClassPath=&Overlay
   parent=GP_BottomContent
   widget=NavigationWrapper_InputText
   properties {navigation=&self:WidgetTree.NavigationWrapper_InputText.WidgetNavigation_0}
   slot {class=GridSlot}
  }
  WBP_TextChat_Input {
   widgetClassPath=&/WBP_TextChat_Input
   bIsVariable=true
   parent=NavigationWrapper_InputText
   widget=WBP_TextChat_Input
   properties {
    backgroundMaterial=/FortUILibrary/Systems/InputField/MaterialInstances/MI_UIKit_InputField_Background_Fallback
    icon=null
    maxTextCharacters=0
    minHeight=48
    useIcon=true
    font {
     bForceMonospaced=false
     bMaterialIsStencil=false
     fontMaterial=/FortUILibrary/Systems/InputField/MaterialInstances/MI_UIKit_InputFieldText
     fontObject=&HeadingNow
     letterSpacing=0
     monospacedWidth=1
     size=18
     skewAmount=0
     typefaceFontName=74 Regular
     outlineSettings {
      bApplyOutlineToDropShadows=false
      bMiteredCorners=false
      bSeparateFillAlpha=false
      outlineMaterial=null
      outlineSize=0
      outlineColor=<rgba 0 0 0 1>
     }
    }
    heightOverride {enable=true, value=48}
    iconProperties {
     material=/FortUILibrary/Textures/MaterialInstances/MI_UIKit_Icon_Like
     image Size=<50 50>
     placement {
      horizontal Alignment=HAlign_Center
      vertical Alignment=VAlign_Center
      grid {column=0, column Span=0, layer=0, row=1, row Span=0, nudge=<0 0>}
      padding=<ltrb 0 0 0 0>
     }
    }
    textPadding=<ltrb 16 4 20 4>
   }
   slot {class=OverlaySlot, properties {horizontalAlignment=HAlign_Fill, verticalAlignment=VAlign_Fill}}
  }
  WBP_TextChat_MultilineInputField {
   widgetClassPath=/FortniteCoreUI/UI/Social/Chat/TextChat/InputField/WBP_TextChat_MultilineInputField
   bIsVariable=true
   parent=NavigationWrapper_InputText
   widget=WBP_TextChat_MultilineInputField
   properties {
    backgroundMaterial=&MI_UIKit_InputField_Background_Fallback
    holdData=&CommonUI_HoldData_Base_C
    maxHeight=80
    minHeight=48
    useIcon=false
    visibility=Collapsed
   }
   slot {class=OverlaySlot, properties {horizontalAlignment=HAlign_Fill, verticalAlignment=VAlign_Fill}}
  }
  NavigationWrapper_ContextMenu {
   widgetClassPath=&Overlay
   parent=GP_BottomContent
   widget=NavigationWrapper_ContextMenu
   properties {navigation=&self:WidgetTree.NavigationWrapper_ContextMenu.WidgetNavigation_0}
   slot {class=GridSlot, properties {column=2, padding=<ltrb 12 0 0 0>}}
  }
  Chat_ContextMenu {
   widgetClassPath=&/WBP_TextChat_ContextMenu_Button
   bIsVariable=true
   parent=NavigationWrapper_ContextMenu
   widget=Chat_ContextMenu
   properties {mobileMinHeight=80, mobileMinWidth=80}
   slot {class=OverlaySlot, properties {horizontalAlignment=HAlign_Fill, verticalAlignment=VAlign_Bottom}}
  }
  Send_ButtonIcon {
   widgetClassPath=&WBP_UIKit_Button_Generic
   bIsVariable=true
   parent=GP_BottomContent
   widget=Send_ButtonIcon
   properties {
    bHideInputAction=true
    backgroundMaterial=/FortUILibrary/UIKit/Buttons/MaterialInstances/CTAButtons/MI_UIKit_BgQuiet_Fallback
    icon=/FortUILibrary/Textures/Icons/T_UI_Icon_UIKit_SendChat_MSDF
    iconMaterial=/FortUILibrary/Textures/MaterialInstances/MI_UIKit_Icon_SendChat
    minHeight=48
    mobileMinHeight=80
    navigation=&self:WidgetTree.Send_ButtonIcon.WidgetNavigation_0
    useTextBlock=false
    buttonPadding=<ltrb 8 0 8 0>
    iconSize=<36 36>
   }
   slot {class=GridSlot, properties {column=1, horizontalAlignment=HAlign_Left, verticalAlignment=VAlign_Bottom, padding=<ltrb 12 0 0 0>}}
  }
  Send_Dictation {
   widgetClassPath=&WBP_UIKit_Button_Generic
   bIsVariable=true
   parent=GP_BottomContent
   widget=Send_Dictation
   properties {
    bHideInputAction=true
    backgroundMaterial=&MI_UIKit_BgQuiet_Fallback
    iconMaterial=&MI_UI_TextChat_Dictation
    minHeight=48
    minWidth=52
    mobileMinHeight=80
    navigation=&self:WidgetTree.Send_Dictation.WidgetNavigation_0
    text=""
    textJustification=Left
    useDisabledVisualsOnLocked=true
    visibility=Collapsed
    buttonPadding=<ltrb 8 0 8 0>
    font {
     bForceMonospaced=false
     bMaterialIsStencil=false
     fontMaterial=/FortUILibrary/UIKit/Buttons/MaterialInstances/CTAButtons/MI_UIKit_TextLoud
     fontObject=&HeadingNow
     letterSpacing=0
     monospacedWidth=1
     size=12
     skewAmount=0
     typefaceFontName=75 Medium
     outlineSettings {
      bApplyOutlineToDropShadows=false
      bMiteredCorners=false
      bSeparateFillAlpha=false
      outlineMaterial=/FortUILibrary/Systems/ModularButton/MaterialInstances/MI_UIKit_TextOutline
      outlineSize=0
      outlineColor=<rgba 0 0 0 1>
     }
    }
    iconSize=<36 36>
    textPadding=<ltrb 4 0 0 0>
   }
   slot {class=GridSlot, properties {column=1, horizontalAlignment=HAlign_Left, verticalAlignment=VAlign_Bottom, padding=<ltrb 12 0 0 0>}}
  }
  WBP_TextChat_InactiveMessage {
   widgetClassPath=&/WBP_TextChat_InactiveMessage
   bIsVariable=true
   parent=GP_TextChat_Container
   widget=WBP_TextChat_InactiveMessage
   properties {visibility=Collapsed}
   slot {class=GridSlot, properties {row=1, verticalAlignment=VAlign_Top, padding=<ltrb 18 0 18 0>}}
  }
  WBP_TextChat_Restrictions {
   widgetClassPath=&/WBP_TextChat_Restrictions
   bIsVariable=true
   parent=GP_TextChat_Container
   widget=WBP_TextChat_Restrictions
   properties {use Updated Visuals=true}
   slot {class=GridSlot, properties {row=1, padding=<ltrb 18 0 18 0>}}
  }
  NavigationWrapper_RestrictionButton {
   widgetClassPath=&Overlay
   namedSlotHost=WBP_TextChat_Restrictions
   widget=NavigationWrapper_RestrictionButton
   properties {navigation=&self:WidgetTree.NavigationWrapper_RestrictionButton.WidgetNavigation_0}
  }
  Button_CompactChat_RestrictionAction {
   widgetClassPath=&WBP_UIKit_Button_Generic
   bIsVariable=true
   parent=NavigationWrapper_RestrictionButton
   widget=Button_CompactChat_RestrictionAction
   properties {
    bHideInputAction=true
    backgroundMaterial=&MI_UIKit_BgQuiet_Fallback
    minHeightOverride=54
    mobileMinHeight=80
    text=Restriction action
    textTransformPolicy=None
    font {
     bForceMonospaced=false
     bMaterialIsStencil=false
     fontMaterial=/FortUILibrary/UIKit/Buttons/MaterialInstances/CTAButtons/MI_UIKit_TextColor_Quiet
     fontObject=&HeadingNow
     letterSpacing=10
     monospacedWidth=1
     size=15
     skewAmount=0
     typefaceFontName=74 Regular
     outlineSettings {
      bApplyOutlineToDropShadows=false
      bMiteredCorners=false
      bSeparateFillAlpha=false
      outlineMaterial=&MI_UIKit_TextOutline
      outlineSize=0
      outlineColor=<rgba 0 0 0 1>
     }
    }
   }
   slot {class=OverlaySlot, properties {horizontalAlignment=HAlign_Fill}}
  }
  CommonVisualAttachment {
   widgetClassPath=&CommonVisualAttachment
   bIsVariable=true
   parent=GP_TextChat_Wrapper
   widget=CommonVisualAttachment
   properties {contentAnchor=<1 0>}
   slot {class=GridSlot, properties {horizontalAlignment=HAlign_Right, row=2}}
  }
  WBP_ActionBarContainer {
   widgetClassPath=/Game/UI/Frontend/BottomBar/WBP_ActionBarContainer
   bIsVariable=true
   parent=CommonVisualAttachment
   widget=WBP_ActionBarContainer
   properties {backplateMaterial=&MI_UIKit_Backplate_Fallback}
   slot {class=SizeBoxSlot, properties {padding=<ltrb 0 16 0 0>}}
  }
  HorizontalBox_1 {widgetClassPath=&HorizontalBox, namedSlotHost=WBP_ActionBarContainer, widget=HorizontalBox_1}
  Button_StopTTS {
   widgetClassPath=/FortUILibrary/UIKit/Buttons/WBP_UIKit_Button_Quiet
   bIsVariable=true
   parent=HorizontalBox_1
   widget=Button_StopTTS
   properties {
    bIsFocusable=false
    minHeight=48
    text=Stop reading aloud
    useDisabledVisualsOnLocked=true
    visibility=Collapsed
    triggeringInputAction {dataTable=&InputActionDataTable, rowName=TextToSpeechPrimaryAction}
   }
   slot {class=HorizontalBoxSlot, properties {horizontalAlignment=HAlign_Left, padding=<ltrb 0 0 32 0>}}
  }
  Button_StopTTS_Decoy {
   widgetClassPath=/Game/UI/Frontend/BottomBar/WBP_BottomBarDecoyButton
   bIsVariable=true
   namedSlotHost=Button_StopTTS
   widget=Button_StopTTS_Decoy
   properties {bDisplayInActionBar=false, triggeringInputAction {dataTable=&InputActionDataTable, rowName=TextToSpeechSecondaryAction}}
  }
  Button_ReportChat {
   widgetClassPath=&WBP_UIKit_Button_Quiet
   bIsVariable=true
   parent=HorizontalBox_1
   widget=Button_ReportChat
   properties {
    bIsFocusable=false
    minHeight=48
    text=Report
    useDisabledVisualsOnLocked=true
    triggeringInputAction {dataTable=&InputActionDataTable, rowName=Report}
   }
   slot {class=HorizontalBoxSlot, properties {horizontalAlignment=HAlign_Left, padding=<ltrb 0 0 32 0>}}
  }
  Button_BackChat {
   widgetClassPath=&WBP_UIKit_Button_Quiet
   bIsVariable=true
   parent=HorizontalBox_1
   widget=Button_BackChat
   properties {bIsFocusable=false, text=Back, triggeringInputAction {dataTable=&InputActionDataTable, rowName=GenericBack}}
   slot {class=HorizontalBoxSlot, properties {horizontalAlignment=HAlign_Left}}
  }
  VisAttach_CoachmarkSettings {
   widgetClassPath=&CommonVisualAttachment
   bIsVariable=true
   parent=GP_TextChat_Wrapper
   widget=VisAttach_CoachmarkSettings
   properties {contentAnchor=<1 1>}
   slot {class=GridSlot, properties {horizontalAlignment=HAlign_Right, layer=50, row=2, verticalAlignment=VAlign_Bottom, nudge=<-80 16>}}
  }
  WBP_TextChat_ChatSettings_CoachMark {
   widgetClassPath=/FortniteCoreUI/UI/Social/Chat/TextChat/Coachmarks/WBP_TextChat_ChatSettings_CoachMark
   bIsVariable=true
   parent=VisAttach_CoachmarkSettings
   widget=WBP_TextChat_ChatSettings_CoachMark
   properties {bIsModal=false}
   slot {class=SizeBoxSlot}
  }
  VisAttach_CoachmarkReport {
   widgetClassPath=&CommonVisualAttachment
   bIsVariable=true
   parent=GP_TextChat_Wrapper
   widget=VisAttach_CoachmarkReport
   properties {contentAnchor=<1 1>}
   slot {class=GridSlot, properties {horizontalAlignment=HAlign_Right, layer=60, rowSpan=2, verticalAlignment=VAlign_Bottom, nudge=<-80 16>}}
  }
  WBP_TextChat_NothingToReport_CoachMark {
   widgetClassPath=/FortniteCoreUI/UI/Social/Chat/TextChat/Coachmarks/WBP_TextChat_NothingToReport_CoachMark
   bIsVariable=true
   parent=VisAttach_CoachmarkReport
   widget=WBP_TextChat_NothingToReport_CoachMark
   properties {bIsModal=false}
   slot {class=SizeBoxSlot}
  }
  VisAttach_CoachmarkSTT {
   widgetClassPath=&CommonVisualAttachment
   bIsVariable=true
   parent=GP_TextChat_Wrapper
   widget=VisAttach_CoachmarkSTT
   properties {contentAnchor=<0.9 0.95>}
   slot {class=GridSlot, properties {horizontalAlignment=HAlign_Left, layer=61, rowSpan=2, verticalAlignment=VAlign_Bottom, nudge=<-80 16>}}
  }
  WBP_TextChat_STT_CoachMark {
   widgetClassPath=/FortniteCoreUI/UI/Social/Chat/TextChat/Coachmarks/WBP_TextChat_STT_CoachMark
   bIsVariable=true
   parent=VisAttach_CoachmarkSTT
   widget=WBP_TextChat_STT_CoachMark
   properties {bIsModal=false}
   slot {class=SizeBoxSlot}
  }
 ]
}
