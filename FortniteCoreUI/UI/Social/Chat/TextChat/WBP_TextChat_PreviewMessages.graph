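# Text chat preview widget showing messages in a bottom-aligned list over a dimmed backplate.
# FortTextChatPlayerPreviewVM opens and closes with activation, and its received-message event plays a sound if chat sounds are enabled.
# FillOutPreviewData refills and scrolls the list, hiding the backplate when empty; the clipping mask updates on entry generation or empty data.
# Compact UI adjusts layout, and helper functions broadcast unread-channel count and prompt-button visibility through the global event router.
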WidgetBlueprint /FortniteCoreUI/UI/Social/Chat/TextChat/WBP_TextChat_PreviewMessages
graphs [
 EventGraph {nodes [
   OnListEntryGenerated!() -> (OutputDelegate, then, Widget:User Widget Object Reference) {is=title K2Node_CustomEvent}
   > UpdateClippingMask()
   OnMessageReceived() {is=title K2Node_CustomEvent}
   > PlayTextChatSound!(execute, self, Sound:Sound Base Object Reference=Get MessageReceivedSound() -> (MessageReceivedSound:Sound Base Object Reference)) -> (then)
   Event PreConstruct()
   > SetRenderOpacity(self=Get Backplate_ClickCatcher() -> (Backplate_ClickCatcher:WBP UIKit Backplate Object Reference), InOpacity=0.7)
   Event On Activated()
   n1 = Get FortTextChatPlayerPreviewVM() -> (FortTextChatPlayerPreviewVM:Fort Text Chat Player Preview VM Object Reference)
   > OnOpen(self:Fort Text Chat Player Preview VM Object Reference=@n1)
   Event OnInitialized()
   > InitializeTextChatVM(self:Fort Text Chat Player Preview VM Object Reference=Get FortTextChatPlayerPreviewVM())
   > SetBackplateVisiblity()
   > SetInGameAdjustments()
   > SetCompactUISpacing()
   Event On Deactivated()
   > OnClose(self:Fort Text Chat Player Preview VM Object Reference=@n1)
   Event Destruct()
   n2 = Create Event()
   n3 = Get ListView!(self:Fort Polymorphic List View Object Reference=Get WBP_TextChat_PreviewPolyList() -> (WBP_TextChat_PreviewPolyList:WBP Text Chat Preview Poly List Object Reference)) -> (ListView:Common List View Object Reference)
   > Unbind Event from On Entry Generated(self:List View Base Object Reference=@n3, Delegate=@n2) {is=title K2Node_RemoveDelegate}
   n4 = Create Event()
   n5 = Get FortTextChatPlayerPreviewVM()
   > Unbind Event from On Message Received(self:Fort Text Chat Player Preview VM Object Reference=@n5, Delegate=@n4) {is=title K2Node_RemoveDelegate}
   Event Construct()
   > SetCompactUIStyle() {is=title K2Node_CallFunction}
   > Bind Event to On Entry Generated(self:List View Base Object Reference=@n3, Delegate=@n2) {is=title K2Node_AddDelegate}
   > Bind Event to On Message Received(self:Fort Text Chat Player Preview VM Object Reference=@n5, Delegate=@n4) {is=title K2Node_AddDelegate}
  ]}
 FillOutPreviewData {nodes [
   n1 = FillOutPreviewData!() -> (then, Messages:Array of Text Chat Message VM Object References) {is=title K2Node_FunctionEntry}
   > ClearListItems(self=Get ListView(self=Get WBP_TextChat_PreviewPolyList()))
   n2 = Is Not Empty!(TargetArray:Array of Text Chat Message VM Object References=@n1) -> (ReturnValue:bool)
   > SetVisibility(self=Get Backplate_ClickCatcher(), InVisibility=Select(Option 0=Collapsed, Option 1=SelfHitTestInvisible, Index=@n2))
   > n3 = Branch(Condition=@n2)
   n4 = Get ListView(self=Get WBP_TextChat_PreviewPolyList())
   > Set List Items!(execute, self:List View Object Reference=@n4, InListItems:Array of Object References=@n1) -> (then) {is=title K2Node_CallFunction}
   > ScrollToBottom(self:List View Base Object Reference=@n4)
   UpdateClippingMask(execute=@n3.else)
  ]}
 PlayTextChatSound {nodes [
   n1 = PlayTextChatSound!() -> (then, Sound:Sound Base Object Reference) {is=title K2Node_FunctionEntry}
   > Branch(Condition=IsTextChatSoundEnabled!(self:Fort Text Chat Player Preview VM Object Reference=Get FortTextChatPlayerPreviewVM()) -> (ReturnValue:bool))
   > Is Valid(InputObject=@n1)
   > PlaySound2D(Sound=@n1)
  ]}
 SetBackplateVisiblity {nodes [
   SetBackplateVisiblity() {is=title K2Node_FunctionEntry}
   n1 = GetListItems!(self:List View Object Reference=Get ListView(self=Get WBP_TextChat_PreviewPolyList())) -> (ReturnValue:Array of Object References)
   > SetVisibility(
    self=Get Backplate_ClickCatcher()
    InVisibility=Select(Option 0=Collapsed, Option 1=SelfHitTestInvisible, Index=Is Not Empty(TargetArray:Array of Object References=@n1))
   )
  ]}
 SetButtonChatPromptVisiblity {nodes [
   n1 = SetButtonChatPromptVisiblity!() -> (then, IsShouldShowChatPromptButton:bool=~) {is=title K2Node_FunctionEntry}
   > Broadcast Event!(
    execute
    self:Gameplay Event Router Component Object Reference=GetGlobalEventRouter()
    EventData:BPS Set Text Chat Prompt Button Visibility Structure=Make BPS Set Text Chat Prompt Button Visibility!(IsShouldShowChatPromptButton_2_#1:bool=@n1) -> (BPS_SetTextChatPromptButtonVisibility:BPS Set Text Chat Prompt Button Visibility Structure) {is=title K2Node_MakeStruct}
    EventContext:Object Reference=GetOwningPlayer()
   ) -> (then, ReturnValue:bool)
  ]}
 SetCompactUISpacing {nodes [
   SetCompactUISpacing() {is=title K2Node_FunctionEntry}
   > Branch(Condition=IsUsingCompactUI())
   > SetEntrySpacing!(execute, self:Common List View Object Reference=Get ListView(self=Get WBP_TextChat_PreviewPolyList()), InEntrySpacing:float=15) -> (then)
  ]}
 SetCompactUIStyle {nodes [
   SetCompactUIStyle() {is=title K2Node_FunctionEntry}
   > Branch(Condition=IsUsingCompactUI())
   > SetVisibility(self=Get ListView(self=Get WBP_TextChat_PreviewPolyList()), InVisibility=HitTestInvisible)
   n1 = Get SB_PreviewSize() -> (SB_PreviewSize:Size Box Object Reference)
   > SetMaxDesiredWidth!(execute, self:Size Box Object Reference=@n1, InMaxDesiredWidth:float=750) -> (then)
   > SetPadding!(
    execute
    self:Overlay Slot Object Reference=SlotAsOverlaySlot(Widget=@n1)
    InPadding_Left:float=25
    InPadding_Top:float=15
    InPadding_Right:float=0
    InPadding_Bottom:float=25
   ) -> (then)
  ]}
 SetInGameAdjustments {nodes [
   SetInGameAdjustments() {is=title K2Node_FunctionEntry}
   n1 = Get FortGlobalUIContext!() -> (ReturnValue:Fort Global UIContext Object Reference) {is=title K2Node_GetBlueprintContextEx}
   > SetMaxDesiredHeight!(
    execute
    self:Size Box Object Reference=Get SB_PreviewSize()
    InMaxDesiredHeight:float=Select(Option 0:float=291, Option 1:float=291, Index=IsInZone!(self:Fort Global UIContext Object Reference=@n1) -> (ReturnValue:bool)) -> (ReturnValue:float)
   ) -> (then)
  ]}
 SetNumChannelsWithUnreadMessages {nodes [
   n1 = SetNumChannelsWithUnreadMessages!() -> (then, NumChannelsWithUnreadMessages:int) {is=title K2Node_FunctionEntry}
   > Broadcast Event(
    self=GetGlobalEventRouter()
    EventData:BPS Bottom Bar Notification Count Structure=Make BPS Bottom Bar Notification Count!(NotificationCount_8_#2:int=@n1) -> (BPS_BottomBarNotificationCount:BPS Bottom Bar Notification Count Structure) {is=title K2Node_MakeStruct}
    EventContext=GetOwningPlayer()
   )
  ]}
 UpdateClippingMask {nodes [
   UpdateClippingMask() {is=title K2Node_FunctionEntry}
   n1 = Get FortTextChatPlayerPreviewVM()
   > SetEnableClippingMask!(
    execute
    self:Fort Maskable Widget Object Reference
    bEnableClipMask:bool=Is Not Empty(TargetArray:Array of Text Chat Message VM Object References=Get Messages!(self:Fort Text Chat Player Preview VM Object Reference=@n1) -> (Messages:Array of Text Chat Message VM Object References))
   ) -> (then)
  ]}
]
variables [MessageReceivedSound {type=SoundBase, default=/Game/Sounds/Fort_UI_Sounds/Chat_UI/MSS/MSS_UI_FrontEnd_Chat_MessageReceived}]
widgets {
 info {inheritedWidgetCount=0, parentClass=/Script/FortniteUI.FortMaskableWidget, rootWidgetClass=&Overlay, widgetCount=4}
 widgets [
  Overlay_0 {widgetClassPath=&Overlay, widget=Overlay_0, properties {clipping=ClipToBounds}}
  Backplate_ClickCatcher {
   widgetClassPath=&WBP_UIKit_Backplate
   bIsVariable=true
   parent=Overlay_0
   widget=Backplate_ClickCatcher
   properties {
    backplateMaterial=/FortUILibrary/Systems/Backplate/MaterialInstances/MI_UIKit_Backplate_Fallback
    brightness=Dark
    cornerRadius=Interactive Component
   }
   slot {class=OverlaySlot, properties {horizontalAlignment=HAlign_Fill, verticalAlignment=VAlign_Fill}}
  }
  SB_PreviewSize {
   widgetClassPath=&SizeBox
   bIsVariable=true
   parent=Overlay_0
   widget=SB_PreviewSize
   properties {bOverride_MaxDesiredHeight=true, bOverride_WidthOverride=true, maxDesiredHeight=291, widthOverride=566}
   slot {class=OverlaySlot, properties {horizontalAlignment=HAlign_Fill, verticalAlignment=VAlign_Bottom, padding=<ltrb 12 12 12 12>}}
  }
  WBP_TextChat_PreviewPolyList {
   widgetClassPath=&/WBP_TextChat_PreviewPolyList
   bIsVariable=true
   parent=SB_PreviewSize
   widget=WBP_TextChat_PreviewPolyList
   slot {class=SizeBoxSlot, properties {verticalAlignment=VAlign_Bottom}}
  }
 ]
}
